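# Control Rig for the M_MED_Downpour male character cosmetic that adds procedural secondary motion to its back, front and skirt cloth bones.
# From BeginExecution, fourteen ChainHarmonicsPerItem units run in sequence, one per left or right dyn_cloth chain root.
# All share reach and pendulum settings (zero gravity vector), while wave settings, mainly noise, vary between chains to differentiate their sway.
# Debug drawing is enabled on every unit, and the dyn_pouch bone is not driven.
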ControlRigBlueprint /BRCosmetics/Characters/Player/Male/Medium/Bodies/M_MED_Downpour/Meshes/M_MED_Downpour_CtrlRig
graphs [
 RigVMFunctionLibrary {links [], nodes []}
 RigVMModel {nodes [
   RigUnit_BeginExecution = &RigUnit_BeginExecution()
   /Script/ControlRig.RigUnit_ChainHarmonicsPerItem(
    ChainRoot=(Type=Bone,Name="dyn_clothBack_outer_01_l")
    DrawWorldOffset=(Rotation=(X=0,Y=0,Z=0,W=1),Translation=(X=0,Y=0,Z=0),Scale3D=(X=1,Y=1,Z=1))
    Pendulum=(bEnabled=True,PendulumStiffness=2,PendulumGravity=(X=0,Y=0,Z=0),PendulumBlend=0.75,PendulumDrag=0.98,PendulumMaximum=0.1,UnwindAxis=(X=0,Y=1,Z=0),UnwindMinimum=0.2,UnwindMaximum=0.05)
    Reach=(bEnabled=True,ReachTarget=(X=0,Y=0,Z=0),ReachAxis=(X=1,Y=0,Z=0),ReachEase=SineEaseInOut)
    Speed=(X=0.15,Y=0.15,Z=0.15)
    Wave=(bEnabled=True,WaveFrequency=(X=0.2,Y=0.2,Z=0.2),WaveAmplitude=(X=0.1,Y=0.2,Z=0.2),WaveOffset=(X=1,Y=2,Z=0.5),WaveNoise=(X=5,Y=13,Z=10),WaveMinimum=0.2,WaveMaximum=0.6,WaveEase=SineEaseIn)
    WaveCurve=(EditorCurveData=(Keys=((Value=1),(Time=1,Value=2))))
    bDrawDebug=True
    ExecutePin=@RigUnit_BeginExecution.ExecutePin
   ) {is=struct RigVMUnitNode}
   > &RigUnit_ChainHarmonicsPerItem(
    ChainRoot=(Type=Bone,Name="dyn_clothBack_outer_01_r")
    DrawWorldOffset=(Rotation=(X=0,Y=0,Z=0,W=1),Translation=(X=0,Y=0,Z=0),Scale3D=(X=1,Y=1,Z=1))
    Pendulum=(bEnabled=True,PendulumStiffness=2,PendulumGravity=(X=0,Y=0,Z=0),PendulumBlend=0.75,PendulumDrag=0.98,PendulumMaximum=0.1,UnwindAxis=(X=0,Y=1,Z=0),UnwindMinimum=0.2,UnwindMaximum=0.05)
    Reach=(bEnabled=True,ReachTarget=(X=0,Y=0,Z=0),ReachAxis=(X=1,Y=0,Z=0),ReachEase=SineEaseInOut)
    Speed=(X=0.15,Y=0.15,Z=0.15)
    Wave=(bEnabled=True,WaveFrequency=(X=0.2,Y=0.2,Z=0.2),WaveAmplitude=(X=0.1,Y=0.2,Z=0.2),WaveOffset=(X=1,Y=2,Z=0.5),WaveNoise=(X=5,Y=15,Z=10),WaveMinimum=0.2,WaveMaximum=0.5,WaveEase=SineEaseIn)
    WaveCurve=(EditorCurveData=(Keys=((Value=1),(Time=1,Value=2))))
    bDrawDebug=True
   )
   > &RigUnit_ChainHarmonicsPerItem(
    ChainRoot=(Type=Bone,Name="dyn_clothBack_inner_01_l")
    DrawWorldOffset=(Rotation=(X=0,Y=0,Z=0,W=1),Translation=(X=0,Y=0,Z=0),Scale3D=(X=1,Y=1,Z=1))
    Pendulum=(bEnabled=True,PendulumStiffness=2,PendulumGravity=(X=0,Y=0,Z=0),PendulumBlend=0.75,PendulumDrag=0.98,PendulumMaximum=0.1,UnwindAxis=(X=0,Y=1,Z=0),UnwindMinimum=0.2,UnwindMaximum=0.05)
    Reach=(bEnabled=True,ReachTarget=(X=0,Y=0,Z=0),ReachAxis=(X=1,Y=0,Z=0),ReachEase=SineEaseInOut)
    Speed=(X=0.15,Y=0.15,Z=0.15)
    Wave=(bEnabled=True,WaveFrequency=(X=0.2,Y=0.2,Z=0.2),WaveAmplitude=(X=0.1,Y=0.2,Z=0.1),WaveOffset=(X=1,Y=2,Z=0),WaveNoise=(X=5,Y=14,Z=5),WaveMinimum=0.2,WaveMaximum=0.6,WaveEase=SineEaseIn)
    WaveCurve=(EditorCurveData=(Keys=((Value=1),(Time=1,Value=2))))
    bDrawDebug=True
   )
   > &RigUnit_ChainHarmonicsPerItem(
    ChainRoot=(Type=Bone,Name="dyn_clothBack_inner_01_r")
    DrawWorldOffset=(Rotation=(X=0,Y=0,Z=0,W=1),Translation=(X=0,Y=0,Z=0),Scale3D=(X=1,Y=1,Z=1))
    Pendulum=(bEnabled=True,PendulumStiffness=2,PendulumGravity=(X=0,Y=0,Z=0),PendulumBlend=0.75,PendulumDrag=0.98,PendulumMaximum=0.1,UnwindAxis=(X=0,Y=1,Z=0),UnwindMinimum=0.2,UnwindMaximum=0.05)
    Reach=(bEnabled=True,ReachTarget=(X=0,Y=0,Z=0),ReachAxis=(X=1,Y=0,Z=0),ReachEase=SineEaseInOut)
    Speed=(X=0.15,Y=0.15,Z=0.15)
    Wave=(bEnabled=True,WaveFrequency=(X=0.2,Y=0.2,Z=0.2),WaveAmplitude=(X=0.1,Y=0.2,Z=0.1),WaveOffset=(X=1,Y=2,Z=0),WaveNoise=(X=5,Y=16,Z=5),WaveMinimum=0.2,WaveMaximum=0.5,WaveEase=SineEaseIn)
    WaveCurve=(EditorCurveData=(Keys=((Value=1),(Time=1,Value=2))))
    bDrawDebug=True
   )
   > &RigUnit_ChainHarmonicsPerItem(
    ChainRoot=(Type=Bone,Name="dyn_clothBack_mid_01_l")
    DrawWorldOffset=(Rotation=(X=0,Y=0,Z=0,W=1),Translation=(X=0,Y=0,Z=0),Scale3D=(X=1,Y=1,Z=1))
    Pendulum=(bEnabled=True,PendulumStiffness=2,PendulumGravity=(X=0,Y=0,Z=0),PendulumBlend=0.75,PendulumDrag=0.98,PendulumMaximum=0.1,UnwindAxis=(X=0,Y=1,Z=0),UnwindMinimum=0.2,UnwindMaximum=0.05)
    Reach=(bEnabled=True,ReachTarget=(X=0,Y=0,Z=0),ReachAxis=(X=1,Y=0,Z=0),ReachEase=SineEaseInOut)
    Speed=(X=0.15,Y=0.15,Z=0.15)
    Wave=(bEnabled=True,WaveFrequency=(X=0.2,Y=0.2,Z=0.2),WaveAmplitude=(X=0.1,Y=0.2,Z=0.2),WaveOffset=(X=1,Y=2,Z=0.5),WaveNoise=(X=5,Y=14,Z=10),WaveMinimum=0.2,WaveMaximum=0.6,WaveEase=SineEaseIn)
    WaveCurve=(EditorCurveData=(Keys=((Value=1),(Time=1,Value=2))))
    bDrawDebug=True
   )
   > &RigUnit_ChainHarmonicsPerItem(
    ChainRoot=(Type=Bone,Name="dyn_clothBack_mid_01_r")
    DrawWorldOffset=(Rotation=(X=0,Y=0,Z=0,W=1),Translation=(X=0,Y=0,Z=0),Scale3D=(X=1,Y=1,Z=1))
    Pendulum=(bEnabled=True,PendulumStiffness=2,PendulumGravity=(X=0,Y=0,Z=0),PendulumBlend=0.75,PendulumDrag=0.98,PendulumMaximum=0.1,UnwindAxis=(X=0,Y=1,Z=0),UnwindMinimum=0.2,UnwindMaximum=0.05)
    Reach=(bEnabled=True,ReachTarget=(X=0,Y=0,Z=0),ReachAxis=(X=1,Y=0,Z=0),ReachEase=SineEaseInOut)
    Speed=(X=0.15,Y=0.15,Z=0.15)
    Wave=(bEnabled=True,WaveFrequency=(X=0.2,Y=0.2,Z=0.2),WaveAmplitude=(X=0.1,Y=0.2,Z=0.2),WaveOffset=(X=1,Y=2,Z=0.5),WaveNoise=(X=5,Y=13,Z=8),WaveMinimum=0.2,WaveMaximum=0.5,WaveEase=SineEaseIn)
    WaveCurve=(EditorCurveData=(Keys=((Value=1),(Time=1,Value=2))))
    bDrawDebug=True
   )
   > &RigUnit_ChainHarmonicsPerItem(
    ChainRoot=(Type=Bone,Name="dyn_clothFront_inner_01_l")
    DrawWorldOffset=(Rotation=(X=0,Y=0,Z=0,W=1),Translation=(X=0,Y=0,Z=0),Scale3D=(X=1,Y=1,Z=1))
    Pendulum=(bEnabled=True,PendulumStiffness=2,PendulumGravity=(X=0,Y=0,Z=0),PendulumBlend=0.75,PendulumDrag=0.98,PendulumMaximum=0.1,UnwindAxis=(X=0,Y=1,Z=0),UnwindMinimum=0.2,UnwindMaximum=0.05)
    Reach=(bEnabled=True,ReachTarget=(X=0,Y=0,Z=0),ReachAxis=(X=1,Y=0,Z=0),ReachEase=SineEaseInOut)
    Speed=(X=0.15,Y=0.15,Z=0.15)
    Wave=(bEnabled=True,WaveFrequency=(X=0.2,Y=0.2,Z=0.2),WaveAmplitude=(X=0.1,Y=0.2,Z=0.1),WaveOffset=(X=1,Y=2,Z=0),WaveNoise=(X=5,Y=15,Z=10),WaveMinimum=0.2,WaveMaximum=0.5,WaveEase=SineEaseIn)
    WaveCurve=(EditorCurveData=(Keys=((Value=1),(Time=1,Value=2))))
    bDrawDebug=True
   )
   > &RigUnit_ChainHarmonicsPerItem(
    ChainRoot=(Type=Bone,Name="dyn_clothFront_inner_01_r")
    DrawWorldOffset=(Rotation=(X=0,Y=0,Z=0,W=1),Translation=(X=0,Y=0,Z=0),Scale3D=(X=1,Y=1,Z=1))
    Pendulum=(bEnabled=True,PendulumStiffness=2,PendulumGravity=(X=0,Y=0,Z=0),PendulumBlend=0.75,PendulumDrag=0.98,PendulumMaximum=0.1,UnwindAxis=(X=0,Y=1,Z=0),UnwindMinimum=0.2,UnwindMaximum=0.05)
    Reach=(bEnabled=True,ReachTarget=(X=0,Y=0,Z=0),ReachAxis=(X=1,Y=0,Z=0),ReachEase=SineEaseInOut)
    Speed=(X=0.15,Y=0.15,Z=0.15)
    Wave=(bEnabled=True,WaveFrequency=(X=0.2,Y=0.2,Z=0.2),WaveAmplitude=(X=0.1,Y=0.2,Z=0.1),WaveOffset=(X=1,Y=2,Z=0),WaveNoise=(X=5,Y=17,Z=10),WaveMinimum=0.2,WaveMaximum=0.5,WaveEase=SineEaseIn)
    WaveCurve=(EditorCurveData=(Keys=((Value=1),(Time=1,Value=2))))
    bDrawDebug=True
   )
   > &RigUnit_ChainHarmonicsPerItem(
    ChainRoot=(Type=Bone,Name="dyn_clothFront_outer_01_l")
    DrawWorldOffset=(Rotation=(X=0,Y=0,Z=0,W=1),Translation=(X=0,Y=0,Z=0),Scale3D=(X=1,Y=1,Z=1))
    Pendulum=(bEnabled=True,PendulumStiffness=2,PendulumGravity=(X=0,Y=0,Z=0),PendulumBlend=0.75,PendulumDrag=0.98,PendulumMaximum=0.1,UnwindAxis=(X=0,Y=1,Z=0),UnwindMinimum=0.2,UnwindMaximum=0.05)
    Reach=(bEnabled=True,ReachTarget=(X=0,Y=0,Z=0),ReachAxis=(X=1,Y=0,Z=0),ReachEase=SineEaseInOut)
    Speed=(X=0.15,Y=0.15,Z=0.15)
    Wave=(bEnabled=True,WaveFrequency=(X=0.2,Y=0.2,Z=0.2),WaveAmplitude=(X=0.1,Y=0.2,Z=0.2),WaveOffset=(X=1,Y=2,Z=1),WaveNoise=(X=5,Y=15,Z=10),WaveMinimum=0.2,WaveMaximum=0.5,WaveEase=SineEaseIn)
    WaveCurve=(EditorCurveData=(Keys=((Value=1),(Time=1,Value=2))))
    bDrawDebug=True
   )
   > &RigUnit_ChainHarmonicsPerItem(
    ChainRoot=(Type=Bone,Name="dyn_clothFront_outer_01_r")
    DrawWorldOffset=(Rotation=(X=0,Y=0,Z=0,W=1),Translation=(X=0,Y=0,Z=0),Scale3D=(X=1,Y=1,Z=1))
    Pendulum=(bEnabled=True,PendulumStiffness=2,PendulumGravity=(X=0,Y=0,Z=0),PendulumBlend=0.75,PendulumDrag=0.98,PendulumMaximum=0.1,UnwindAxis=(X=0,Y=1,Z=0),UnwindMinimum=0.2,UnwindMaximum=0.05)
    Reach=(bEnabled=True,ReachTarget=(X=0,Y=0,Z=0),ReachAxis=(X=1,Y=0,Z=0),ReachEase=SineEaseInOut)
    Speed=(X=0.15,Y=0.15,Z=0.15)
    Wave=(bEnabled=True,WaveFrequency=(X=0.2,Y=0.2,Z=0.2),WaveAmplitude=(X=0.1,Y=0.2,Z=0.2),WaveOffset=(X=1,Y=2,Z=1),WaveNoise=(X=5,Y=17,Z=10),WaveMinimum=0.2,WaveMaximum=0.5,WaveEase=SineEaseIn)
    WaveCurve=(EditorCurveData=(Keys=((Value=1),(Time=1,Value=2))))
    bDrawDebug=True
   )
   > &RigUnit_ChainHarmonicsPerItem(
    ChainRoot=(Type=Bone,Name="dyn_clothSkirt_inner_01_l")
    DrawWorldOffset=(Rotation=(X=0,Y=0,Z=0,W=1),Translation=(X=0,Y=0,Z=0),Scale3D=(X=1,Y=1,Z=1))
    Pendulum=(bEnabled=True,PendulumStiffness=2,PendulumGravity=(X=0,Y=0,Z=0),PendulumBlend=0.75,PendulumDrag=0.98,PendulumMaximum=0.1,UnwindAxis=(X=0,Y=1,Z=0),UnwindMinimum=0.2,UnwindMaximum=0.05)
    Reach=(bEnabled=True,ReachTarget=(X=0,Y=0,Z=0),ReachAxis=(X=1,Y=0,Z=0),ReachEase=SineEaseInOut)
    Speed=(X=0.15,Y=0.15,Z=0.15)
    Wave=(bEnabled=True,WaveFrequency=(X=0.2,Y=0.2,Z=0.2),WaveAmplitude=(X=0.1,Y=0.2,Z=0.1),WaveOffset=(X=1,Y=2,Z=0),WaveNoise=(X=5,Y=15,Z=5),WaveMinimum=0.2,WaveMaximum=0.5,WaveEase=SineEaseIn)
    WaveCurve=(EditorCurveData=(Keys=((Value=1),(Time=1,Value=2))))
    bDrawDebug=True
   )
   > &RigUnit_ChainHarmonicsPerItem(
    ChainRoot=(Type=Bone,Name="dyn_clothSkirt_inner_01_r")
    DrawWorldOffset=(Rotation=(X=0,Y=0,Z=0,W=1),Translation=(X=0,Y=0,Z=0),Scale3D=(X=1,Y=1,Z=1))
    Pendulum=(bEnabled=True,PendulumStiffness=2,PendulumGravity=(X=0,Y=0,Z=0),PendulumBlend=0.75,PendulumDrag=0.98,PendulumMaximum=0.1,UnwindAxis=(X=0,Y=1,Z=0),UnwindMinimum=0.2,UnwindMaximum=0.05)
    Reach=(bEnabled=True,ReachTarget=(X=0,Y=0,Z=0),ReachAxis=(X=1,Y=0,Z=0),ReachEase=SineEaseInOut)
    Speed=(X=0.15,Y=0.15,Z=0.15)
    Wave=(bEnabled=True,WaveFrequency=(X=0.2,Y=0.2,Z=0.2),WaveAmplitude=(X=0.1,Y=0.2,Z=0.1),WaveOffset=(X=1,Y=2,Z=0),WaveNoise=(X=5,Y=12,Z=5),WaveMinimum=0.2,WaveMaximum=0.5,WaveEase=SineEaseIn)
    WaveCurve=(EditorCurveData=(Keys=((Value=1),(Time=1,Value=2))))
    bDrawDebug=True
   )
   > &RigUnit_ChainHarmonicsPerItem(
    ChainRoot=(Type=Bone,Name="dyn_clothSkirt_outer_01_l")
    DrawWorldOffset=(Rotation=(X=0,Y=0,Z=0,W=1),Translation=(X=0,Y=0,Z=0),Scale3D=(X=1,Y=1,Z=1))
    Pendulum=(bEnabled=True,PendulumStiffness=2,PendulumGravity=(X=0,Y=0,Z=0),PendulumBlend=0.75,PendulumDrag=0.98,PendulumMaximum=0.1,UnwindAxis=(X=0,Y=1,Z=0),UnwindMinimum=0.2,UnwindMaximum=0.05)
    Reach=(bEnabled=True,ReachTarget=(X=0,Y=0,Z=0),ReachAxis=(X=1,Y=0,Z=0),ReachEase=SineEaseInOut)
    Speed=(X=0.15,Y=0.15,Z=0.15)
    Wave=(bEnabled=True,WaveFrequency=(X=0.2,Y=0.2,Z=0.2),WaveAmplitude=(X=0.1,Y=0.2,Z=0.1),WaveOffset=(X=1,Y=2,Z=0),WaveNoise=(X=5,Y=13,Z=5),WaveMinimum=0.2,WaveMaximum=0.5,WaveEase=SineEaseIn)
    WaveCurve=(EditorCurveData=(Keys=((Value=1),(Time=1,Value=2))))
    bDrawDebug=True
   )
   > &RigUnit_ChainHarmonicsPerItem(
    ChainRoot=(Type=Bone,Name="dyn_clothSkirt_outer_01_r")
    DrawWorldOffset=(Rotation=(X=0,Y=0,Z=0,W=1),Translation=(X=0,Y=0,Z=0),Scale3D=(X=1,Y=1,Z=1))
    Pendulum=(bEnabled=True,PendulumStiffness=2,PendulumGravity=(X=0,Y=0,Z=0),PendulumBlend=0.75,PendulumDrag=0.98,PendulumMaximum=0.1,UnwindAxis=(X=0,Y=1,Z=0),UnwindMinimum=0.2,UnwindMaximum=0.05)
    Reach=(bEnabled=True,ReachTarget=(X=0,Y=0,Z=0),ReachAxis=(X=1,Y=0,Z=0),ReachEase=SineEaseInOut)
    Speed=(X=0.15,Y=0.15,Z=0.15)
    Wave=(bEnabled=True,WaveFrequency=(X=0.2,Y=0.2,Z=0.2),WaveAmplitude=(X=0.1,Y=0.2,Z=0.1),WaveOffset=(X=1,Y=2,Z=0),WaveNoise=(X=5,Y=15,Z=5),WaveMinimum=0.2,WaveMaximum=0.5,WaveEase=SineEaseIn)
    WaveCurve=(EditorCurveData=(Keys=((Value=1),(Time=1,Value=2))))
    bDrawDebug=True
   )
  ]}
]
hierarchy [
 root {loc=<0 0 0>, rot=<0 0 0 1>}
 pelvis {parent=root, loc=<0 -0.7071 93.1191>, rot=<0.0127 0.707 -0.0127 -0.707>}
 spine_01 {parent=pelvis, loc=<6.7494 -0.6388 0>, rot=<0 0 0.034 -0.9994>}
 spine_02 {parent=spine_01, loc=<7.2518 -0.035 0>, rot=<0 0 0.0238 -0.9997>}
 spine_03 {parent=spine_02, loc=<6.0104 0 0>, rot=<0 0 0.0149 -0.9999>}
 spine_04 {parent=spine_03, loc=<8.1024 0 0>, rot=<0 0 -0.0132 -0.9999>}
 spine_05 {parent=spine_04, loc=<10.0146 0 0>, rot=<0 0 -0.0015 -1>}
 clavicle_l {parent=spine_05, loc=<4.3575 -0.1466 -2.3223>, rot=<0.0465 -0.7289 -0.009 -0.6829>}
 upperarm_l {parent=clavicle_l, loc=<15.5111 0 0>, rot=<0.0226 -0.3234 -0.0088 -0.9459>}
 lowerarm_l {parent=upperarm_l, loc=<26.6514 0 0>, rot=<0.0391 -0.0099 -0.2453 -0.9686>}
 hand_l {parent=lowerarm_l, loc=<22.3491 0 0>, rot=<-0.7081 -0.0507 0.0975 -0.6975>}
 index_metacarpal_l {parent=hand_l, loc=<1.6535 0.0183 -2.96>, rot=<0.0158 -0.0124 -0.0173 -0.9996>}
 index_01_l {parent=index_metacarpal_l, loc=<8.3008 0.0357 -0.0524>, rot=<-0.0044 0.0485 0.0924 -0.9945>}
 index_02_l {parent=index_01_l, loc=<3.3587 -0.3004 0.0452>, rot=<-0.002 0.0027 0.3225 -0.9466>}
 index_03_l {parent=index_02_l, loc=<2.5578 -0.123 -0.0252>, rot=<-0.002 0.0067 0.2558 -0.9667>}
 middle_metacarpal_l {parent=hand_l, loc=<2.1856 -0.1725 -0.5984>, rot=<0.006 0.0168 -0.0329 -0.9993>}
 middle_01_l {parent=middle_metacarpal_l, loc=<7.3275 -0.2447 0.0129>, rot=<-0.0208 0.0042 0.0858 -0.9961>}
 middle_02_l {parent=middle_01_l, loc=<4.0123 0.1401 0.1175>, rot=<-0.0218 0.0258 0.3512 -0.9357>}
 middle_03_l {parent=middle_02_l, loc=<2.6077 0.0534 0.1373>, rot=<0.0082 0.0188 0.1708 -0.9851>}
 pinky_metacarpal_l {parent=hand_l, loc=<1.7616 -0.9753 3.3814>, rot=<-0.0397 0.0814 -0.069 -0.9935>}
 pinky_01_l {parent=pinky_metacarpal_l, loc=<7.1709 -0.4831 -0.1094>, rot=<-0.0654 -0.0035 0.2242 -0.9723>}
 pinky_02_l {parent=pinky_01_l, loc=<2.8274 0.0893 -0.1043>, rot=<0.0037 0.0011 0.2854 -0.9584>}
 pinky_03_l {parent=pinky_02_l, loc=<2.2046 0.0307 -0.0078>, rot=<-0.0002 0.0098 0.1445 -0.9895>}
 ring_metacarpal_l {parent=hand_l, loc=<2.2074 -0.3625 1.3638>, rot=<-0.0284 0.0722 -0.0458 -0.9959>}
 ring_01_l {parent=ring_metacarpal_l, loc=<7.0295 -0.6197 -0.1446>, rot=<-0.0489 -0.0252 0.1892 -0.9804>}
 ring_02_l {parent=ring_01_l, loc=<3.6174 0.4647 -0.1027>, rot=<0 0 0.2806 -0.9598>}
 ring_03_l {parent=ring_02_l, loc=<2.4152 0.114 0.0199>, rot=<0 0 0.1189 -0.9929>}
 thumb_01_l {parent=hand_l, loc=<1.3118 -0.3191 -2.459>, rot=<0.6546 -0.4042 0.0664 -0.6354>}
 thumb_02_l {parent=thumb_01_l, loc=<5.9336 -0.0448 0.0401>, rot=<-0.0037 0.0213 0.298 -0.9543>}
 thumb_03_l {parent=thumb_02_l, loc=<2.2223 0.3682 -0.0907>, rot=<0.0118 -0.013 0.0066 -0.9998>}
 lowerarm_twist_01_l {parent=lowerarm_l, loc=<17.8549 0 0>, rot=<0 0 0 -1>}
 lowerarm_twist_02_l {parent=lowerarm_l, loc=<9.6199 0 0>, rot=<0 0 0 -1>}
 upperarm_twist_01_l {parent=upperarm_l, loc=<0.0041 0 0>, rot=<0 0 0 -1>}
 upperarm_twist_02_l {parent=upperarm_l, loc=<13.7941 0 0>, rot=<0 0 0 -1>}
 dyn_clothFront_inner_mstr_l {parent=clavicle_l, loc=<14.4974 12.4167 -11.6222>, rot=<0.45 -0.5274 -0.4096 -0.5929>}
 dyn_clothFront_inner_01_l {parent=dyn_clothFront_inner_mstr_l, loc=<0 0 0>, rot=<0 0 0 1>}
 dyn_clothFront_inner_02_l {parent=dyn_clothFront_inner_01_l, loc=<19.7265 0 0>, rot=<-0.0207 -0.0126 -0.0036 -0.9997>}
 dyn_clothFront_inner_03_l {parent=dyn_clothFront_inner_02_l, loc=<19.3071 0 0>, rot=<-0.0135 0.0074 -0.0006 -0.9999>}
 dyn_clothFront_inner_04_l {parent=dyn_clothFront_inner_03_l, loc=<20.2858 0 0>, rot=<-0.0036 -0.0099 0.0045 -0.9999>}
 dyn_clothFront_inner_05_l {parent=dyn_clothFront_inner_04_l, loc=<19.647 0 0>, rot=<0 0 0 1>}
 dyn_clothFront_outer_mstr_l {parent=clavicle_l, loc=<19.6587 8.2912 -2.4504>, rot=<0.3378 -0.5586 -0.3342 -0.6798>}
 dyn_clothFront_outer_01_l {parent=dyn_clothFront_outer_mstr_l, loc=<0 0 0>, rot=<0 0 0 1>}
 dyn_clothFront_outer_02_l {parent=dyn_clothFront_outer_01_l, loc=<9.1253 0 0>, rot=<0.0014 -0.0494 -0.0012 -0.9988>}
 dyn_clothFront_outer_03_l {parent=dyn_clothFront_outer_02_l, loc=<19.4008 0 0>, rot=<-0.0056 -0.0029 -0.0037 -1>}
 dyn_clothFront_outer_04_l {parent=dyn_clothFront_outer_03_l, loc=<19.4119 0 0>, rot=<-0.0136 0.0009 0.0005 -0.9999>}
 dyn_clothFront_outer_05_l {parent=dyn_clothFront_outer_04_l, loc=<19.7138 0 0>, rot=<-0.0048 -0.0021 0.0063 -1>}
 dyn_clothFront_outer_06_l {parent=dyn_clothFront_outer_05_l, loc=<20.1105 0 0>, rot=<0 0 0 1>}
 dyn_clothBack_outer_mstr_l {parent=clavicle_l, loc=<17.8185 -7.5438 2.9283>, rot=<-0.4631 -0.4681 0.4207 -0.624>}
 dyn_clothBack_outer_01_l {parent=dyn_clothBack_outer_mstr_l, loc=<0 0 0>, rot=<0 0 0 1>}
 dyn_clothBack_outer_02_l {parent=dyn_clothBack_outer_01_l, loc=<17.1298 0 0>, rot=<-0.0264 -0.0432 0.0228 -0.9985>}
 dyn_clothBack_outer_03_l {parent=dyn_clothBack_outer_02_l, loc=<20.9204 0 0>, rot=<0.0257 -0.0095 0.0002 -0.9996>}
 dyn_clothBack_outer_04_l {parent=dyn_clothBack_outer_03_l, loc=<20.3234 0 0>, rot=<0.0048 -0.0004 0.0001 -1>}
 dyn_clothBack_outer_05_l {parent=dyn_clothBack_outer_04_l, loc=<20.3344 0 0>, rot=<-0.0007 0 0 -1>}
 dyn_clothBack_outer_06_l {parent=dyn_clothBack_outer_05_l, loc=<20.6194 0 0>, rot=<0 0 0 1>}
 clavicle_r {parent=spine_05, loc=<4.3575 -0.1466 2.3223>, rot=<0.7289 0.0465 -0.6829 0.009>}
 upperarm_r {parent=clavicle_r, loc=<-15.5111 0 0>, rot=<0.0226 -0.3234 -0.0088 -0.9459>}
 lowerarm_r {parent=upperarm_r, loc=<-26.6515 0 0>, rot=<0.0391 -0.0099 -0.2453 -0.9686>}
 hand_r {parent=lowerarm_r, loc=<-22.3492 0 0>, rot=<-0.7081 -0.0507 0.0975 -0.6975>}
 index_metacarpal_r {parent=hand_r, loc=<-1.6534 -0.0181 2.9601>, rot=<0.0158 -0.0124 -0.0173 -0.9996>}
 index_01_r {parent=index_metacarpal_r, loc=<-8.3008 -0.0357 0.0524>, rot=<-0.0044 0.0485 0.0924 -0.9945>}
 index_02_r {parent=index_01_r, loc=<-3.3589 0.3003 -0.0453>, rot=<-0.002 0.0027 0.3225 -0.9466>}
 index_03_r {parent=index_02_r, loc=<-2.5578 0.1231 0.0251>, rot=<-0.002 0.0067 0.2558 -0.9667>}
 middle_metacarpal_r {parent=hand_r, loc=<-2.1856 0.1726 0.5984>, rot=<0.006 0.0168 -0.0329 -0.9993>}
 middle_01_r {parent=middle_metacarpal_r, loc=<-7.3273 0.2447 -0.0129>, rot=<-0.0208 0.0042 0.0858 -0.9961>}
 middle_02_r {parent=middle_01_r, loc=<-4.0142 -0.1411 -0.07>, rot=<-0.0217 0.0259 0.3512 -0.9357>}
 middle_03_r {parent=middle_02_r, loc=<-2.6074 -0.0534 -0.1312>, rot=<0.0083 0.0188 0.1708 -0.9851>}
 pinky_metacarpal_r {parent=hand_r, loc=<-1.7617 0.9753 -3.3814>, rot=<-0.0397 0.0814 -0.069 -0.9935>}
 pinky_01_r {parent=pinky_metacarpal_r, loc=<-7.1707 0.4832 0.1094>, rot=<-0.0654 -0.0035 0.2242 -0.9723>}
 pinky_02_r {parent=pinky_01_r, loc=<-2.8274 -0.0892 0.1043>, rot=<0.0037 0.0011 0.2854 -0.9584>}
 pinky_03_r {parent=pinky_02_r, loc=<-2.2045 -0.031 0.0078>, rot=<-0.0002 0.0098 0.1445 -0.9895>}
 ring_metacarpal_r {parent=hand_r, loc=<-2.2073 0.3627 -1.3638>, rot=<-0.0284 0.0722 -0.0458 -0.9959>}
 ring_01_r {parent=ring_metacarpal_r, loc=<-7.0297 0.6197 0.1446>, rot=<-0.0489 -0.0252 0.1892 -0.9804>}
 ring_02_r {parent=ring_01_r, loc=<-3.6171 -0.4648 0.1027>, rot=<0 0 0.2806 -0.9598>}
 ring_03_r {parent=ring_02_r, loc=<-2.4153 -0.1141 -0.0199>, rot=<0 0 0.1189 -0.9929>}
 thumb_01_r {parent=hand_r, loc=<-1.3119 0.3192 2.459>, rot=<0.6546 -0.4042 0.0664 -0.6354>}
 thumb_02_r {parent=thumb_01_r, loc=<-5.9336 0.0447 -0.0401>, rot=<-0.0037 0.0213 0.298 -0.9543>}
 thumb_03_r {parent=thumb_02_r, loc=<-2.2223 -0.3682 0.0907>, rot=<0.0118 -0.013 0.0066 -0.9998>}
 lowerarm_twist_01_r {parent=lowerarm_r, loc=<-17.855 0 0>, rot=<0 0 0 1>}
 lowerarm_twist_02_r {parent=lowerarm_r, loc=<-9.6199 0 0>, rot=<0 0 0 1>}
 upperarm_twist_01_r {parent=upperarm_r, loc=<-0.0041 0 0>, rot=<0 0 0 1>}
 upperarm_twist_02_r {parent=upperarm_r, loc=<-13.7942 0 0>, rot=<0 0 0 1>}
 dyn_clothFront_inner_mstr_r {parent=clavicle_r, loc=<-14.4974 -12.4167 11.6223>, rot=<-0.4096 0.5929 -0.45 -0.5274>}
 dyn_clothFront_inner_01_r {parent=dyn_clothFront_inner_mstr_r, loc=<0 0 0>, rot=<0 0 0 1>}
 dyn_clothFront_inner_02_r {parent=dyn_clothFront_inner_01_r, loc=<19.7265 0 0>, rot=<-0.023 -0.0125 0.0042 -0.9997>}
 dyn_clothFront_inner_03_r {parent=dyn_clothFront_inner_02_r, loc=<19.3071 0 0>, rot=<0.0135 0.0074 -0.0001 -0.9999>}
 dyn_clothFront_inner_04_r {parent=dyn_clothFront_inner_03_r, loc=<20.2858 0 0>, rot=<0.0036 -0.0103 -0.0036 -0.9999>}
 dyn_clothFront_inner_05_r {parent=dyn_clothFront_inner_04_r, loc=<19.647 0 0>, rot=<0 0 0 1>}
 dyn_clothFront_outer_mstr_r {parent=clavicle_r, loc=<-19.6587 -8.2912 2.4505>, rot=<-0.3342 0.6798 -0.3378 -0.5586>}
 dyn_clothFront_outer_01_r {parent=dyn_clothFront_outer_mstr_r, loc=<0 0 0>, rot=<0 0 0 -1>}
 dyn_clothFront_outer_02_r {parent=dyn_clothFront_outer_01_r, loc=<9.1253 0 0>, rot=<-0.0014 -0.0494 0.0012 -0.9988>}
 dyn_clothFront_outer_03_r {parent=dyn_clothFront_outer_02_r, loc=<19.4008 0 0>, rot=<0.0056 -0.0029 0.0037 -1>}
 dyn_clothFront_outer_04_r {parent=dyn_clothFront_outer_03_r, loc=<19.4119 0 0>, rot=<0.0136 0.0009 -0.0005 -0.9999>}
 dyn_clothFront_outer_05_r {parent=dyn_clothFront_outer_04_r, loc=<19.7138 0 0>, rot=<0.0048 -0.0021 -0.0063 -1>}
 dyn_clothFront_outer_06_r {parent=dyn_clothFront_outer_05_r, loc=<20.1105 0 0>, rot=<0 0 0 -1>}
 dyn_clothBack_outer_mstr_r {parent=clavicle_r, loc=<-17.8185 7.5438 -2.9282>, rot=<-0.3999 -0.6436 -0.4355 0.486>}
 dyn_clothBack_outer_01_r {parent=dyn_clothBack_outer_mstr_r, loc=<0 0 0>, rot=<0 0 0 1>}
 dyn_clothBack_outer_02_r {parent=dyn_clothBack_outer_01_r, loc=<17.1298 0 0>, rot=<-0.0172 -0.0458 -0.0169 -0.9987>}
 dyn_clothBack_outer_03_r {parent=dyn_clothBack_outer_02_r, loc=<20.9204 0 0>, rot=<-0.0257 -0.0094 0.0014 -0.9996>}
 dyn_clothBack_outer_04_r {parent=dyn_clothBack_outer_03_r, loc=<20.3234 0 0>, rot=<-0.0048 -0.0004 0 -1>}
 dyn_clothBack_outer_05_r {parent=dyn_clothBack_outer_04_r, loc=<20.3344 0 0>, rot=<0.0007 0 -0.0001 -1>}
 dyn_clothBack_outer_06_r {parent=dyn_clothBack_outer_05_r, loc=<20.6194 0 0>, rot=<0 0 0 -1>}
 neck_01 {parent=spine_05, loc=<11.3138 0.1529 0>, rot=<0 0 -0.2003 -0.9797>}
 neck_02 {parent=neck_01, loc=<4.5627 0 0>, rot=<0 0 0.0504 -0.9987>}
 head {parent=neck_02, loc=<5.3914 0 0>, rot=<0 0 0.115 -0.9934>}
 capeBase {parent=spine_05, loc=<4.8985 -12.4038 0>, rot=<0.0284 -0.7065 -0.0284 -0.7065>}
 dyn_clothBack_mid_01_l {parent=capeBase, loc=<13.4275 1.3098 0.1559>, rot=<-0.4376 -0.5223 0.4354 -0.5884>}
 dyn_clothBack_mid_02_l {parent=dyn_clothBack_mid_01_l, loc=<17.3943 0 0>, rot=<0.0197 -0.0184 -0.0034 -0.9996>}
 dyn_clothBack_mid_03_l {parent=dyn_clothBack_mid_02_l, loc=<21.4882 0 0>, rot=<-0.0436 0.0041 -0.0003 -0.999>}
 dyn_clothBack_mid_04_l {parent=dyn_clothBack_mid_03_l, loc=<20.446 0 0>, rot=<-0.0127 0.0013 -0.0002 -0.9999>}
 dyn_clothBack_mid_05_l {parent=dyn_clothBack_mid_04_l, loc=<20.5943 0 0>, rot=<0.0264 -0.0026 0.0003 -0.9996>}
 dyn_clothBack_mid_06_l {parent=dyn_clothBack_mid_05_l, loc=<20.8999 0 0>, rot=<0 0 0 -1>}
 dyn_clothBack_inner_01_l {parent=capeBase, loc=<5.1626 -1.2192 -3.3595>, rot=<-0.446 -0.5161 0.4713 -0.5591>}
 dyn_clothBack_inner_02_l {parent=dyn_clothBack_inner_01_l, loc=<17.5323 0 0>, rot=<0.0075 -0.014 -0.0064 -0.9999>}
 dyn_clothBack_inner_03_l {parent=dyn_clothBack_inner_02_l, loc=<22.7192 0 0>, rot=<-0.0133 0.0017 -0.0002 -0.9999>}
 dyn_clothBack_inner_04_l {parent=dyn_clothBack_inner_03_l, loc=<22.0583 0 0>, rot=<0.0527 -0.0042 -0.0005 -0.9986>}
 dyn_clothBack_inner_05_l {parent=dyn_clothBack_inner_04_l, loc=<20.7254 0 0>, rot=<-0.0302 0.0008 0.0011 -0.9995>}
 dyn_clothBack_inner_06_l {parent=dyn_clothBack_inner_05_l, loc=<20.658 0 0>, rot=<0 0 0 1>}
 dyn_clothBack_inner_01_r {parent=capeBase, loc=<-5.1626 -1.2192 -3.3595>, rot=<-0.5161 -0.446 0.5591 -0.4713>}
 dyn_clothBack_inner_02_r {parent=dyn_clothBack_inner_01_r, loc=<17.5323 0 0>, rot=<-0.0075 -0.014 0.0064 -0.9999>}
 dyn_clothBack_inner_03_r {parent=dyn_clothBack_inner_02_r, loc=<22.7192 0 0>, rot=<0.0133 0.0017 0.0002 -0.9999>}
 dyn_clothBack_inner_04_r {parent=dyn_clothBack_inner_03_r, loc=<22.0583 0 0>, rot=<-0.0091 -0.0043 0.0003 -0.9999>}
 dyn_clothBack_inner_05_r {parent=dyn_clothBack_inner_04_r, loc=<20.7254 0 0>, rot=<-0.0134 0.0008 -0.0011 -0.9999>}
 dyn_clothBack_inner_06_r {parent=dyn_clothBack_inner_05_r, loc=<20.658 0 0>, rot=<0 0 0 1>}
 dyn_clothBack_mid_01_r {parent=capeBase, loc=<-13.4275 1.3098 0.1559>, rot=<-0.5223 -0.4376 0.5884 -0.4354>}
 dyn_clothBack_mid_02_r {parent=dyn_clothBack_mid_01_r, loc=<17.3943 0 0>, rot=<-0.0197 -0.0184 0.0034 -0.9996>}
 dyn_clothBack_mid_03_r {parent=dyn_clothBack_mid_02_r, loc=<21.4882 0 0>, rot=<0.0436 0.0041 0.0003 -0.999>}
 dyn_clothBack_mid_04_r {parent=dyn_clothBack_mid_03_r, loc=<20.446 0 0>, rot=<-0.0309 0.0013 0.0002 -0.9995>}
 dyn_clothBack_mid_05_r {parent=dyn_clothBack_mid_04_r, loc=<20.5943 0 0>, rot=<0.0172 -0.0026 -0.0002 -0.9998>}
 dyn_clothBack_mid_06_r {parent=dyn_clothBack_mid_05_r, loc=<20.8999 0 0>, rot=<0 0 0 -1>}
 thigh_l {parent=pelvis, loc=<-2.5693 -0.0109 -7.7501>, rot=<0.0012 0.0572 0.0155 -0.9982>}
 calf_l {parent=thigh_l, loc=<-39.0484 0 0>, rot=<0.0006 -0.0161 -0.0391 -0.9991>}
 calf_twist_01_l {parent=calf_l, loc=<-35.813 -0.0967 0.0397>, rot=<0.0237 0.0095 -0.0035 -0.9997>}
 calf_twist_02_l {parent=calf_l, loc=<-18.8064 -0.0967 0.0397>, rot=<0.0237 0.0095 -0.0035 -0.9997>}
 foot_l {parent=calf_l, loc=<-42.561 0.1077 0.4292>, rot=<0.0418 -0.0455 0.0343 -0.9975>}
 ball_l {parent=foot_l, loc=<-6.1599 13.8264 0.6806>, rot=<0.0001 -0.0001 -0.7063 -0.7079>}
 thigh_twist_01_l {parent=thigh_l, loc=<-21.6303 0 0>, rot=<0.0013 0 0 -1>}
 thigh_r {parent=pelvis, loc=<-2.5691 -0.0109 7.7501>, rot=<0.0572 -0.0012 0.9982 0.0155>}
 calf_r {parent=thigh_r, loc=<39.0487 0 0>, rot=<0.0006 -0.0161 -0.0391 -0.9991>}
 calf_twist_01_r {parent=calf_r, loc=<35.8132 0.0967 -0.0397>, rot=<0.0237 0.0095 -0.0035 -0.9997>}
 calf_twist_02_r {parent=calf_r, loc=<18.8064 0.0967 -0.0397>, rot=<0.0237 0.0095 -0.0035 -0.9997>}
 foot_r {parent=calf_r, loc=<42.5612 -0.1077 -0.4292>, rot=<0.0418 -0.0455 0.0343 -0.9975>}
 ball_r {parent=foot_r, loc=<6.1598 -13.8264 -0.6806>, rot=<0.0001 -0.0001 -0.7063 -0.7079>}
 thigh_twist_01_r {parent=thigh_r, loc=<21.6303 0 0>, rot=<0.0013 0 0 -1>}
 dyn_pouch_mstr {parent=pelvis, loc=<6.8384 9.7216 -6.3453>, rot=<-0.3237 -0.6287 -0.301 -0.6399>}
 dyn_pouch {parent=dyn_pouch_mstr, loc=<0 0 0>, rot=<0 0 0 -1>}
 dyn_clothSkirt_outer_mstr_l {parent=pelvis, loc=<-4.9172 -6.9417 -15.3641>, rot=<0.0113 -0.9581 0.2779 -0.0686>}
 dyn_clothSkirt_outer_01_l {parent=dyn_clothSkirt_outer_mstr_l, loc=<0 0 0>, rot=<0 0 0 1>}
 dyn_clothSkirt_outer_02_l {parent=dyn_clothSkirt_outer_01_l, loc=<15.7732 0 0>, rot=<-0.0188 -0.0266 0.0002 -0.9995>}
 dyn_clothSkirt_outer_03_l {parent=dyn_clothSkirt_outer_02_l, loc=<19.4909 0 0>, rot=<0.0068 -0.0134 -0.0052 -0.9999>}
 dyn_clothSkirt_outer_04_l {parent=dyn_clothSkirt_outer_03_l, loc=<20.9874 0 0>, rot=<0 0 0 -1>}
 dyn_clothSkirt_inner_mstr_l {parent=pelvis, loc=<-4.9717 -12.9253 -5.6444>, rot=<0.0192 -0.7552 0.6538 -0.043>}
 dyn_clothSkirt_inner_01_l {parent=dyn_clothSkirt_inner_mstr_l, loc=<0 0 0>, rot=<0 0 0 1>}
 dyn_clothSkirt_inner_02_l {parent=dyn_clothSkirt_inner_01_l, loc=<15.8777 0 0>, rot=<-0.0213 -0.0216 0.0023 -0.9995>}
 dyn_clothSkirt_inner_03_l {parent=dyn_clothSkirt_inner_02_l, loc=<19.0605 0 0>, rot=<-0.0194 -0.0144 -0.0027 -0.9997>}
 dyn_clothSkirt_inner_04_l {parent=dyn_clothSkirt_inner_03_l, loc=<21.8888 0 0>, rot=<0.0436 0 0 -0.999>}
 dyn_clothSkirt_inner_mstr_r {parent=pelvis, loc=<-4.9742 -12.9193 5.6497>, rot=<0.043 -0.6551 0.754 -0.0194>}
 dyn_clothSkirt_inner_01_r {parent=dyn_clothSkirt_inner_mstr_r, loc=<0 0 0>, rot=<0 0 0 -1>}
 dyn_clothSkirt_inner_02_r {parent=dyn_clothSkirt_inner_01_r, loc=<15.8756 0 0>, rot=<-0.0241 -0.0218 -0.0012 -0.9995>}
 dyn_clothSkirt_inner_03_r {parent=dyn_clothSkirt_inner_02_r, loc=<19.0605 0 0>, rot=<0.0194 -0.0141 0.004 -0.9997>}
 dyn_clothSkirt_inner_04_r {parent=dyn_clothSkirt_inner_03_r, loc=<21.8888 0 0>, rot=<0 0 0 1>}
 dyn_clothSkirt_outer_mstr_r {parent=pelvis, loc=<-4.9117 -6.9413 15.3696>, rot=<0.0685 -0.2786 0.9579 -0.0113>}
 dyn_clothSkirt_outer_01_r {parent=dyn_clothSkirt_outer_mstr_r, loc=<0 0 0>, rot=<0 0 0 -1>}
 dyn_clothSkirt_outer_02_r {parent=dyn_clothSkirt_outer_01_r, loc=<15.778 0 0>, rot=<0.018 -0.0264 -0.0001 -0.9995>}
 dyn_clothSkirt_outer_03_r {parent=dyn_clothSkirt_outer_02_r, loc=<19.4909 0 0>, rot=<-0.0068 -0.0134 0.0052 -0.9999>}
 dyn_clothSkirt_outer_04_r {parent=dyn_clothSkirt_outer_03_r, loc=<20.9874 0 0>, rot=<0 0 0 1>}
]
