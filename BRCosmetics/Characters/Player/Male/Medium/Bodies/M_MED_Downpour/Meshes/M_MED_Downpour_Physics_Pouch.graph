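# Physics asset for the pouch on the M_MED_Downpour male character body, apparently for secondary dangle motion.
# It chains three capsule bodies: pelvis, a small dyn_pouch_mstr link, and an offset capsule on dyn_pouch sized for the pouch.
# The master link rotates freely relative to the pelvis.
# The pouch constraint to the master has a wider limit on one swing axis and very tight limits on the other swing and twist.
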PhysicsAsset /BRCosmetics/Characters/Player/Male/Medium/Bodies/M_MED_Downpour/Meshes/M_MED_Downpour_Physics_Pouch
bodies {
 dyn_pouch [{shapeType=Capsule, length=4.8256, radius=3.6047, center=<2.9873 1.0097 -2.2487>}]
 dyn_pouch_mstr [{shapeType=Capsule, length=1, radius=1}]
 pelvis [{shapeType=Capsule, length=1, radius=1}]
}
constraints [
 {
  bone1Name=dyn_pouch_mstr
  bone2Name=pelvis
  swing1LimitDegrees=45
  swing1Motion=Free
  swing2LimitDegrees=45
  swing2Motion=Free
  twistLimitDegrees=45
  twistMotion=Free
 }
 {bone1Name=dyn_pouch, bone2Name=dyn_pouch_mstr, swing1LimitDegrees=25, swing2LimitDegrees=2, twistLimitDegrees=2}
]
