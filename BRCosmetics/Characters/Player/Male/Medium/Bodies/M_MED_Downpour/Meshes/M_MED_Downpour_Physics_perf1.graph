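# Physics asset for the M_MED_Downpour player body (perf1 variant), built for secondary motion of its cape, front cloth and skirt.
# Capsules cover spine, pelvis, legs and upper arms, linked by angularly free constraints.
# Cloth uses capsule bone chains with swing and twist limits, tightest where inner and mid back chains join capeBase.
# Tiny master bodies anchor front and outer-back chains to the clavicles and skirt chains to the pelvis.
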PhysicsAsset /BRCosmetics/Characters/Player/Male/Medium/Bodies/M_MED_Downpour/Meshes/M_MED_Downpour_Physics_perf1
bodies {
 calf_l [{shapeType=Capsule, length=26, radius=8, center=<-18 0 0>, rotation=<rot 90 0 0>}]
 calf_r [{shapeType=Capsule, length=26, radius=8, center=<18 0 0>, rotation=<rot 90 0 0>}]
 capeBase [{shapeType=Capsule, length=1, radius=1}]
 clavicle_l [{shapeType=Capsule, length=1, radius=1}]
 clavicle_r [{shapeType=Capsule, length=1, radius=1}]
 dyn_clothBack_inner_01_l [{shapeType=Capsule, length=12, radius=5, center=<6 0 5>, rotation=<rot 90 0 0>}]
 dyn_clothBack_inner_01_r [{shapeType=Capsule, length=12, radius=5, center=<6 0 5>, rotation=<rot 90 0 0>}]
 dyn_clothBack_inner_02_l [{shapeType=Capsule, length=12, radius=5, center=<6 0 5>, rotation=<rot 90 0 0>}]
 dyn_clothBack_inner_02_r [{shapeType=Capsule, length=12, radius=5, center=<6 0 5>, rotation=<rot 90 0 0>}]
 dyn_clothBack_inner_03_l [{shapeType=Capsule, length=12, radius=5, center=<6 0 4>, rotation=<rot 90 0 0>}]
 dyn_clothBack_inner_03_r [{shapeType=Capsule, length=12, radius=5, center=<6 0 4>, rotation=<rot 90 0 0>}]
 dyn_clothBack_inner_04_l [{shapeType=Capsule, length=12, radius=5, center=<6 0 4>, rotation=<rot 90 0 0>}]
 dyn_clothBack_inner_04_r [{shapeType=Capsule, length=12, radius=5, center=<6 0 4>, rotation=<rot 90 0 0>}]
 dyn_clothBack_inner_05_l [{shapeType=Capsule, length=12, radius=5, center=<6 0 4>, rotation=<rot 90 0 0>}]
 dyn_clothBack_inner_05_r [{shapeType=Capsule, length=12, radius=5, center=<6 0 4>, rotation=<rot 90 0 0>}]
 dyn_clothBack_inner_06_l [{shapeType=Capsule, length=12, radius=5, center=<6 0 4>, rotation=<rot 90 0 0>}]
 dyn_clothBack_inner_06_r [{shapeType=Capsule, length=12, radius=5, center=<6 0 4>, rotation=<rot 90 0 0>}]
 dyn_clothBack_mid_01_l [{shapeType=Capsule, length=12, radius=5, center=<6 0 5>, rotation=<rot 90 0 0>}]
 dyn_clothBack_mid_01_r [{shapeType=Capsule, length=12, radius=5, center=<6 0 5>, rotation=<rot 90 0 0>}]
 dyn_clothBack_mid_02_l [{shapeType=Capsule, length=12, radius=5, center=<6 0 5>, rotation=<rot 90 0 0>}]
 dyn_clothBack_mid_02_r [{shapeType=Capsule, length=12, radius=5, center=<6 0 5>, rotation=<rot 90 0 0>}]
 dyn_clothBack_mid_03_l [{shapeType=Capsule, length=12, radius=5, center=<6 0 4>, rotation=<rot 90 0 0>}]
 dyn_clothBack_mid_03_r [{shapeType=Capsule, length=12, radius=5, center=<6 0 4>, rotation=<rot 90 0 0>}]
 dyn_clothBack_mid_04_l [{shapeType=Capsule, length=12, radius=5, center=<6 0 4>, rotation=<rot 90 0 0>}]
 dyn_clothBack_mid_04_r [{shapeType=Capsule, length=12, radius=5, center=<6 0 4>, rotation=<rot 90 0 0>}]
 dyn_clothBack_mid_05_l [{shapeType=Capsule, length=12, radius=5, center=<6 0 4>, rotation=<rot 90 0 0>}]
 dyn_clothBack_mid_05_r [{shapeType=Capsule, length=12, radius=5, center=<6 0 4>, rotation=<rot 90 0 0>}]
 dyn_clothBack_mid_06_l [{shapeType=Capsule, length=12, radius=5, center=<6 0 4>, rotation=<rot 90 0 0>}]
 dyn_clothBack_mid_06_r [{shapeType=Capsule, length=12, radius=5, center=<6 0 4>, rotation=<rot 90 0 0>}]
 dyn_clothBack_outer_01_l [{shapeType=Capsule, length=12, radius=4, center=<6 0 3>, rotation=<rot 90 0 0>}]
 dyn_clothBack_outer_01_r [{shapeType=Capsule, length=12, radius=4.25, center=<6 -0.5 3.5>, rotation=<rot 90 0 0>}]
 dyn_clothBack_outer_02_l [{shapeType=Capsule, length=12, radius=5, shapeName=dyn_clothBack_outer_02_l_capsule, center=<5 0 3>, rotation=<rot 90 0 0>}]
 dyn_clothBack_outer_02_r [{shapeType=Capsule, length=12, radius=5, shapeName=dyn_clothBack_outer_02_l_capsule, center=<5 0 3>, rotation=<rot 90 0 0>}]
 dyn_clothBack_outer_03_l [{shapeType=Capsule, length=12, radius=5, shapeName=dyn_clothBack_outer_03_l_capsule, center=<5 0 3>, rotation=<rot 90 0 0>}]
 dyn_clothBack_outer_03_r [{shapeType=Capsule, length=12, radius=5, shapeName=dyn_clothBack_outer_03_l_capsule, center=<5 0 3>, rotation=<rot 90 0 0>}]
 dyn_clothBack_outer_04_l [{shapeType=Capsule, length=12, radius=5, shapeName=dyn_clothBack_outer_04_l_capsule, center=<5 0 3>, rotation=<rot 90 0 0>}]
 dyn_clothBack_outer_04_r [{shapeType=Capsule, length=12, radius=5, shapeName=dyn_clothBack_outer_04_l_capsule, center=<5 0 3>, rotation=<rot 90 0 0>}]
 dyn_clothBack_outer_05_l [{shapeType=Capsule, length=12, radius=5, shapeName=dyn_clothBack_outer_05_l_capsule, center=<5 0 3>, rotation=<rot 90 0 0>}]
 dyn_clothBack_outer_05_r [{shapeType=Capsule, length=12, radius=5, shapeName=dyn_clothBack_outer_05_l_capsule, center=<5 0 3>, rotation=<rot 90 0 0>}]
 dyn_clothBack_outer_06_l [{shapeType=Capsule, length=12, radius=5, shapeName=dyn_clothBack_outer_06_l_capsule, center=<5 0 3>, rotation=<rot 90 0 0>}]
 dyn_clothBack_outer_06_r [{shapeType=Capsule, length=12, radius=5, shapeName=dyn_clothBack_outer_06_l_capsule, center=<5 0 3>, rotation=<rot 90 0 0>}]
 dyn_clothBack_outer_mstr_l [{shapeType=Capsule, length=0.505, radius=0.505, rotation=<rot 0 -90 0>}]
 dyn_clothBack_outer_mstr_r [{shapeType=Capsule, length=0.505, radius=0.505, rotation=<rot 0 -90 0>}]
 dyn_clothFront_inner_01_l [{shapeType=Capsule, length=12, radius=4, center=<6 0 3>, rotation=<rot 90 0 0>}]
 dyn_clothFront_inner_01_r [{shapeType=Capsule, length=12, radius=4, center=<6 0 3>, rotation=<rot 90 0 0>}]
 dyn_clothFront_inner_02_l [{shapeType=Capsule, length=12, radius=5, shapeName=dyn_clothBack_outer_02_l_capsule, center=<5 0 3>, rotation=<rot 90 0 0>}]
 dyn_clothFront_inner_02_r [{shapeType=Capsule, length=12, radius=5, shapeName=dyn_clothBack_outer_02_l_capsule, center=<5 0 3>, rotation=<rot 90 0 0>}]
 dyn_clothFront_inner_03_l [{shapeType=Capsule, length=12, radius=5, shapeName=dyn_clothBack_outer_02_l_capsule, center=<5 0 3>, rotation=<rot 90 0 0>}]
 dyn_clothFront_inner_03_r [{shapeType=Capsule, length=12, radius=5, shapeName=dyn_clothBack_outer_02_l_capsule, center=<5 0 3>, rotation=<rot 90 0 0>}]
 dyn_clothFront_inner_04_l [{shapeType=Capsule, length=12, radius=5, shapeName=dyn_clothBack_outer_02_l_capsule, center=<5 0 3>, rotation=<rot 90 0 0>}]
 dyn_clothFront_inner_04_r [{shapeType=Capsule, length=12, radius=5, shapeName=dyn_clothBack_outer_02_l_capsule, center=<5 0 3>, rotation=<rot 90 0 0>}]
 dyn_clothFront_inner_05_l [{shapeType=Capsule, length=12, radius=5, shapeName=dyn_clothBack_outer_02_l_capsule, center=<5 0 3>, rotation=<rot 90 0 0>}]
 dyn_clothFront_inner_05_r [{shapeType=Capsule, length=12, radius=5, shapeName=dyn_clothBack_outer_02_l_capsule, center=<5 0 3>, rotation=<rot 90 0 0>}]
 dyn_clothFront_inner_mstr_l [{shapeType=Capsule, length=0.505, radius=0.505, rotation=<rot 0 -90 0>}]
 dyn_clothFront_inner_mstr_r [{shapeType=Capsule, length=0.505, radius=0.505, rotation=<rot 0 -90 0>}]
 dyn_clothFront_outer_01_l [{shapeType=Capsule, length=12, radius=4, center=<6 0 3>, rotation=<rot 90 0 0>}]
 dyn_clothFront_outer_01_r [{shapeType=Capsule, length=12, radius=4, center=<6 0 3>, rotation=<rot 90 0 0>}]
 dyn_clothFront_outer_02_l [{shapeType=Capsule, length=12, radius=4, center=<6 0 3>, rotation=<rot 90 0 0>}]
 dyn_clothFront_outer_02_r [{shapeType=Capsule, length=12, radius=4, center=<6 0 3>, rotation=<rot 90 0 0>}]
 dyn_clothFront_outer_03_l [{shapeType=Capsule, length=12, radius=4, center=<6 0 3>, rotation=<rot 90 0 0>}]
 dyn_clothFront_outer_03_r [{shapeType=Capsule, length=12, radius=4, center=<6 0 3>, rotation=<rot 90 0 0>}]
 dyn_clothFront_outer_04_l [{shapeType=Capsule, length=12, radius=4, center=<6 0 3>, rotation=<rot 90 0 0>}]
 dyn_clothFront_outer_04_r [{shapeType=Capsule, length=12, radius=4, center=<6 0 3>, rotation=<rot 90 0 0>}]
 dyn_clothFront_outer_05_l [{shapeType=Capsule, length=12, radius=4, center=<6 0 3>, rotation=<rot 90 0 0>}]
 dyn_clothFront_outer_05_r [{shapeType=Capsule, length=12, radius=4, center=<6 0 3>, rotation=<rot 90 0 0>}]
 dyn_clothFront_outer_06_l [{shapeType=Capsule, length=12, radius=4, center=<6 0 3>, rotation=<rot 90 0 0>}]
 dyn_clothFront_outer_06_r [{shapeType=Capsule, length=12, radius=4, center=<6 0 3>, rotation=<rot 90 0 0>}]
 dyn_clothFront_outer_mstr_l [{shapeType=Capsule, length=0.505, radius=0.505, rotation=<rot 0 -90 0>}]
 dyn_clothFront_outer_mstr_r [{shapeType=Capsule, length=0.505, radius=0.505, rotation=<rot 0 -90 0>}]
 dyn_clothSkirt_inner_01_l [{shapeType=Capsule, length=7, radius=5, center=<9 0 4>, rotation=<rot 90 0 0>}]
 dyn_clothSkirt_inner_01_r [{shapeType=Capsule, length=7, radius=5, center=<9 0 4>, rotation=<rot 90 0 0>}]
 dyn_clothSkirt_inner_02_l [{shapeType=Capsule, length=8, radius=6, center=<11 0 5>, rotation=<rot 90 0 0>}]
 dyn_clothSkirt_inner_02_r [{shapeType=Capsule, length=8, radius=6, center=<11 0 5>, rotation=<rot 90 0 0>}]
 dyn_clothSkirt_inner_03_l [{shapeType=Capsule, length=10, radius=6, center=<12 0 6>, rotation=<rot 90 0 0>}]
 dyn_clothSkirt_inner_03_r [{shapeType=Capsule, length=10, radius=6, center=<12 0 6>, rotation=<rot 90 0 0>}]
 dyn_clothSkirt_inner_04_l [{shapeType=Capsule, length=8, radius=6, center=<12 0 6>, rotation=<rot 90 0 0>}]
 dyn_clothSkirt_inner_04_r [{shapeType=Capsule, length=8, radius=6, center=<12 0 6>, rotation=<rot 90 0 0>}]
 dyn_clothSkirt_inner_mstr_l [{shapeType=Capsule, length=0.505, radius=0.505}]
 dyn_clothSkirt_inner_mstr_r [{shapeType=Capsule, length=0.505, radius=0.505}]
 dyn_clothSkirt_outer_01_l [{shapeType=Capsule, length=7, radius=5, center=<9 0 4>, rotation=<rot 90 0 0>}]
 dyn_clothSkirt_outer_01_r [{shapeType=Capsule, length=7, radius=5, center=<9 0 4>, rotation=<rot 90 0 0>}]
 dyn_clothSkirt_outer_02_l [{shapeType=Capsule, length=8, radius=6, center=<11 0 5>, rotation=<rot 90 0 0>}]
 dyn_clothSkirt_outer_02_r [{shapeType=Capsule, length=8, radius=6, center=<11 0 5>, rotation=<rot 90 0 0>}]
 dyn_clothSkirt_outer_03_l [{shapeType=Capsule, length=10, radius=6, center=<12 0 6>, rotation=<rot 90 0 0>}]
 dyn_clothSkirt_outer_03_r [{shapeType=Capsule, length=10, radius=6, center=<12 0 6>, rotation=<rot 90 0 0>}]
 dyn_clothSkirt_outer_04_l [{shapeType=Capsule, length=8, radius=6, center=<12 0 6>, rotation=<rot 90 0 0>}]
 dyn_clothSkirt_outer_04_r [{shapeType=Capsule, length=8, radius=6, center=<12 0 6>, rotation=<rot 90 0 0>}]
 dyn_clothSkirt_outer_mstr_l [{shapeType=Capsule, length=0.505, radius=0.505}]
 dyn_clothSkirt_outer_mstr_r [{shapeType=Capsule, length=0.505, radius=0.505}]
 pelvis [{shapeType=Capsule, length=11.5, radius=12.5, center=<0 -0.4265 0>}]
 spine_01 [{shapeType=Capsule, length=14, radius=12}]
 spine_02 [{shapeType=Capsule, length=12, radius=12}]
 spine_03 [{shapeType=Capsule, length=12, radius=12}]
 spine_04 [{shapeType=Capsule, length=12, radius=12}]
 spine_05 [{shapeType=Capsule, length=15, radius=11}]
 thigh_l [{shapeType=Capsule, length=25.125, radius=8.875, center=<-20 0 0>, rotation=<rot 85 0 0>}]
 thigh_r [
  {shapeType=Capsule, length=25.125, radius=8.875, center=<20 0 0>, rotation=<rot 85 0 0>}
  {shapeType=Capsule, length=30, radius=6, center=<16.9465 0 5>, rotation=<rot 85 0 0>}
 ]
 upperarm_l [{shapeType=Capsule, length=20, radius=7, center=<13 0 0>, rotation=<rot 90 0 0>}]
 upperarm_r [{shapeType=Capsule, length=20, radius=7, center=<-13 0 0>, rotation=<rot 90 0 180>}]
}
constraints [
 {
  bone1Name=capeBase
  bone2Name=spine_05
  swing1LimitDegrees=45
  swing1Motion=Free
  swing2LimitDegrees=45
  swing2Motion=Free
  twistLimitDegrees=45
  twistMotion=Free
 }
 {
  bone1Name=spine_01
  bone2Name=pelvis
  swing1LimitDegrees=45
  swing1Motion=Free
  swing2LimitDegrees=45
  swing2Motion=Free
  twistLimitDegrees=45
  twistMotion=Free
 }
 {
  bone1Name=spine_02
  bone2Name=spine_01
  swing1LimitDegrees=45
  swing1Motion=Free
  swing2LimitDegrees=45
  swing2Motion=Free
  twistLimitDegrees=45
  twistMotion=Free
 }
 {
  bone1Name=spine_03
  bone2Name=spine_02
  swing1LimitDegrees=45
  swing1Motion=Free
  swing2LimitDegrees=45
  swing2Motion=Free
  twistLimitDegrees=45
  twistMotion=Free
 }
 {
  bone1Name=spine_04
  bone2Name=spine_03
  swing1LimitDegrees=45
  swing1Motion=Free
  swing2LimitDegrees=45
  swing2Motion=Free
  twistLimitDegrees=45
  twistMotion=Free
 }
 {
  bone1Name=spine_05
  bone2Name=spine_04
  swing1LimitDegrees=45
  swing1Motion=Free
  swing2LimitDegrees=45
  swing2Motion=Free
  twistLimitDegrees=45
  twistMotion=Free
 }
 {
  bone1Name=dyn_clothSkirt_outer_mstr_l
  bone2Name=pelvis
  swing1LimitDegrees=45
  swing1Motion=Free
  swing2LimitDegrees=45
  swing2Motion=Free
  twistLimitDegrees=45
  twistMotion=Free
 }
 {
  bone1Name=dyn_clothSkirt_outer_01_l
  bone2Name=dyn_clothSkirt_outer_mstr_l
  swing1LimitDegrees=25
  swing2LimitDegrees=25
  twistLimitDegrees=5
 }
 {bone1Name=dyn_clothSkirt_outer_02_l, bone2Name=dyn_clothSkirt_outer_01_l, swing1LimitDegrees=25, swing2LimitDegrees=25, twistLimitDegrees=5}
 {bone1Name=dyn_clothSkirt_outer_03_l, bone2Name=dyn_clothSkirt_outer_02_l, swing1LimitDegrees=25, swing2LimitDegrees=25, twistLimitDegrees=5}
 {bone1Name=dyn_clothSkirt_outer_04_l, bone2Name=dyn_clothSkirt_outer_03_l, swing1LimitDegrees=25, swing2LimitDegrees=25, twistLimitDegrees=5}
 {
  bone1Name=dyn_clothSkirt_inner_mstr_l
  bone2Name=pelvis
  swing1LimitDegrees=45
  swing1Motion=Free
  swing2LimitDegrees=45
  swing2Motion=Free
  twistLimitDegrees=45
  twistMotion=Free
 }
 {
  bone1Name=dyn_clothSkirt_inner_01_l
  bone2Name=dyn_clothSkirt_inner_mstr_l
  swing1LimitDegrees=25
  swing2LimitDegrees=25
  twistLimitDegrees=5
 }
 {bone1Name=dyn_clothSkirt_inner_02_l, bone2Name=dyn_clothSkirt_inner_01_l, swing1LimitDegrees=25, swing2LimitDegrees=25, twistLimitDegrees=5}
 {bone1Name=dyn_clothSkirt_inner_03_l, bone2Name=dyn_clothSkirt_inner_02_l, swing1LimitDegrees=25, swing2LimitDegrees=25, twistLimitDegrees=5}
 {bone1Name=dyn_clothSkirt_inner_04_l, bone2Name=dyn_clothSkirt_inner_03_l, swing1LimitDegrees=25, swing2LimitDegrees=25, twistLimitDegrees=5}
 {
  bone1Name=dyn_clothSkirt_inner_mstr_r
  bone2Name=pelvis
  swing1LimitDegrees=45
  swing1Motion=Free
  swing2LimitDegrees=45
  swing2Motion=Free
  twistLimitDegrees=45
  twistMotion=Free
 }
 {
  bone1Name=dyn_clothSkirt_inner_01_r
  bone2Name=dyn_clothSkirt_inner_mstr_r
  swing1LimitDegrees=25
  swing2LimitDegrees=25
  twistLimitDegrees=5
 }
 {bone1Name=dyn_clothSkirt_inner_02_r, bone2Name=dyn_clothSkirt_inner_01_r, swing1LimitDegrees=25, swing2LimitDegrees=25, twistLimitDegrees=5}
 {bone1Name=dyn_clothSkirt_inner_03_r, bone2Name=dyn_clothSkirt_inner_02_r, swing1LimitDegrees=25, swing2LimitDegrees=25, twistLimitDegrees=5}
 {bone1Name=dyn_clothSkirt_inner_04_r, bone2Name=dyn_clothSkirt_inner_03_r, swing1LimitDegrees=25, swing2LimitDegrees=25, twistLimitDegrees=5}
 {
  bone1Name=dyn_clothSkirt_outer_mstr_r
  bone2Name=pelvis
  swing1LimitDegrees=45
  swing1Motion=Free
  swing2LimitDegrees=45
  swing2Motion=Free
  twistLimitDegrees=45
  twistMotion=Free
 }
 {
  bone1Name=dyn_clothSkirt_outer_01_r
  bone2Name=dyn_clothSkirt_outer_mstr_r
  swing1LimitDegrees=10
  swing2LimitDegrees=10
  twistLimitDegrees=5
 }
 {bone1Name=dyn_clothSkirt_outer_02_r, bone2Name=dyn_clothSkirt_outer_01_r, swing1LimitDegrees=20, swing2LimitDegrees=25, twistLimitDegrees=5}
 {bone1Name=dyn_clothSkirt_outer_03_r, bone2Name=dyn_clothSkirt_outer_02_r, swing1LimitDegrees=25, swing2LimitDegrees=25, twistLimitDegrees=5}
 {bone1Name=dyn_clothSkirt_outer_04_r, bone2Name=dyn_clothSkirt_outer_03_r, swing1LimitDegrees=25, swing2LimitDegrees=25, twistLimitDegrees=5}
 {
  bone1Name=thigh_l
  bone2Name=pelvis
  swing1LimitDegrees=45
  swing1Motion=Free
  swing2LimitDegrees=45
  swing2Motion=Free
  twistLimitDegrees=45
  twistMotion=Free
 }
 {
  bone1Name=calf_l
  bone2Name=thigh_l
  swing1LimitDegrees=45
  swing1Motion=Free
  swing2LimitDegrees=45
  swing2Motion=Free
  twistLimitDegrees=45
  twistMotion=Free
 }
 {
  bone1Name=calf_r
  bone2Name=thigh_r
  swing1LimitDegrees=45
  swing1Motion=Free
  swing2LimitDegrees=45
  swing2Motion=Free
  twistLimitDegrees=45
  twistMotion=Free
 }
 {
  bone1Name=thigh_r
  bone2Name=pelvis
  swing1LimitDegrees=45
  swing1Motion=Free
  swing2LimitDegrees=45
  swing2Motion=Free
  twistLimitDegrees=45
  twistMotion=Free
 }
 {
  bone1Name=clavicle_l
  bone2Name=spine_05
  swing1LimitDegrees=45
  swing1Motion=Free
  swing2LimitDegrees=45
  swing2Motion=Free
  twistLimitDegrees=45
  twistMotion=Free
 }
 {
  bone1Name=clavicle_r
  bone2Name=spine_05
  swing1LimitDegrees=45
  swing1Motion=Free
  swing2LimitDegrees=45
  swing2Motion=Free
  twistLimitDegrees=45
  twistMotion=Free
 }
 {
  bone1Name=dyn_clothFront_inner_01_l
  bone2Name=dyn_clothFront_inner_mstr_l
  swing1LimitDegrees=25
  swing2LimitDegrees=25
  twistLimitDegrees=5
 }
 {
  bone1Name=dyn_clothFront_inner_01_r
  bone2Name=dyn_clothFront_inner_mstr_r
  swing1LimitDegrees=30
  swing2LimitDegrees=30
  twistLimitDegrees=5
 }
 {bone1Name=dyn_clothBack_mid_01_l, bone2Name=capeBase, swing1LimitDegrees=5, swing2LimitDegrees=5, twistLimitDegrees=2}
 {bone1Name=dyn_clothBack_mid_02_l, bone2Name=dyn_clothBack_mid_01_l, swing1LimitDegrees=25, swing2LimitDegrees=25, twistLimitDegrees=5}
 {bone1Name=dyn_clothBack_mid_03_l, bone2Name=dyn_clothBack_mid_02_l, swing1LimitDegrees=25, swing2LimitDegrees=25, twistLimitDegrees=5}
 {bone1Name=dyn_clothBack_mid_04_l, bone2Name=dyn_clothBack_mid_03_l, swing1LimitDegrees=25, swing2LimitDegrees=25, twistLimitDegrees=5}
 {bone1Name=dyn_clothBack_mid_05_l, bone2Name=dyn_clothBack_mid_04_l, swing1LimitDegrees=25, swing2LimitDegrees=25, twistLimitDegrees=5}
 {bone1Name=dyn_clothBack_mid_06_l, bone2Name=dyn_clothBack_mid_05_l, swing1LimitDegrees=25, swing2LimitDegrees=25, twistLimitDegrees=5}
 {bone1Name=dyn_clothBack_inner_01_l, bone2Name=capeBase, swing1LimitDegrees=5, swing2LimitDegrees=5, twistLimitDegrees=2}
 {bone1Name=dyn_clothBack_inner_02_l, bone2Name=dyn_clothBack_inner_01_l, swing1LimitDegrees=25, swing2LimitDegrees=25, twistLimitDegrees=5}
 {bone1Name=dyn_clothBack_inner_03_l, bone2Name=dyn_clothBack_inner_02_l, swing1LimitDegrees=25, swing2LimitDegrees=25, twistLimitDegrees=5}
 {bone1Name=dyn_clothBack_inner_04_l, bone2Name=dyn_clothBack_inner_03_l, swing1LimitDegrees=25, swing2LimitDegrees=25, twistLimitDegrees=5}
 {bone1Name=dyn_clothBack_inner_05_l, bone2Name=dyn_clothBack_inner_04_l, swing1LimitDegrees=25, swing2LimitDegrees=25, twistLimitDegrees=5}
 {bone1Name=dyn_clothBack_inner_06_l, bone2Name=dyn_clothBack_inner_05_l, swing1LimitDegrees=25, swing2LimitDegrees=25, twistLimitDegrees=5}
 {bone1Name=dyn_clothBack_inner_01_r, bone2Name=capeBase, swing1LimitDegrees=5, swing2LimitDegrees=5, twistLimitDegrees=2}
 {bone1Name=dyn_clothBack_inner_02_r, bone2Name=dyn_clothBack_inner_01_r, swing1LimitDegrees=25, swing2LimitDegrees=25, twistLimitDegrees=5}
 {bone1Name=dyn_clothBack_inner_03_r, bone2Name=dyn_clothBack_inner_02_r, swing1LimitDegrees=25, swing2LimitDegrees=25, twistLimitDegrees=5}
 {bone1Name=dyn_clothBack_inner_04_r, bone2Name=dyn_clothBack_inner_03_r, swing1LimitDegrees=25, swing2LimitDegrees=25, twistLimitDegrees=5}
 {bone1Name=dyn_clothBack_inner_05_r, bone2Name=dyn_clothBack_inner_04_r, swing1LimitDegrees=25, swing2LimitDegrees=25, twistLimitDegrees=5}
 {bone1Name=dyn_clothBack_inner_06_r, bone2Name=dyn_clothBack_inner_05_r, swing1LimitDegrees=25, swing2LimitDegrees=25, twistLimitDegrees=5}
 {bone1Name=dyn_clothBack_mid_01_r, bone2Name=capeBase, swing1LimitDegrees=5, swing2LimitDegrees=5, twistLimitDegrees=2}
 {bone1Name=dyn_clothBack_mid_02_r, bone2Name=dyn_clothBack_mid_01_r, swing1LimitDegrees=25, swing2LimitDegrees=25, twistLimitDegrees=5}
 {bone1Name=dyn_clothBack_mid_03_r, bone2Name=dyn_clothBack_mid_02_r, swing1LimitDegrees=25, swing2LimitDegrees=25, twistLimitDegrees=5}
 {bone1Name=dyn_clothBack_mid_04_r, bone2Name=dyn_clothBack_mid_03_r, swing1LimitDegrees=25, swing2LimitDegrees=25, twistLimitDegrees=5}
 {bone1Name=dyn_clothBack_mid_05_r, bone2Name=dyn_clothBack_mid_04_r, swing1LimitDegrees=25, swing2LimitDegrees=25, twistLimitDegrees=5}
 {bone1Name=dyn_clothBack_mid_06_r, bone2Name=dyn_clothBack_mid_05_r, swing1LimitDegrees=25, swing2LimitDegrees=25, twistLimitDegrees=5}
 {
  bone1Name=dyn_clothFront_outer_01_l
  bone2Name=dyn_clothFront_outer_mstr_l
  swing1LimitDegrees=25
  swing2LimitDegrees=25
  twistLimitDegrees=5
 }
 {bone1Name=dyn_clothFront_outer_02_l, bone2Name=dyn_clothFront_outer_01_l, swing1LimitDegrees=25, swing2LimitDegrees=25, twistLimitDegrees=5}
 {bone1Name=dyn_clothFront_outer_03_l, bone2Name=dyn_clothFront_outer_02_l, swing1LimitDegrees=25, swing2LimitDegrees=25, twistLimitDegrees=5}
 {bone1Name=dyn_clothFront_outer_04_l, bone2Name=dyn_clothFront_outer_03_l, swing1LimitDegrees=25, swing2LimitDegrees=25, twistLimitDegrees=5}
 {bone1Name=dyn_clothFront_outer_05_l, bone2Name=dyn_clothFront_outer_04_l, swing1LimitDegrees=25, swing2LimitDegrees=25, twistLimitDegrees=5}
 {bone1Name=dyn_clothFront_outer_06_l, bone2Name=dyn_clothFront_outer_05_l, swing1LimitDegrees=25, swing2LimitDegrees=25, twistLimitDegrees=5}
 {
  bone1Name=dyn_clothFront_outer_01_r
  bone2Name=dyn_clothFront_outer_mstr_r
  swing1LimitDegrees=25
  swing2LimitDegrees=25
  twistLimitDegrees=5
 }
 {bone1Name=dyn_clothFront_outer_02_r, bone2Name=dyn_clothFront_outer_01_r, swing1LimitDegrees=25, swing2LimitDegrees=25, twistLimitDegrees=5}
 {bone1Name=dyn_clothFront_outer_03_r, bone2Name=dyn_clothFront_outer_02_r, swing1LimitDegrees=25, swing2LimitDegrees=25, twistLimitDegrees=5}
 {bone1Name=dyn_clothFront_outer_04_r, bone2Name=dyn_clothFront_outer_03_r, swing1LimitDegrees=25, swing2LimitDegrees=25, twistLimitDegrees=5}
 {bone1Name=dyn_clothFront_outer_05_r, bone2Name=dyn_clothFront_outer_04_r, swing1LimitDegrees=25, swing2LimitDegrees=25, twistLimitDegrees=5}
 {bone1Name=dyn_clothFront_outer_06_r, bone2Name=dyn_clothFront_outer_05_r, swing1LimitDegrees=25, swing2LimitDegrees=25, twistLimitDegrees=5}
 {
  bone1Name=dyn_clothFront_inner_mstr_r
  bone2Name=clavicle_r
  swing1LimitDegrees=45
  swing1Motion=Free
  swing2LimitDegrees=45
  swing2Motion=Free
  twistLimitDegrees=45
  twistMotion=Free
 }
 {
  bone1Name=dyn_clothFront_inner_mstr_l
  bone2Name=clavicle_l
  swing1LimitDegrees=45
  swing1Motion=Free
  swing2LimitDegrees=45
  swing2Motion=Free
  twistLimitDegrees=45
  twistMotion=Free
 }
 {
  bone1Name=dyn_clothFront_outer_mstr_l
  bone2Name=clavicle_l
  swing1LimitDegrees=45
  swing1Motion=Free
  swing2LimitDegrees=45
  swing2Motion=Free
  twistLimitDegrees=45
  twistMotion=Free
 }
 {
  bone1Name=dyn_clothFront_outer_mstr_r
  bone2Name=clavicle_r
  swing1LimitDegrees=45
  swing1Motion=Free
  swing2LimitDegrees=45
  swing2Motion=Free
  twistLimitDegrees=45
  twistMotion=Free
 }
 {
  bone1Name=dyn_clothBack_outer_mstr_r
  bone2Name=clavicle_r
  swing1LimitDegrees=45
  swing1Motion=Free
  swing2LimitDegrees=45
  swing2Motion=Free
  twistLimitDegrees=45
  twistMotion=Free
 }
 {
  bone1Name=dyn_clothBack_outer_mstr_l
  bone2Name=clavicle_l
  swing1LimitDegrees=45
  swing1Motion=Free
  swing2LimitDegrees=45
  swing2Motion=Free
  twistLimitDegrees=45
  twistMotion=Free
 }
 {
  bone1Name=upperarm_l
  bone2Name=clavicle_l
  swing1LimitDegrees=45
  swing1Motion=Free
  swing2LimitDegrees=45
  swing2Motion=Free
  twistLimitDegrees=45
  twistMotion=Free
 }
 {
  bone1Name=upperarm_r
  bone2Name=clavicle_r
  swing1LimitDegrees=45
  swing1Motion=Free
  swing2LimitDegrees=45
  swing2Motion=Free
  twistLimitDegrees=45
  twistMotion=Free
 }
 {bone1Name=dyn_clothBack_outer_01_l, bone2Name=dyn_clothBack_outer_mstr_l, swing1LimitDegrees=20, swing2LimitDegrees=20, twistLimitDegrees=4}
 {bone1Name=dyn_clothBack_outer_02_l, bone2Name=dyn_clothBack_outer_01_l, swing1LimitDegrees=25, swing2LimitDegrees=25, twistLimitDegrees=5}
 {bone1Name=dyn_clothBack_outer_03_l, bone2Name=dyn_clothBack_outer_02_l, swing1LimitDegrees=25, swing2LimitDegrees=25, twistLimitDegrees=5}
 {bone1Name=dyn_clothBack_outer_04_l, bone2Name=dyn_clothBack_outer_03_l, swing1LimitDegrees=25, swing2LimitDegrees=25, twistLimitDegrees=5}
 {bone1Name=dyn_clothBack_outer_05_l, bone2Name=dyn_clothBack_outer_04_l, swing1LimitDegrees=25, swing2LimitDegrees=25, twistLimitDegrees=5}
 {bone1Name=dyn_clothBack_outer_06_l, bone2Name=dyn_clothBack_outer_05_l, swing1LimitDegrees=25, swing2LimitDegrees=25, twistLimitDegrees=5}
 {bone1Name=dyn_clothBack_outer_01_r, bone2Name=dyn_clothBack_outer_mstr_r, swing1LimitDegrees=20, swing2LimitDegrees=20, twistLimitDegrees=4}
 {bone1Name=dyn_clothBack_outer_02_r, bone2Name=dyn_clothBack_outer_01_r, swing1LimitDegrees=25, swing2LimitDegrees=25, twistLimitDegrees=5}
 {bone1Name=dyn_clothBack_outer_03_r, bone2Name=dyn_clothBack_outer_02_r, swing1LimitDegrees=25, swing2LimitDegrees=25, twistLimitDegrees=5}
 {bone1Name=dyn_clothBack_outer_04_r, bone2Name=dyn_clothBack_outer_03_r, swing1LimitDegrees=25, swing2LimitDegrees=25, twistLimitDegrees=5}
 {bone1Name=dyn_clothBack_outer_05_r, bone2Name=dyn_clothBack_outer_04_r, swing1LimitDegrees=25, swing2LimitDegrees=25, twistLimitDegrees=5}
 {bone1Name=dyn_clothBack_outer_06_r, bone2Name=dyn_clothBack_outer_05_r, swing1LimitDegrees=25, swing2LimitDegrees=25, twistLimitDegrees=5}
 {bone1Name=dyn_clothFront_inner_02_l, bone2Name=dyn_clothFront_inner_01_l, swing1LimitDegrees=25, swing2LimitDegrees=25, twistLimitDegrees=5}
 {bone1Name=dyn_clothFront_inner_03_l, bone2Name=dyn_clothFront_inner_02_l, swing1LimitDegrees=25, swing2LimitDegrees=25, twistLimitDegrees=5}
 {bone1Name=dyn_clothFront_inner_04_l, bone2Name=dyn_clothFront_inner_03_l, swing1LimitDegrees=25, swing2LimitDegrees=25, twistLimitDegrees=5}
 {bone1Name=dyn_clothFront_inner_05_l, bone2Name=dyn_clothFront_inner_04_l, swing1LimitDegrees=25, swing2LimitDegrees=25, twistLimitDegrees=5}
 {bone1Name=dyn_clothFront_inner_02_r, bone2Name=dyn_clothFront_inner_01_r, swing1LimitDegrees=30, swing2LimitDegrees=30, twistLimitDegrees=5}
 {bone1Name=dyn_clothFront_inner_03_r, bone2Name=dyn_clothFront_inner_02_r, swing1LimitDegrees=25, swing2LimitDegrees=25, twistLimitDegrees=5}
 {bone1Name=dyn_clothFront_inner_04_r, bone2Name=dyn_clothFront_inner_03_r, swing1LimitDegrees=25, swing2LimitDegrees=25, twistLimitDegrees=5}
 {bone1Name=dyn_clothFront_inner_05_r, bone2Name=dyn_clothFront_inner_04_r, swing1LimitDegrees=25, swing2LimitDegrees=25, twistLimitDegrees=5}
]
