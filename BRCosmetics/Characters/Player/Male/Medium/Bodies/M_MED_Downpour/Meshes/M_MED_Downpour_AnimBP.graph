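# Animation Blueprint for the Downpour medium male player body that adds physics-driven secondary motion to its cloth bones.
# Additive bone offsets, some driven by arm-raise and crouch-walk alphas, correct the cached pre-simulation pose.
# One chain runs spring-based AnimDynamics on top cloth bones, then a RigidBody simulation using the perf1 physics asset.
# Another chain simulates back-cloth bones segment by segment with AnimDynamics; further RigidBody nodes use pouch and perf1 assets.
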AnimBlueprint /BRCosmetics/Characters/Player/Male/Medium/Bodies/M_MED_Downpour/Meshes/M_MED_Downpour_AnimBP
target_skeleton=&/M_MED_Downpour_Skeleton
graphs [
 AnimGraph {nodes [
   Anim Dynamics - Bone: dyn_clothBack_outer_01_r!(ComponentPose, Alpha=1, GravityOverride:Vector="0,0,-980", ExternalForce="0,0,0") -> (Pose) {
    angularDampingOverride=1
    angularSpringConstant=15000
    bAngularSpring=true
    bGravityOverrideInSimSpace=true
    bOverrideAngularDamping=true
    bUseGravityOverride=true
    gravityScale=0.35
    simulationSpace=RootRelative
    alphaBoolBlend {blendOption=Linear}
    boundBone {boneName=dyn_clothBack_outer_01_r}
    componentAppliedLinearAccClamp=<3000 3000 3000>
    componentLinearAccScale=<0.4 0.3 0.3>
    componentLinearVelScale=<0.4 0.3 0.3>
    gravityOverride {z=-980}
    physicsBodyDefinitions [{
      sphereCollisionRadius=0
      boundBone {boneName=dyn_clothBack_outer_01_r}
      boxExtents=<27 42 42>
      constraintSetup {angularLimitsMax=<2.5 0 20>, angularLimitsMin=<-1 -5 0>, angularTarget=<1 0 0>}
      localJointOffset=<4 0 0>
     }]
   }
   n1 = Get GravityOverride()
   RigidBody!(ComponentPose, Alpha, OverrideWorldGravity:Vector=@n1, ExternalForce) -> (Pose) {
    bClampLinearTranslationLimitToRefPose=true
    bOverrideWorldGravity=true
    overridePhysicsAsset=&/M_MED_Downpour_Physics_Pouch
    simulationSpace=BaseBoneSpace
    alphaBoolBlend {blendOption=Linear}
    baseBoneRef {boneName=spine_05}
    componentAppliedLinearAccClamp=<3000 3000 3000>
    componentLinearAccScale=<0.4 0.3 0.3>
    componentLinearVelScale=<0.4 0.3 0.3>
    overrideWorldGravity {z=-980}
   }
   RigidBody(OverrideWorldGravity=@n1) {
    bClampLinearTranslationLimitToRefPose=true
    bOverrideWorldGravity=true
    overridePhysicsAsset=&/M_MED_Downpour_Physics_perf1
    simulationSpace=BaseBoneSpace
    alphaBoolBlend {blendOption=Linear}
    baseBoneRef {boneName=pelvis}
    componentAppliedLinearAccClamp=<3000 3000 3000>
    componentLinearAccScale=<0.4 0.3 0.3>
    componentLinearVelScale=<0.4 0.3 0.3>
    overrideWorldGravity {z=-980}
    simSpaceSettings {
     maxAngularAcceleration=1000
     maxAngularVelocity=3
     maxLinearAcceleration=700
     maxLinearVelocity=500
     velocityScaleZ=0
     worldAlpha=1
     externalLinearDragV=<0.2 0.2>
    }
   }
   n2 = "Transform (Modify) Bone - Bone: dyn_clothFront_outer_mstr_l"(
    ComponentPose=Local To Component(LocalPose=Use cached pose 'preSimCache'())
    Alpha=Get L Arm Up Alpha() -> (L Arm Up Alpha:double)
    Translation="-2,2,2.5"
    Rotation="0,25,75"
    Scale="1,1,1"
   ) {
    rotationMode=BMM_Additive
    rotationSpace=BCS_BoneSpace
    translationMode=BMM_Additive
    translationSpace=BCS_BoneSpace
    alphaBoolBlend {blendOption=Linear}
    boneToModify {boneName=dyn_clothFront_outer_mstr_l}
    rotation {roll=75, yaw=25}
    translation=<-2 2 2.5>
   }
   n3 = "Transform (Modify) Bone - Bone: dyn_clothFront_inner_mstr_r"!(ComponentPose=@n2, bAlphaBoolEnabled:bool=True, Translation="-2,1,-1", Rotation="10,5,0.000", Scale="1,1,1") -> (Pose) {
    alphaInputType=Bool
    rotationMode=BMM_Additive
    rotationSpace=BCS_BoneSpace
    translationMode=BMM_Additive
    translationSpace=BCS_BoneSpace
    alphaBoolBlend {blendInTime=0.1, blendOption=Linear, blendOutTime=0.1}
    boneToModify {boneName=dyn_clothFront_inner_mstr_r}
    rotation {pitch=10, yaw=5}
    translation=<-2 1 -1>
   }
   n4 = "Transform (Modify) Bone - Bone: dyn_clothFront_inner_mstr_r"(
    ComponentPose=@n3
    bAlphaBoolEnabled=Get CrouchWalkAdjAlpha() -> (CrouchWalkAdjAlpha:bool)
    Translation="-2,4,0"
    Rotation="0,2,0"
    Scale="1,1,1"
   ) {
    alphaInputType=Bool
    rotationMode=BMM_Additive
    rotationSpace=BCS_BoneSpace
    translationMode=BMM_Additive
    translationSpace=BCS_BoneSpace
    alphaBoolBlend {blendInTime=0.2, blendOption=Linear, blendOutTime=0.2}
    boneToModify {boneName=dyn_clothFront_inner_mstr_r}
    rotation {yaw=2}
    translation=<-2 4>
   }
   n5 = "Transform (Modify) Bone - Bone: dyn_clothBack_outer_mstr_r"!(ComponentPose=@n4, bAlphaBoolEnabled:bool=True, Translation="0,1,0", Rotation="2.5,0,0.000", Scale="1,1,1") -> (Pose) {
    alphaInputType=Bool
    rotationMode=BMM_Additive
    rotationSpace=BCS_BoneSpace
    translationMode=BMM_Additive
    translationSpace=BCS_BoneSpace
    alphaBoolBlend {blendInTime=0.1, blendOption=Linear, blendOutTime=0.1}
    boneToModify {boneName=dyn_clothBack_outer_mstr_r}
    rotation {pitch=2.5}
    translation {y=1}
   }
   n6 = Anim Dynamics - Bone: dyn_clothBack_outer_01_l!(ComponentPose=@n5, Alpha=1, GravityOverride:Vector="0,0,-980", ExternalForce="0,0,0") -> (Pose) {
    angularDampingOverride=1
    angularSpringConstant=15000
    bAngularSpring=true
    bGravityOverrideInSimSpace=true
    bOverrideAngularDamping=true
    bUseGravityOverride=true
    gravityScale=0.35
    simulationSpace=RootRelative
    alphaBoolBlend {blendOption=Linear}
    boundBone {boneName=dyn_clothBack_outer_01_l}
    componentAppliedLinearAccClamp=<3000 3000 3000>
    componentLinearAccScale=<0.4 0.3 0.3>
    componentLinearVelScale=<0.4 0.3 0.3>
    gravityOverride {z=-980}
    physicsBodyDefinitions [{
      sphereCollisionRadius=0
      boundBone {boneName=dyn_clothBack_outer_01_l}
      boxExtents=<26 41 41>
      constraintSetup {angularLimitsMax=<2.5 0 2>, angularLimitsMin=<-1 -5 -20>, angularTarget=<1 0 0>}
      localJointOffset=<4 0 0>
     }]
   }
   n7 = Anim Dynamics - Bone: dyn_clothBack_mid_01_r!(ComponentPose=@n6, Alpha=1, GravityOverride:Vector="0,0,-980", ExternalForce="0,0,0") -> (Pose) {
    angularDampingOverride=1
    angularSpringConstant=15000
    bAngularSpring=true
    bGravityOverrideInSimSpace=true
    bOverrideAngularDamping=true
    bUseGravityOverride=true
    gravityScale=0.35
    simulationSpace=RootRelative
    alphaBoolBlend {blendOption=Linear}
    boundBone {boneName=dyn_clothBack_mid_01_r}
    componentAppliedLinearAccClamp=<3000 3000 3000>
    componentLinearAccScale=<0.4 0.3 0.3>
    componentLinearVelScale=<0.4 0.3 0.3>
    gravityOverride {z=-980}
    physicsBodyDefinitions [{
      sphereCollisionRadius=0
      boundBone {boneName=dyn_clothBack_mid_01_r}
      boxExtents=<27 42 42>
      constraintSetup {angularLimitsMax=<2.5 2.5 12.5>, angularLimitsMin=<-2.5 -12.5 -12.5>, angularTarget=<1 0 0>}
      localJointOffset=<4 0 0>
     }]
   }
   n8 = Anim Dynamics - Bone: dyn_clothBack_inner_01_r!(ComponentPose=@n7, Alpha=1, GravityOverride:Vector="0,0,-980", ExternalForce="0,0,0") -> (Pose) {
    angularDampingOverride=1
    angularSpringConstant=15000
    bAngularSpring=true
    bGravityOverrideInSimSpace=true
    bOverrideAngularDamping=true
    bUseGravityOverride=true
    gravityScale=0.35
    simulationSpace=RootRelative
    alphaBoolBlend {blendOption=Linear}
    boundBone {boneName=dyn_clothBack_inner_01_r}
    componentAppliedLinearAccClamp=<3000 3000 3000>
    componentLinearAccScale=<0.4 0.3 0.3>
    componentLinearVelScale=<0.4 0.3 0.3>
    gravityOverride {z=-980}
    physicsBodyDefinitions [{
      sphereCollisionRadius=0
      boundBone {boneName=dyn_clothBack_inner_01_r}
      boxExtents=<23 37 37>
      constraintSetup {angularLimitsMax=<2.5 2.5 12.5>, angularLimitsMin=<-2.5 -12.5 -12.5>, angularTarget=<1 0 0>}
      localJointOffset=<4 0 0>
     }]
   }
   n9 = Anim Dynamics - Bone: dyn_clothBack_inner_01_l!(ComponentPose=@n8, Alpha=1, GravityOverride:Vector="0,0,-980", ExternalForce="0,0,0") -> (Pose) {
    angularDampingOverride=1
    angularSpringConstant=15000
    bAngularSpring=true
    bGravityOverrideInSimSpace=true
    bOverrideAngularDamping=true
    bUseGravityOverride=true
    gravityScale=0.35
    simulationSpace=RootRelative
    alphaBoolBlend {blendOption=Linear}
    boundBone {boneName=dyn_clothBack_inner_01_l}
    componentAppliedLinearAccClamp=<3000 3000 3000>
    componentLinearAccScale=<0.4 0.3 0.3>
    componentLinearVelScale=<0.4 0.3 0.3>
    gravityOverride {z=-980}
    physicsBodyDefinitions [{
      sphereCollisionRadius=0
      boundBone {boneName=dyn_clothBack_inner_01_l}
      boxExtents=<25 40 40>
      constraintSetup {angularLimitsMax=<2.5 2.5 12.5>, angularLimitsMin=<-2.5 -12.5 -12.5>, angularTarget=<1 0 0>}
      localJointOffset=<4 0 0>
     }]
   }
   n10 = Anim Dynamics - Bone: dyn_clothBack_mid_01_l!(ComponentPose=@n9, Alpha=1, GravityOverride:Vector="0,0,-980", ExternalForce="0,0,0") -> (Pose) {
    angularDampingOverride=1
    angularSpringConstant=15000
    bAngularSpring=true
    bGravityOverrideInSimSpace=true
    bOverrideAngularDamping=true
    bUseGravityOverride=true
    gravityScale=0.35
    simulationSpace=RootRelative
    alphaBoolBlend {blendOption=Linear}
    boundBone {boneName=dyn_clothBack_mid_01_l}
    componentAppliedLinearAccClamp=<3000 3000 3000>
    componentLinearAccScale=<0.4 0.3 0.3>
    componentLinearVelScale=<0.4 0.3 0.3>
    gravityOverride {z=-980}
    physicsBodyDefinitions [{
      sphereCollisionRadius=0
      boundBone {boneName=dyn_clothBack_mid_01_l}
      boxExtents=<26 41 41>
      constraintSetup {angularLimitsMax=<2.5 2.5 12.5>, angularLimitsMin=<-2.5 -12.5 -12.5>, angularTarget=<1 0 0>}
      localJointOffset=<4 0 0>
     }]
   }
   n11 = Anim Dynamics - Bone: dyn_clothFront_outer_01_l!(ComponentPose=@n10, Alpha=1, GravityOverride:Vector="0,0,-980", ExternalForce="0,0,0") -> (Pose) {
    angularDampingOverride=1
    angularSpringConstant=15000
    bAngularSpring=true
    bGravityOverrideInSimSpace=true
    bOverrideAngularDamping=true
    bUseGravityOverride=true
    gravityScale=0.35
    simulationSpace=RootRelative
    alphaBoolBlend {blendOption=Linear}
    boundBone {boneName=dyn_clothFront_outer_01_l}
    componentAppliedLinearAccClamp=<3000 3000 3000>
    componentLinearAccScale=<0.4 0.3 0.3>
    componentLinearVelScale=<0.4 0.3 0.3>
    gravityOverride {z=-980}
    physicsBodyDefinitions [{
      sphereCollisionRadius=0
      boundBone {boneName=dyn_clothFront_outer_01_l}
      boxExtents=<25 40 40>
      constraintSetup {angularLimitsMax=<2.5 12.5 12.5>, angularLimitsMin=<-2.5 -12.5 -12.5>, angularTarget=<1 0 0>}
      localJointOffset=<4 0 0>
     }]
   }
   n12 = Anim Dynamics - Bone: dyn_clothFront_inner_01_l!(ComponentPose=@n11, Alpha=1, GravityOverride:Vector="0,0,-980", ExternalForce="0,0,0") -> (Pose) {
    angularDampingOverride=1
    angularSpringConstant=15000
    bAngularSpring=true
    bGravityOverrideInSimSpace=true
    bOverrideAngularDamping=true
    bUseGravityOverride=true
    gravityScale=0.35
    simulationSpace=RootRelative
    alphaBoolBlend {blendOption=Linear}
    boundBone {boneName=dyn_clothFront_inner_01_l}
    componentAppliedLinearAccClamp=<3000 3000 3000>
    componentLinearAccScale=<0.4 0.3 0.3>
    componentLinearVelScale=<0.4 0.3 0.3>
    gravityOverride {z=-980}
    physicsBodyDefinitions [{
      sphereCollisionRadius=0
      boundBone {boneName=dyn_clothFront_inner_01_l}
      boxExtents=<25 40 40>
      constraintSetup {angularLimitsMax=<2.5 12.5 12.5>, angularLimitsMin=<-2.5 -12.5 -12.5>, angularTarget=<1 0 0>}
      localJointOffset=<4 0 0>
     }]
   }
   n13 = Anim Dynamics - Bone: dyn_clothFront_outer_01_r!(ComponentPose=@n12, Alpha=1, GravityOverride:Vector="0,0,-980", ExternalForce="0,0,0") -> (Pose) {
    angularDampingOverride=1
    angularSpringConstant=15000
    bAngularSpring=true
    bGravityOverrideInSimSpace=true
    bOverrideAngularDamping=true
    bUseGravityOverride=true
    gravityScale=0.35
    simulationSpace=RootRelative
    alphaBoolBlend {blendOption=Linear}
    boundBone {boneName=dyn_clothFront_outer_01_r}
    componentAppliedLinearAccClamp=<3000 3000 3000>
    componentLinearAccScale=<0.4 0.3 0.3>
    componentLinearVelScale=<0.4 0.3 0.3>
    gravityOverride {z=-980}
    physicsBodyDefinitions [{
      sphereCollisionRadius=0
      boundBone {boneName=dyn_clothFront_outer_01_r}
      boxExtents=<25 40 40>
      constraintSetup {angularLimitsMax=<2.5 12.5 12.5>, angularLimitsMin=<-2.5 -12.5 -12.5>, angularTarget=<1 0 0>}
      localJointOffset=<4 0 0>
     }]
   }
   n14 = Anim Dynamics - Bone: dyn_clothSkirt_inner_01_r!(ComponentPose=@n13, Alpha=1, GravityOverride:Vector="0,0,-980", ExternalForce="0,0,0") -> (Pose) {
    angularDampingOverride=1
    angularSpringConstant=15000
    bAngularSpring=true
    bGravityOverrideInSimSpace=true
    bOverrideAngularDamping=true
    bUseGravityOverride=true
    gravityScale=0.35
    simulationSpace=RootRelative
    alphaBoolBlend {blendOption=Linear}
    boundBone {boneName=dyn_clothSkirt_inner_01_r}
    componentAppliedLinearAccClamp=<3000 3000 3000>
    componentLinearAccScale=<0.4 0.3 0.3>
    componentLinearVelScale=<0.4 0.3 0.3>
    gravityOverride {z=-980}
    physicsBodyDefinitions [{
      sphereCollisionRadius=0
      boundBone {boneName=dyn_clothSkirt_inner_01_r}
      boxExtents=<30 40 40>
      constraintSetup {angularLimitsMax=<2.5 12.5 12.5>, angularLimitsMin=<-2.5 -12.5 -12.5>, angularTarget=<1 0 0>}
      localJointOffset=<4 0 0>
     }]
   }
   n15 = Anim Dynamics - Bone: dyn_clothSkirt_inner_01_l!(ComponentPose=@n14, Alpha=1, GravityOverride:Vector="0,0,-980", ExternalForce="0,0,0") -> (Pose) {
    angularDampingOverride=1
    angularSpringConstant=15000
    bAngularSpring=true
    bGravityOverrideInSimSpace=true
    bOverrideAngularDamping=true
    bUseGravityOverride=true
    gravityScale=0.35
    simulationSpace=RootRelative
    alphaBoolBlend {blendOption=Linear}
    boundBone {boneName=dyn_clothSkirt_inner_01_l}
    componentAppliedLinearAccClamp=<3000 3000 3000>
    componentLinearAccScale=<0.4 0.3 0.3>
    componentLinearVelScale=<0.4 0.3 0.3>
    gravityOverride {z=-980}
    physicsBodyDefinitions [{
      sphereCollisionRadius=0
      boundBone {boneName=dyn_clothSkirt_inner_01_l}
      boxExtents=<30 40 40>
      constraintSetup {angularLimitsMax=<2.5 12.5 12.5>, angularLimitsMin=<-2.5 -12.5 -12.5>, angularTarget=<1 0 0>}
      localJointOffset=<4 0 0>
     }]
   }
   n16 = Anim Dynamics - Bone: dyn_clothSkirt_outer_01_l!(ComponentPose=@n15, Alpha=1, GravityOverride:Vector="0,0,-980", ExternalForce="0,0,0") -> (Pose) {
    angularDampingOverride=1
    angularSpringConstant=15000
    bAngularSpring=true
    bGravityOverrideInSimSpace=true
    bOverrideAngularDamping=true
    bUseGravityOverride=true
    gravityScale=0.35
    simulationSpace=RootRelative
    alphaBoolBlend {blendOption=Linear}
    boundBone {boneName=dyn_clothSkirt_outer_01_l}
    componentAppliedLinearAccClamp=<3000 3000 3000>
    componentLinearAccScale=<0.4 0.3 0.3>
    componentLinearVelScale=<0.4 0.3 0.3>
    gravityOverride {z=-980}
    physicsBodyDefinitions [{
      sphereCollisionRadius=0
      boundBone {boneName=dyn_clothSkirt_outer_01_l}
      boxExtents=<30 40 40>
      constraintSetup {angularLimitsMax=<2.5 12.5 12.5>, angularLimitsMin=<-2.5 -12.5 -12.5>, angularTarget=<1 0 0>}
      localJointOffset=<4 0 0>
     }]
   }
   n17 = RigidBody(ComponentPose=@n16, OverrideWorldGravity=@n1) {
    bClampLinearTranslationLimitToRefPose=true
    bOverrideWorldGravity=true
    overridePhysicsAsset=&M_MED_Downpour_Physics_perf1
    simulationSpace=BaseBoneSpace
    alphaBoolBlend {blendOption=Linear}
    baseBoneRef {boneName=root}
    componentAppliedLinearAccClamp=<3000 3000 3000>
    componentLinearAccScale=<0.35 0.25 0.25>
    componentLinearVelScale=<0.35 0.25 0.25>
    overrideWorldGravity {z=-980}
   }
   n18 = Anim Dynamics - Bone: dyn_clothBack_mid_01_r!(ComponentPose=Local To Component(LocalPose=Use cached pose 'preSimCache'()), Alpha=1, ExternalForce="0,0,0") -> (Pose) {
    angularDampingOverride=1
    angularSpringConstant=15000
    bAngularSpring=true
    bGravityOverrideInSimSpace=true
    bOverrideAngularDamping=true
    bUseGravityOverride=true
    gravityScale=0.35
    simulationSpace=RootRelative
    alphaBoolBlend {blendOption=Linear}
    boundBone {boneName=dyn_clothBack_mid_01_r}
    componentAppliedLinearAccClamp=<3000 3000 3000>
    componentLinearAccScale=<1 1 1>
    componentLinearVelScale=<1 1 1>
    gravityOverride {z=-980}
    physicsBodyDefinitions [{
      sphereCollisionRadius=0
      boundBone {boneName=dyn_clothBack_mid_01_r}
      boxExtents=<27 42 42>
      constraintSetup {angularLimitsMax=<15 3 15>, angularLimitsMin=<-15 -15 -15>, angularTarget=<1 0 0>}
      localJointOffset=<4 0 0>
     }]
   }
   n19 = Anim Dynamics - Bone: dyn_clothBack_mid_02_r(ComponentPose=@n18, Alpha=1, ExternalForce="0,0,0") {
    angularDampingOverride=1
    angularSpringConstant=15000
    bAngularSpring=true
    bGravityOverrideInSimSpace=true
    bOverrideAngularDamping=true
    bUseGravityOverride=true
    gravityScale=0.35
    simulationSpace=RootRelative
    alphaBoolBlend {blendOption=Linear}
    boundBone {boneName=dyn_clothBack_mid_02_r}
    componentAppliedLinearAccClamp=<3000 3000 3000>
    componentLinearAccScale=<1 1 1>
    componentLinearVelScale=<1 1 1>
    gravityOverride {z=-980}
    physicsBodyDefinitions [{
      sphereCollisionRadius=0
      boundBone {boneName=dyn_clothBack_mid_02_r}
      boxExtents=<27 42 42>
      constraintSetup {angularLimitsMax=<15 3 15>, angularLimitsMin=<-15 -15 -15>, angularTarget=<1 0 0>}
      localJointOffset=<4 0 0>
     }]
   }
   n20 = Anim Dynamics - Bone: dyn_clothBack_mid_03_r(ComponentPose=@n19, Alpha=1, ExternalForce="0,0,0") {
    angularDampingOverride=1
    angularSpringConstant=15000
    bAngularSpring=true
    bGravityOverrideInSimSpace=true
    bOverrideAngularDamping=true
    bUseGravityOverride=true
    gravityScale=0.35
    simulationSpace=RootRelative
    alphaBoolBlend {blendOption=Linear}
    boundBone {boneName=dyn_clothBack_mid_03_r}
    componentAppliedLinearAccClamp=<3000 3000 3000>
    componentLinearAccScale=<1 1 1>
    componentLinearVelScale=<1 1 1>
    gravityOverride {z=-980}
    physicsBodyDefinitions [{
      sphereCollisionRadius=0
      boundBone {boneName=dyn_clothBack_mid_03_r}
      boxExtents=<27 42 42>
      constraintSetup {angularLimitsMax=<15 3 15>, angularLimitsMin=<-15 -15 -15>, angularTarget=<1 0 0>}
      localJointOffset=<4 0 0>
     }]
   }
   n21 = Anim Dynamics - Bone: dyn_clothBack_mid_04_r(ComponentPose=@n20, Alpha=1, ExternalForce="0,0,0") {
    angularDampingOverride=1
    angularSpringConstant=15000
    bAngularSpring=true
    bGravityOverrideInSimSpace=true
    bOverrideAngularDamping=true
    bUseGravityOverride=true
    gravityScale=0.35
    simulationSpace=RootRelative
    alphaBoolBlend {blendOption=Linear}
    boundBone {boneName=dyn_clothBack_mid_04_r}
    componentAppliedLinearAccClamp=<3000 3000 3000>
    componentLinearAccScale=<1 1 1>
    componentLinearVelScale=<1 1 1>
    gravityOverride {z=-980}
    physicsBodyDefinitions [{
      sphereCollisionRadius=0
      boundBone {boneName=dyn_clothBack_mid_04_r}
      boxExtents=<27 42 42>
      constraintSetup {angularLimitsMax=<15 3 15>, angularLimitsMin=<-15 -15 -15>, angularTarget=<1 0 0>}
      localJointOffset=<4 0 0>
     }]
   }
   n22 = Anim Dynamics - Bone: dyn_clothBack_mid_05_r(ComponentPose=@n21, Alpha=1, ExternalForce="0,0,0") {
    angularDampingOverride=1
    angularSpringConstant=15000
    bAngularSpring=true
    bGravityOverrideInSimSpace=true
    bOverrideAngularDamping=true
    bUseGravityOverride=true
    gravityScale=0.35
    simulationSpace=RootRelative
    alphaBoolBlend {blendOption=Linear}
    boundBone {boneName=dyn_clothBack_mid_05_r}
    componentAppliedLinearAccClamp=<3000 3000 3000>
    componentLinearAccScale=<1 1 1>
    componentLinearVelScale=<1 1 1>
    gravityOverride {z=-980}
    physicsBodyDefinitions [{
      sphereCollisionRadius=0
      boundBone {boneName=dyn_clothBack_mid_05_r}
      boxExtents=<27 42 42>
      constraintSetup {angularLimitsMax=<15 3 15>, angularLimitsMin=<-15 -15 -15>, angularTarget=<1 0 0>}
      localJointOffset=<4 0 0>
     }]
   }
   n23 = Anim Dynamics - Bone: dyn_clothBack_mid_06_r(ComponentPose=@n22, Alpha=1, ExternalForce="0,0,0") {
    angularDampingOverride=1
    angularSpringConstant=15000
    bAngularSpring=true
    bGravityOverrideInSimSpace=true
    bOverrideAngularDamping=true
    bUseGravityOverride=true
    gravityScale=0.35
    simulationSpace=RootRelative
    alphaBoolBlend {blendOption=Linear}
    boundBone {boneName=dyn_clothBack_mid_06_r}
    componentAppliedLinearAccClamp=<3000 3000 3000>
    componentLinearAccScale=<1 1 1>
    componentLinearVelScale=<1 1 1>
    gravityOverride {z=-980}
    physicsBodyDefinitions [{
      sphereCollisionRadius=0
      boundBone {boneName=dyn_clothBack_mid_06_r}
      boxExtents=<27 42 42>
      constraintSetup {angularLimitsMax=<15 3 15>, angularLimitsMin=<-15 -15 -15>, angularTarget=<1 0 0>}
      localJointOffset=<4 0 0>
     }]
   }
   n24 = Anim Dynamics - Bone: dyn_clothBack_inner_01_r!(ComponentPose=@n23, Alpha=1, ExternalForce="0,0,0") -> (Pose) {
    angularDampingOverride=1
    angularSpringConstant=15000
    bAngularSpring=true
    bGravityOverrideInSimSpace=true
    bOverrideAngularDamping=true
    bUseGravityOverride=true
    gravityScale=0.35
    simulationSpace=RootRelative
    alphaBoolBlend {blendOption=Linear}
    boundBone {boneName=dyn_clothBack_inner_01_r}
    componentAppliedLinearAccClamp=<3000 3000 3000>
    componentLinearAccScale=<1 1 1>
    componentLinearVelScale=<1 1 1>
    gravityOverride {z=-980}
    physicsBodyDefinitions [{
      sphereCollisionRadius=0
      boundBone {boneName=dyn_clothBack_inner_01_r}
      boxExtents=<23 37 37>
      constraintSetup {angularLimitsMax=<15 3 15>, angularLimitsMin=<-15 -15 -15>, angularTarget=<1 0 0>}
      localJointOffset=<4 0 0>
     }]
   }
   n25 = Anim Dynamics - Bone: dyn_clothBack_inner_02_r(ComponentPose=@n24, Alpha=1, ExternalForce="0,0,0") {
    angularDampingOverride=1
    angularSpringConstant=15000
    bAngularSpring=true
    bGravityOverrideInSimSpace=true
    bOverrideAngularDamping=true
    bUseGravityOverride=true
    gravityScale=0.35
    simulationSpace=RootRelative
    alphaBoolBlend {blendOption=Linear}
    boundBone {boneName=dyn_clothBack_inner_02_r}
    componentAppliedLinearAccClamp=<3000 3000 3000>
    componentLinearAccScale=<1 1 1>
    componentLinearVelScale=<1 1 1>
    gravityOverride {z=-980}
    physicsBodyDefinitions [{
      sphereCollisionRadius=0
      boundBone {boneName=dyn_clothBack_inner_02_r}
      boxExtents=<23 37 37>
      constraintSetup {angularLimitsMax=<15 3 15>, angularLimitsMin=<-15 -15 -15>, angularTarget=<1 0 0>}
      localJointOffset=<4 0 0>
     }]
   }
   n26 = Anim Dynamics - Bone: dyn_clothBack_inner_03_r(ComponentPose=@n25, Alpha=1, ExternalForce="0,0,0") {
    angularDampingOverride=1
    angularSpringConstant=15000
    bAngularSpring=true
    bGravityOverrideInSimSpace=true
    bOverrideAngularDamping=true
    bUseGravityOverride=true
    gravityScale=0.35
    simulationSpace=RootRelative
    alphaBoolBlend {blendOption=Linear}
    boundBone {boneName=dyn_clothBack_inner_03_r}
    componentAppliedLinearAccClamp=<3000 3000 3000>
    componentLinearAccScale=<1 1 1>
    componentLinearVelScale=<1 1 1>
    gravityOverride {z=-980}
    physicsBodyDefinitions [{
      sphereCollisionRadius=0
      boundBone {boneName=dyn_clothBack_inner_03_r}
      boxExtents=<23 37 37>
      constraintSetup {angularLimitsMax=<15 3 15>, angularLimitsMin=<-15 -15 -15>, angularTarget=<1 0 0>}
      localJointOffset=<4 0 0>
     }]
   }
   n27 = Anim Dynamics - Bone: dyn_clothBack_inner_04_r(ComponentPose=@n26, Alpha=1, ExternalForce="0,0,0") {
    angularDampingOverride=1
    angularSpringConstant=15000
    bAngularSpring=true
    bGravityOverrideInSimSpace=true
    bOverrideAngularDamping=true
    bUseGravityOverride=true
    gravityScale=0.35
    simulationSpace=RootRelative
    alphaBoolBlend {blendOption=Linear}
    boundBone {boneName=dyn_clothBack_inner_04_r}
    componentAppliedLinearAccClamp=<3000 3000 3000>
    componentLinearAccScale=<1 1 1>
    componentLinearVelScale=<1 1 1>
    gravityOverride {z=-980}
    physicsBodyDefinitions [{
      sphereCollisionRadius=0
      boundBone {boneName=dyn_clothBack_inner_04_r}
      boxExtents=<23 37 37>
      constraintSetup {angularLimitsMax=<15 3 15>, angularLimitsMin=<-15 -15 -15>, angularTarget=<1 0 0>}
      localJointOffset=<4 0 0>
     }]
   }
   n28 = Anim Dynamics - Bone: dyn_clothBack_inner_05_r(ComponentPose=@n27, Alpha=1, ExternalForce="0,0,0") {
    angularDampingOverride=1
    angularSpringConstant=15000
    bAngularSpring=true
    bGravityOverrideInSimSpace=true
    bOverrideAngularDamping=true
    bUseGravityOverride=true
    gravityScale=0.35
    simulationSpace=RootRelative
    alphaBoolBlend {blendOption=Linear}
    boundBone {boneName=dyn_clothBack_inner_05_r}
    componentAppliedLinearAccClamp=<3000 3000 3000>
    componentLinearAccScale=<1 1 1>
    componentLinearVelScale=<1 1 1>
    gravityOverride {z=-980}
    physicsBodyDefinitions [{
      sphereCollisionRadius=0
      boundBone {boneName=dyn_clothBack_inner_05_r}
      boxExtents=<23 37 37>
      constraintSetup {angularLimitsMax=<15 3 15>, angularLimitsMin=<-15 -15 -15>, angularTarget=<1 0 0>}
      localJointOffset=<4 0 0>
     }]
   }
   n29 = Anim Dynamics - Bone: dyn_clothBack_inner_06_r(ComponentPose=@n28, Alpha=1, ExternalForce="0,0,0") {
    angularDampingOverride=1
    angularSpringConstant=15000
    bAngularSpring=true
    bGravityOverrideInSimSpace=true
    bOverrideAngularDamping=true
    bUseGravityOverride=true
    gravityScale=0.35
    simulationSpace=RootRelative
    alphaBoolBlend {blendOption=Linear}
    boundBone {boneName=dyn_clothBack_inner_06_r}
    componentAppliedLinearAccClamp=<3000 3000 3000>
    componentLinearAccScale=<1 1 1>
    componentLinearVelScale=<1 1 1>
    gravityOverride {z=-980}
    physicsBodyDefinitions [{
      sphereCollisionRadius=0
      boundBone {boneName=dyn_clothBack_inner_06_r}
      boxExtents=<23 37 37>
      constraintSetup {angularLimitsMax=<15 3 15>, angularLimitsMin=<-15 -15 -15>, angularTarget=<1 0 0>}
      localJointOffset=<4 0 0>
     }]
   }
   n30 = Anim Dynamics - Bone: dyn_clothBack_inner_01_l!(ComponentPose=@n29, Alpha=1, ExternalForce="0,0,0") -> (Pose) {
    angularDampingOverride=1
    angularSpringConstant=15000
    bAngularSpring=true
    bGravityOverrideInSimSpace=true
    bOverrideAngularDamping=true
    bUseGravityOverride=true
    gravityScale=0.35
    simulationSpace=RootRelative
    alphaBoolBlend {blendOption=Linear}
    boundBone {boneName=dyn_clothBack_inner_01_l}
    componentAppliedLinearAccClamp=<3000 3000 3000>
    componentLinearAccScale=<1 1 1>
    componentLinearVelScale=<1 1 1>
    gravityOverride {z=-980}
    physicsBodyDefinitions [{
      sphereCollisionRadius=0
      boundBone {boneName=dyn_clothBack_inner_01_l}
      boxExtents=<25 40 40>
      constraintSetup {angularLimitsMax=<15 3 15>, angularLimitsMin=<-15 -15 -15>, angularTarget=<1 0 0>}
      localJointOffset=<4 0 0>
     }]
   }
   n31 = Anim Dynamics - Bone: dyn_clothBack_inner_02_l(ComponentPose=@n30, Alpha=1, ExternalForce="0,0,0") {
    angularDampingOverride=1
    angularSpringConstant=15000
    bAngularSpring=true
    bGravityOverrideInSimSpace=true
    bOverrideAngularDamping=true
    bUseGravityOverride=true
    gravityScale=0.35
    simulationSpace=RootRelative
    alphaBoolBlend {blendOption=Linear}
    boundBone {boneName=dyn_clothBack_inner_02_l}
    componentAppliedLinearAccClamp=<3000 3000 3000>
    componentLinearAccScale=<1 1 1>
    componentLinearVelScale=<1 1 1>
    gravityOverride {z=-980}
    physicsBodyDefinitions [{
      sphereCollisionRadius=0
      boundBone {boneName=dyn_clothBack_inner_02_l}
      boxExtents=<25 40 40>
      constraintSetup {angularLimitsMax=<15 3 15>, angularLimitsMin=<-15 -15 -15>, angularTarget=<1 0 0>}
      localJointOffset=<4 0 0>
     }]
   }
   n32 = Anim Dynamics - Bone: dyn_clothBack_inner_03_l(ComponentPose=@n31, Alpha=1, ExternalForce="0,0,0") {
    angularDampingOverride=1
    angularSpringConstant=15000
    bAngularSpring=true
    bGravityOverrideInSimSpace=true
    bOverrideAngularDamping=true
    bUseGravityOverride=true
    gravityScale=0.35
    simulationSpace=RootRelative
    alphaBoolBlend {blendOption=Linear}
    boundBone {boneName=dyn_clothBack_inner_03_l}
    componentAppliedLinearAccClamp=<3000 3000 3000>
    componentLinearAccScale=<1 1 1>
    componentLinearVelScale=<1 1 1>
    gravityOverride {z=-980}
    physicsBodyDefinitions [{
      sphereCollisionRadius=0
      boundBone {boneName=dyn_clothBack_inner_03_l}
      boxExtents=<25 40 40>
      constraintSetup {angularLimitsMax=<15 3 15>, angularLimitsMin=<-15 -15 -15>, angularTarget=<1 0 0>}
      localJointOffset=<4 0 0>
     }]
   }
   n33 = Anim Dynamics - Bone: dyn_clothBack_inner_04_l(ComponentPose=@n32, Alpha=1, ExternalForce="0,0,0") {
    angularDampingOverride=1
    angularSpringConstant=15000
    bAngularSpring=true
    bGravityOverrideInSimSpace=true
    bOverrideAngularDamping=true
    bUseGravityOverride=true
    gravityScale=0.35
    simulationSpace=RootRelative
    alphaBoolBlend {blendOption=Linear}
    boundBone {boneName=dyn_clothBack_inner_04_l}
    componentAppliedLinearAccClamp=<3000 3000 3000>
    componentLinearAccScale=<1 1 1>
    componentLinearVelScale=<1 1 1>
    gravityOverride {z=-980}
    physicsBodyDefinitions [{
      sphereCollisionRadius=0
      boundBone {boneName=dyn_clothBack_inner_04_l}
      boxExtents=<25 40 40>
      constraintSetup {angularLimitsMax=<15 3 15>, angularLimitsMin=<-15 -15 -15>, angularTarget=<1 0 0>}
      localJointOffset=<4 0 0>
     }]
   }
   n34 = Anim Dynamics - Bone: dyn_clothBack_inner_05_l(ComponentPose=@n33, Alpha=1, ExternalForce="0,0,0") {
    angularDampingOverride=1
    angularSpringConstant=15000
    bAngularSpring=true
    bGravityOverrideInSimSpace=true
    bOverrideAngularDamping=true
    bUseGravityOverride=true
    gravityScale=0.35
    simulationSpace=RootRelative
    alphaBoolBlend {blendOption=Linear}
    boundBone {boneName=dyn_clothBack_inner_05_l}
    componentAppliedLinearAccClamp=<3000 3000 3000>
    componentLinearAccScale=<1 1 1>
    componentLinearVelScale=<1 1 1>
    gravityOverride {z=-980}
    physicsBodyDefinitions [{
      sphereCollisionRadius=0
      boundBone {boneName=dyn_clothBack_inner_05_l}
      boxExtents=<25 40 40>
      constraintSetup {angularLimitsMax=<15 3 15>, angularLimitsMin=<-15 -15 -15>, angularTarget=<1 0 0>}
      localJointOffset=<4 0 0>
     }]
   }
   n35 = Anim Dynamics - Bone: dyn_clothBack_inner_06_l(ComponentPose=@n34, Alpha=1, ExternalForce="0,0,0") {
    angularDampingOverride=1
    angularSpringConstant=15000
    bAngularSpring=true
    bGravityOverrideInSimSpace=true
    bOverrideAngularDamping=true
    bUseGravityOverride=true
    gravityScale=0.35
    simulationSpace=RootRelative
    alphaBoolBlend {blendOption=Linear}
    boundBone {boneName=dyn_clothBack_inner_06_l}
    componentAppliedLinearAccClamp=<3000 3000 3000>
    componentLinearAccScale=<1 1 1>
    componentLinearVelScale=<1 1 1>
    gravityOverride {z=-980}
    physicsBodyDefinitions [{
      sphereCollisionRadius=0
      boundBone {boneName=dyn_clothBack_inner_06_l}
      boxExtents=<25 40 40>
      constraintSetup {angularLimitsMax=<15 3 15>, angularLimitsMin=<-15 -15 -15>, angularTarget=<1 0 0>}
      localJointOffset=<4 0 0>
     }]
   }
   n36 = Anim Dynamics - Bone: dyn_clothBack_mid_01_l!(ComponentPose=@n35, Alpha=1, ExternalForce="0,0,0") -> (Pose) {
    angularDampingOverride=1
    angularSpringConstant=15000
    bAngularSpring=true
    bGravityOverrideInSimSpace=true
    bOverrideAngularDamping=true
    bUseGravityOverride=true
    gravityScale=0.35
    simulationSpace=RootRelative
    alphaBoolBlend {blendOption=Linear}
    boundBone {boneName=dyn_clothBack_mid_01_l}
    componentAppliedLinearAccClamp=<3000 3000 3000>
    componentLinearAccScale=<1 1 1>
    componentLinearVelScale=<1 1 1>
    gravityOverride {z=-980}
    physicsBodyDefinitions [{
      sphereCollisionRadius=0
      boundBone {boneName=dyn_clothBack_mid_01_l}
      boxExtents=<26 41 41>
      constraintSetup {angularLimitsMax=<15 3 15>, angularLimitsMin=<-15 -15 -15>, angularTarget=<1 0 0>}
      localJointOffset=<4 0 0>
     }]
   }
   n37 = Anim Dynamics - Bone: dyn_clothBack_mid_02_l(ComponentPose=@n36, Alpha=1, ExternalForce="0,0,0") {
    angularDampingOverride=1
    angularSpringConstant=15000
    bAngularSpring=true
    bGravityOverrideInSimSpace=true
    bOverrideAngularDamping=true
    bUseGravityOverride=true
    gravityScale=0.35
    simulationSpace=RootRelative
    alphaBoolBlend {blendOption=Linear}
    boundBone {boneName=dyn_clothBack_mid_02_l}
    componentAppliedLinearAccClamp=<3000 3000 3000>
    componentLinearAccScale=<1 1 1>
    componentLinearVelScale=<1 1 1>
    gravityOverride {z=-980}
    physicsBodyDefinitions [{
      sphereCollisionRadius=0
      boundBone {boneName=dyn_clothBack_mid_02_l}
      boxExtents=<26 41 41>
      constraintSetup {angularLimitsMax=<15 3 15>, angularLimitsMin=<-15 -15 -15>, angularTarget=<1 0 0>}
      localJointOffset=<4 0 0>
     }]
   }
   n38 = Anim Dynamics - Bone: dyn_clothBack_mid_03_l(ComponentPose=@n37, Alpha=1, ExternalForce="0,0,0") {
    angularDampingOverride=1
    angularSpringConstant=15000
    bAngularSpring=true
    bGravityOverrideInSimSpace=true
    bOverrideAngularDamping=true
    bUseGravityOverride=true
    gravityScale=0.35
    simulationSpace=RootRelative
    alphaBoolBlend {blendOption=Linear}
    boundBone {boneName=dyn_clothBack_mid_03_l}
    componentAppliedLinearAccClamp=<3000 3000 3000>
    componentLinearAccScale=<1 1 1>
    componentLinearVelScale=<1 1 1>
    gravityOverride {z=-980}
    physicsBodyDefinitions [{
      sphereCollisionRadius=0
      boundBone {boneName=dyn_clothBack_mid_03_l}
      boxExtents=<26 41 41>
      constraintSetup {angularLimitsMax=<15 3 15>, angularLimitsMin=<-15 -15 -15>, angularTarget=<1 0 0>}
      localJointOffset=<4 0 0>
     }]
   }
   n39 = Anim Dynamics - Bone: dyn_clothBack_mid_04_l(ComponentPose=@n38, Alpha=1, ExternalForce="0,0,0") {
    angularDampingOverride=1
    angularSpringConstant=15000
    bAngularSpring=true
    bGravityOverrideInSimSpace=true
    bOverrideAngularDamping=true
    bUseGravityOverride=true
    gravityScale=0.35
    simulationSpace=RootRelative
    alphaBoolBlend {blendOption=Linear}
    boundBone {boneName=dyn_clothBack_mid_04_l}
    componentAppliedLinearAccClamp=<3000 3000 3000>
    componentLinearAccScale=<1 1 1>
    componentLinearVelScale=<1 1 1>
    gravityOverride {z=-980}
    physicsBodyDefinitions [{
      sphereCollisionRadius=0
      boundBone {boneName=dyn_clothBack_mid_04_l}
      boxExtents=<26 41 41>
      constraintSetup {angularLimitsMax=<15 3 15>, angularLimitsMin=<-15 -15 -15>, angularTarget=<1 0 0>}
      localJointOffset=<4 0 0>
     }]
   }
   n40 = Anim Dynamics - Bone: dyn_clothBack_mid_05_l(ComponentPose=@n39, Alpha=1, ExternalForce="0,0,0") {
    angularDampingOverride=1
    angularSpringConstant=15000
    bAngularSpring=true
    bGravityOverrideInSimSpace=true
    bOverrideAngularDamping=true
    bUseGravityOverride=true
    gravityScale=0.35
    simulationSpace=RootRelative
    alphaBoolBlend {blendOption=Linear}
    boundBone {boneName=dyn_clothBack_mid_05_l}
    componentAppliedLinearAccClamp=<3000 3000 3000>
    componentLinearAccScale=<1 1 1>
    componentLinearVelScale=<1 1 1>
    gravityOverride {z=-980}
    physicsBodyDefinitions [{
      sphereCollisionRadius=0
      boundBone {boneName=dyn_clothBack_mid_05_l}
      boxExtents=<26 41 41>
      constraintSetup {angularLimitsMax=<15 3 15>, angularLimitsMin=<-15 -15 -15>, angularTarget=<1 0 0>}
      localJointOffset=<4 0 0>
     }]
   }
   n41 = Anim Dynamics - Bone: dyn_clothBack_mid_06_l(ComponentPose=@n40, Alpha=1, ExternalForce="0,0,0") {
    angularDampingOverride=1
    angularSpringConstant=15000
    bAngularSpring=true
    bGravityOverrideInSimSpace=true
    bOverrideAngularDamping=true
    bUseGravityOverride=true
    gravityScale=0.35
    simulationSpace=RootRelative
    alphaBoolBlend {blendOption=Linear}
    boundBone {boneName=dyn_clothBack_mid_06_l}
    componentAppliedLinearAccClamp=<3000 3000 3000>
    componentLinearAccScale=<1 1 1>
    componentLinearVelScale=<1 1 1>
    gravityOverride {z=-980}
    physicsBodyDefinitions [{
      sphereCollisionRadius=0
      boundBone {boneName=dyn_clothBack_mid_06_l}
      boxExtents=<26 41 41>
      constraintSetup {angularLimitsMax=<15 3 15>, angularLimitsMin=<-15 -15 -15>, angularTarget=<1 0 0>}
      localJointOffset=<4 0 0>
     }]
   }
   n42 = Anim Dynamics - Bone: dyn_clothBack_outer_01_r!(ComponentPose=@n41, Alpha=1, ExternalForce="0,0,0") -> (Pose) {
    angularDampingOverride=1
    angularSpringConstant=15000
    bAngularSpring=true
    bGravityOverrideInSimSpace=true
    bOverrideAngularDamping=true
    bUseGravityOverride=true
    gravityScale=0.35
    simulationSpace=RootRelative
    alphaBoolBlend {blendOption=Linear}
    boundBone {boneName=dyn_clothBack_outer_01_r}
    componentAppliedLinearAccClamp=<3000 3000 3000>
    componentLinearAccScale=<1 1 1>
    componentLinearVelScale=<1 1 1>
    gravityOverride {z=-980}
    physicsBodyDefinitions [{
      sphereCollisionRadius=0
      boundBone {boneName=dyn_clothBack_outer_01_r}
      boxExtents=<26 42 42>
      constraintSetup {angularLimitsMax=<15 3 15>, angularLimitsMin=<-15 -15 -15>, angularTarget=<1 0 0>}
      localJointOffset=<4 0 0>
     }]
   }
   n43 = Anim Dynamics - Bone: dyn_clothBack_outer_02_r(ComponentPose=@n42, Alpha=1, ExternalForce="0,0,0") {
    angularDampingOverride=1
    angularSpringConstant=15000
    bAngularSpring=true
    bGravityOverrideInSimSpace=true
    bOverrideAngularDamping=true
    bUseGravityOverride=true
    gravityScale=0.35
    simulationSpace=RootRelative
    alphaBoolBlend {blendOption=Linear}
    boundBone {boneName=dyn_clothBack_outer_02_r}
    componentAppliedLinearAccClamp=<3000 3000 3000>
    componentLinearAccScale=<1 1 1>
    componentLinearVelScale=<1 1 1>
    gravityOverride {z=-980}
    physicsBodyDefinitions [{
      sphereCollisionRadius=0
      boundBone {boneName=dyn_clothBack_outer_02_r}
      boxExtents=<26 42 42>
      constraintSetup {angularLimitsMax=<15 3 15>, angularLimitsMin=<-15 -15 -15>, angularTarget=<1 0 0>}
      localJointOffset=<4 0 0>
     }]
   }
   n44 = Anim Dynamics - Bone: dyn_clothBack_outer_03_r(ComponentPose=@n43, Alpha=1, ExternalForce="0,0,0") {
    angularDampingOverride=1
    angularSpringConstant=15000
    bAngularSpring=true
    bGravityOverrideInSimSpace=true
    bOverrideAngularDamping=true
    bUseGravityOverride=true
    gravityScale=0.35
    simulationSpace=RootRelative
    alphaBoolBlend {blendOption=Linear}
    boundBone {boneName=dyn_clothBack_outer_03_r}
    componentAppliedLinearAccClamp=<3000 3000 3000>
    componentLinearAccScale=<1 1 1>
    componentLinearVelScale=<1 1 1>
    gravityOverride {z=-980}
    physicsBodyDefinitions [{
      sphereCollisionRadius=0
      boundBone {boneName=dyn_clothBack_outer_03_r}
      boxExtents=<26 42 42>
      constraintSetup {angularLimitsMax=<15 3 15>, angularLimitsMin=<-15 -15 -15>, angularTarget=<1 0 0>}
      localJointOffset=<4 0 0>
     }]
   }
   n45 = Anim Dynamics - Bone: dyn_clothBack_outer_04_r(ComponentPose=@n44, Alpha=1, ExternalForce="0,0,0") {
    angularDampingOverride=1
    angularSpringConstant=15000
    bAngularSpring=true
    bGravityOverrideInSimSpace=true
    bOverrideAngularDamping=true
    bUseGravityOverride=true
    gravityScale=0.35
    simulationSpace=RootRelative
    alphaBoolBlend {blendOption=Linear}
    boundBone {boneName=dyn_clothBack_outer_04_r}
    componentAppliedLinearAccClamp=<3000 3000 3000>
    componentLinearAccScale=<1 1 1>
    componentLinearVelScale=<1 1 1>
    gravityOverride {z=-980}
    physicsBodyDefinitions [{
      sphereCollisionRadius=0
      boundBone {boneName=dyn_clothBack_outer_04_r}
      boxExtents=<26 42 42>
      constraintSetup {angularLimitsMax=<15 3 15>, angularLimitsMin=<-15 -15 -15>, angularTarget=<1 0 0>}
      localJointOffset=<4 0 0>
     }]
   }
   n46 = Anim Dynamics - Bone: dyn_clothBack_outer_05_r(ComponentPose=@n45, Alpha=1, ExternalForce="0,0,0") {
    angularDampingOverride=1
    angularSpringConstant=15000
    bAngularSpring=true
    bGravityOverrideInSimSpace=true
    bOverrideAngularDamping=true
    bUseGravityOverride=true
    gravityScale=0.35
    simulationSpace=RootRelative
    alphaBoolBlend {blendOption=Linear}
    boundBone {boneName=dyn_clothBack_outer_05_r}
    componentAppliedLinearAccClamp=<3000 3000 3000>
    componentLinearAccScale=<1 1 1>
    componentLinearVelScale=<1 1 1>
    gravityOverride {z=-980}
    physicsBodyDefinitions [{
      sphereCollisionRadius=0
      boundBone {boneName=dyn_clothBack_outer_05_r}
      boxExtents=<26 42 42>
      constraintSetup {angularLimitsMax=<15 3 15>, angularLimitsMin=<-15 -15 -15>, angularTarget=<1 0 0>}
      localJointOffset=<4 0 0>
     }]
   }
   n47 = Anim Dynamics - Bone: dyn_clothBack_outer_06_r(ComponentPose=@n46, Alpha=1, ExternalForce="0,0,0") {
    angularDampingOverride=1
    angularSpringConstant=15000
    bAngularSpring=true
    bGravityOverrideInSimSpace=true
    bOverrideAngularDamping=true
    bUseGravityOverride=true
    gravityScale=0.35
    simulationSpace=RootRelative
    alphaBoolBlend {blendOption=Linear}
    boundBone {boneName=dyn_clothBack_outer_06_r}
    componentAppliedLinearAccClamp=<3000 3000 3000>
    componentLinearAccScale=<1 1 1>
    componentLinearVelScale=<1 1 1>
    gravityOverride {z=-980}
    physicsBodyDefinitions [{
      sphereCollisionRadius=0
      boundBone {boneName=dyn_clothBack_outer_06_r}
      boxExtents=<26 42 42>
      constraintSetup {angularLimitsMax=<15 3 15>, angularLimitsMin=<-15 -15 -15>, angularTarget=<1 0 0>}
      localJointOffset=<4 0 0>
     }]
   }
   n48 = Anim Dynamics - Bone: dyn_clothBack_outer_01_l!(ComponentPose=@n47, Alpha=1, ExternalForce="0,0,0") -> (Pose) {
    angularDampingOverride=1
    angularSpringConstant=15000
    bAngularSpring=true
    bGravityOverrideInSimSpace=true
    bOverrideAngularDamping=true
    bUseGravityOverride=true
    gravityScale=0.35
    simulationSpace=RootRelative
    alphaBoolBlend {blendOption=Linear}
    boundBone {boneName=dyn_clothBack_outer_01_l}
    componentAppliedLinearAccClamp=<3000 3000 3000>
    componentLinearAccScale=<1 1 1>
    componentLinearVelScale=<1 1 1>
    gravityOverride {z=-980}
    physicsBodyDefinitions [{
      sphereCollisionRadius=0
      boundBone {boneName=dyn_clothBack_outer_01_l}
      boxExtents=<24 38 38>
      constraintSetup {angularLimitsMax=<15 3 15>, angularLimitsMin=<-15 -15 -15>, angularTarget=<1 0 0>}
      localJointOffset=<4 0 0>
     }]
   }
   n49 = Anim Dynamics - Bone: dyn_clothBack_outer_02_l(ComponentPose=@n48, Alpha=1, ExternalForce="0,0,0") {
    angularDampingOverride=1
    angularSpringConstant=15000
    bAngularSpring=true
    bGravityOverrideInSimSpace=true
    bOverrideAngularDamping=true
    bUseGravityOverride=true
    gravityScale=0.35
    simulationSpace=RootRelative
    alphaBoolBlend {blendOption=Linear}
    boundBone {boneName=dyn_clothBack_outer_02_l}
    componentAppliedLinearAccClamp=<3000 3000 3000>
    componentLinearAccScale=<1 1 1>
    componentLinearVelScale=<1 1 1>
    gravityOverride {z=-980}
    physicsBodyDefinitions [{
      sphereCollisionRadius=0
      boundBone {boneName=dyn_clothBack_outer_02_l}
      boxExtents=<24 38 38>
      constraintSetup {angularLimitsMax=<15 3 15>, angularLimitsMin=<-15 -15 -15>, angularTarget=<1 0 0>}
      localJointOffset=<4 0 0>
     }]
   }
   n50 = Anim Dynamics - Bone: dyn_clothBack_outer_03_l(ComponentPose=@n49, Alpha=1, ExternalForce="0,0,0") {
    angularDampingOverride=1
    angularSpringConstant=15000
    bAngularSpring=true
    bGravityOverrideInSimSpace=true
    bOverrideAngularDamping=true
    bUseGravityOverride=true
    gravityScale=0.35
    simulationSpace=RootRelative
    alphaBoolBlend {blendOption=Linear}
    boundBone {boneName=dyn_clothBack_outer_03_l}
    componentAppliedLinearAccClamp=<3000 3000 3000>
    componentLinearAccScale=<1 1 1>
    componentLinearVelScale=<1 1 1>
    gravityOverride {z=-980}
    physicsBodyDefinitions [{
      sphereCollisionRadius=0
      boundBone {boneName=dyn_clothBack_outer_03_l}
      boxExtents=<24 38 38>
      constraintSetup {angularLimitsMax=<15 3 15>, angularLimitsMin=<-15 -15 -15>, angularTarget=<1 0 0>}
      localJointOffset=<4 0 0>
     }]
   }
   n51 = Anim Dynamics - Bone: dyn_clothBack_outer_04_l(ComponentPose=@n50, Alpha=1, ExternalForce="0,0,0") {
    angularDampingOverride=1
    angularSpringConstant=15000
    bAngularSpring=true
    bGravityOverrideInSimSpace=true
    bOverrideAngularDamping=true
    bUseGravityOverride=true
    gravityScale=0.35
    simulationSpace=RootRelative
    alphaBoolBlend {blendOption=Linear}
    boundBone {boneName=dyn_clothBack_outer_04_l}
    componentAppliedLinearAccClamp=<3000 3000 3000>
    componentLinearAccScale=<1 1 1>
    componentLinearVelScale=<1 1 1>
    gravityOverride {z=-980}
    physicsBodyDefinitions [{
      sphereCollisionRadius=0
      boundBone {boneName=dyn_clothBack_outer_04_l}
      boxExtents=<24 38 38>
      constraintSetup {angularLimitsMax=<15 3 15>, angularLimitsMin=<-15 -15 -15>, angularTarget=<1 0 0>}
      localJointOffset=<4 0 0>
     }]
   }
   n52 = Anim Dynamics - Bone: dyn_clothBack_outer_05_l(ComponentPose=@n51, Alpha=1, ExternalForce="0,0,0") {
    angularDampingOverride=1
    angularSpringConstant=15000
    bAngularSpring=true
    bGravityOverrideInSimSpace=true
    bOverrideAngularDamping=true
    bUseGravityOverride=true
    gravityScale=0.35
    simulationSpace=RootRelative
    alphaBoolBlend {blendOption=Linear}
    boundBone {boneName=dyn_clothBack_outer_05_l}
    componentAppliedLinearAccClamp=<3000 3000 3000>
    componentLinearAccScale=<1 1 1>
    componentLinearVelScale=<1 1 1>
    gravityOverride {z=-980}
    physicsBodyDefinitions [{
      sphereCollisionRadius=0
      boundBone {boneName=dyn_clothBack_outer_05_l}
      boxExtents=<24 38 38>
      constraintSetup {angularLimitsMax=<15 3 15>, angularLimitsMin=<-15 -15 -15>, angularTarget=<1 0 0>}
      localJointOffset=<4 0 0>
     }]
   }
   n53 = Anim Dynamics - Bone: dyn_clothBack_outer_06_l(ComponentPose=@n52, Alpha=1, ExternalForce="0,0,0") {
    angularDampingOverride=1
    angularSpringConstant=15000
    bAngularSpring=true
    bGravityOverrideInSimSpace=true
    bOverrideAngularDamping=true
    bUseGravityOverride=true
    gravityScale=0.35
    simulationSpace=RootRelative
    alphaBoolBlend {blendOption=Linear}
    boundBone {boneName=dyn_clothBack_outer_06_l}
    componentAppliedLinearAccClamp=<3000 3000 3000>
    componentLinearAccScale=<1 1 1>
    componentLinearVelScale=<1 1 1>
    gravityOverride {z=-980}
    physicsBodyDefinitions [{
      sphereCollisionRadius=0
      boundBone {boneName=dyn_clothBack_outer_06_l}
      boxExtents=<24 38 38>
      constraintSetup {angularLimitsMax=<15 3 15>, angularLimitsMin=<-15 -15 -15>, angularTarget=<1 0 0>}
      localJointOffset=<4 0 0>
     }]
   }
   n54 = Anim Dynamics - Bone: dyn_clothFront_outer_01_l!(ComponentPose=@n53, Alpha=1, ExternalForce="0,0,0") -> (Pose) {
    angularDampingOverride=1
    angularSpringConstant=20000
    bAngularSpring=true
    bGravityOverrideInSimSpace=true
    bOverrideAngularDamping=true
    bUseGravityOverride=true
    gravityScale=0.35
    simulationSpace=RootRelative
    alphaBoolBlend {blendOption=Linear}
    boundBone {boneName=dyn_clothFront_outer_01_l}
    componentAppliedLinearAccClamp=<3000 3000 3000>
    componentLinearAccScale=<1 1 1>
    componentLinearVelScale=<1 1 1>
    gravityOverride {z=-980}
    physicsBodyDefinitions [{
      sphereCollisionRadius=0
      boundBone {boneName=dyn_clothFront_outer_01_l}
      boxExtents=<25 40 40>
      constraintSetup {angularLimitsMax=<15 15 -5>, angularLimitsMin=<-15 -15 -20>, angularTarget=<1 0 0>}
      localJointOffset=<4 0 0>
     }]
   }
   n55 = Anim Dynamics - Bone: dyn_clothFront_outer_02_l(ComponentPose=@n54, Alpha=1, ExternalForce="0,0,0") {
    angularDampingOverride=1
    angularSpringConstant=20000
    bAngularSpring=true
    bGravityOverrideInSimSpace=true
    bOverrideAngularDamping=true
    bUseGravityOverride=true
    gravityScale=0.35
    simulationSpace=RootRelative
    alphaBoolBlend {blendOption=Linear}
    boundBone {boneName=dyn_clothFront_outer_02_l}
    componentAppliedLinearAccClamp=<3000 3000 3000>
    componentLinearAccScale=<1 1 1>
    componentLinearVelScale=<1 1 1>
    gravityOverride {z=-980}
    physicsBodyDefinitions [{
      sphereCollisionRadius=0
      boundBone {boneName=dyn_clothFront_outer_02_l}
      boxExtents=<25 40 40>
      constraintSetup {angularLimitsMax=<15 15 2>, angularLimitsMin=<-15 -15 -15>, angularTarget=<1 0 0>}
      localJointOffset=<4 0 0>
     }]
   }
   n56 = Anim Dynamics - Bone: dyn_clothFront_outer_03_l(ComponentPose=@n55, Alpha=1, ExternalForce="0,0,0") {
    angularDampingOverride=1
    angularSpringConstant=20000
    bAngularSpring=true
    bGravityOverrideInSimSpace=true
    bOverrideAngularDamping=true
    bUseGravityOverride=true
    gravityScale=0.35
    simulationSpace=RootRelative
    alphaBoolBlend {blendOption=Linear}
    boundBone {boneName=dyn_clothFront_outer_03_l}
    componentAppliedLinearAccClamp=<3000 3000 3000>
    componentLinearAccScale=<1 1 1>
    componentLinearVelScale=<1 1 1>
    gravityOverride {z=-980}
    physicsBodyDefinitions [{
      sphereCollisionRadius=0
      boundBone {boneName=dyn_clothFront_outer_03_l}
      boxExtents=<25 40 40>
      constraintSetup {angularLimitsMax=<15 15 2>, angularLimitsMin=<-15 -15 -15>, angularTarget=<1 0 0>}
      localJointOffset=<4 0 0>
     }]
   }
   n57 = Anim Dynamics - Bone: dyn_clothFront_outer_04_l(ComponentPose=@n56, Alpha=1, ExternalForce="0,0,0") {
    angularDampingOverride=1
    angularSpringConstant=20000
    bAngularSpring=true
    bGravityOverrideInSimSpace=true
    bOverrideAngularDamping=true
    bUseGravityOverride=true
    gravityScale=0.35
    simulationSpace=RootRelative
    alphaBoolBlend {blendOption=Linear}
    boundBone {boneName=dyn_clothFront_outer_04_l}
    componentAppliedLinearAccClamp=<3000 3000 3000>
    componentLinearAccScale=<1 1 1>
    componentLinearVelScale=<1 1 1>
    gravityOverride {z=-980}
    physicsBodyDefinitions [{
      sphereCollisionRadius=0
      boundBone {boneName=dyn_clothFront_outer_04_l}
      boxExtents=<25 40 40>
      constraintSetup {angularLimitsMax=<15 15 2>, angularLimitsMin=<-15 -15 -15>, angularTarget=<1 0 0>}
      localJointOffset=<4 0 0>
     }]
   }
   n58 = Anim Dynamics - Bone: dyn_clothFront_outer_05_l(ComponentPose=@n57, Alpha=1, ExternalForce="0,0,0") {
    angularDampingOverride=1
    angularSpringConstant=20000
    bAngularSpring=true
    bGravityOverrideInSimSpace=true
    bOverrideAngularDamping=true
    bUseGravityOverride=true
    gravityScale=0.35
    simulationSpace=RootRelative
    alphaBoolBlend {blendOption=Linear}
    boundBone {boneName=dyn_clothFront_outer_05_l}
    componentAppliedLinearAccClamp=<3000 3000 3000>
    componentLinearAccScale=<1 1 1>
    componentLinearVelScale=<1 1 1>
    gravityOverride {z=-980}
    physicsBodyDefinitions [{
      sphereCollisionRadius=0
      boundBone {boneName=dyn_clothFront_outer_05_l}
      boxExtents=<25 40 40>
      constraintSetup {angularLimitsMax=<15 15 2>, angularLimitsMin=<-15 -15 -15>, angularTarget=<1 0 0>}
      localJointOffset=<4 0 0>
     }]
   }
   n59 = Anim Dynamics - Bone: dyn_clothFront_outer_06_l(ComponentPose=@n58, Alpha=1, ExternalForce="0,0,0") {
    angularDampingOverride=1
    angularSpringConstant=20000
    bAngularSpring=true
    bGravityOverrideInSimSpace=true
    bOverrideAngularDamping=true
    bUseGravityOverride=true
    gravityScale=0.35
    simulationSpace=RootRelative
    alphaBoolBlend {blendOption=Linear}
    boundBone {boneName=dyn_clothFront_outer_06_l}
    componentAppliedLinearAccClamp=<3000 3000 3000>
    componentLinearAccScale=<1 1 1>
    componentLinearVelScale=<1 1 1>
    gravityOverride {z=-980}
    physicsBodyDefinitions [{
      sphereCollisionRadius=0
      boundBone {boneName=dyn_clothFront_outer_06_l}
      boxExtents=<25 40 40>
      constraintSetup {angularLimitsMax=<15 15 2>, angularLimitsMin=<-15 -15 -15>, angularTarget=<1 0 0>}
      localJointOffset=<4 0 0>
     }]
   }
   n60 = Anim Dynamics - Bone: dyn_clothFront_inner_01_l!(ComponentPose=@n59, Alpha=1, ExternalForce="0,0,0") -> (Pose) {
    angularDampingOverride=1
    angularSpringConstant=20000
    bAngularSpring=true
    bGravityOverrideInSimSpace=true
    bOverrideAngularDamping=true
    bUseGravityOverride=true
    gravityScale=0.35
    simulationSpace=RootRelative
    alphaBoolBlend {blendOption=Linear}
    boundBone {boneName=dyn_clothFront_inner_01_l}
    componentAppliedLinearAccClamp=<3000 3000 3000>
    componentLinearAccScale=<1 1 1>
    componentLinearVelScale=<1 1 1>
    gravityOverride {z=-980}
    physicsBodyDefinitions [{
      sphereCollisionRadius=0
      boundBone {boneName=dyn_clothFront_inner_01_l}
      boxExtents=<25 40 40>
      constraintSetup {angularLimitsMax=<15 15 -5>, angularLimitsMin=<-15 -15 -20>, angularTarget=<1 0 0>}
      localJointOffset=<4 0 0>
     }]
   }
   n61 = Anim Dynamics - Bone: dyn_clothFront_inner_02_l(ComponentPose=@n60, Alpha=1, ExternalForce="0,0,0") {
    angularDampingOverride=1
    angularSpringConstant=20000
    bAngularSpring=true
    bGravityOverrideInSimSpace=true
    bOverrideAngularDamping=true
    bUseGravityOverride=true
    gravityScale=0.35
    simulationSpace=RootRelative
    alphaBoolBlend {blendOption=Linear}
    boundBone {boneName=dyn_clothFront_inner_02_l}
    componentAppliedLinearAccClamp=<3000 3000 3000>
    componentLinearAccScale=<1 1 1>
    componentLinearVelScale=<1 1 1>
    gravityOverride {z=-980}
    physicsBodyDefinitions [{
      sphereCollisionRadius=0
      boundBone {boneName=dyn_clothFront_inner_02_l}
      boxExtents=<25 40 40>
      constraintSetup {angularLimitsMax=<15 15 2>, angularLimitsMin=<-15 -15 -15>, angularTarget=<1 0 0>}
      localJointOffset=<4 0 0>
     }]
   }
   n62 = Anim Dynamics - Bone: dyn_clothFront_inner_03_l(ComponentPose=@n61, Alpha=1, ExternalForce="0,0,0") {
    angularDampingOverride=1
    angularSpringConstant=20000
    bAngularSpring=true
    bGravityOverrideInSimSpace=true
    bOverrideAngularDamping=true
    bUseGravityOverride=true
    gravityScale=0.35
    simulationSpace=RootRelative
    alphaBoolBlend {blendOption=Linear}
    boundBone {boneName=dyn_clothFront_inner_03_l}
    componentAppliedLinearAccClamp=<3000 3000 3000>
    componentLinearAccScale=<1 1 1>
    componentLinearVelScale=<1 1 1>
    gravityOverride {z=-980}
    physicsBodyDefinitions [{
      sphereCollisionRadius=0
      boundBone {boneName=dyn_clothFront_inner_03_l}
      boxExtents=<25 40 40>
      constraintSetup {angularLimitsMax=<15 15 2>, angularLimitsMin=<-15 -15 -15>, angularTarget=<1 0 0>}
      localJointOffset=<4 0 0>
     }]
   }
   n63 = Anim Dynamics - Bone: dyn_clothFront_inner_04_l(ComponentPose=@n62, Alpha=1, ExternalForce="0,0,0") {
    angularDampingOverride=1
    angularSpringConstant=20000
    bAngularSpring=true
    bGravityOverrideInSimSpace=true
    bOverrideAngularDamping=true
    bUseGravityOverride=true
    gravityScale=0.35
    simulationSpace=RootRelative
    alphaBoolBlend {blendOption=Linear}
    boundBone {boneName=dyn_clothFront_inner_04_l}
    componentAppliedLinearAccClamp=<3000 3000 3000>
    componentLinearAccScale=<1 1 1>
    componentLinearVelScale=<1 1 1>
    gravityOverride {z=-980}
    physicsBodyDefinitions [{
      sphereCollisionRadius=0
      boundBone {boneName=dyn_clothFront_inner_04_l}
      boxExtents=<25 40 40>
      constraintSetup {angularLimitsMax=<15 15 2>, angularLimitsMin=<-15 -15 -15>, angularTarget=<1 0 0>}
      localJointOffset=<4 0 0>
     }]
   }
   n64 = Anim Dynamics - Bone: dyn_clothFront_inner_05_l(ComponentPose=@n63, Alpha=1, ExternalForce="0,0,0") {
    angularDampingOverride=1
    angularSpringConstant=20000
    bAngularSpring=true
    bGravityOverrideInSimSpace=true
    bOverrideAngularDamping=true
    bUseGravityOverride=true
    gravityScale=0.35
    simulationSpace=RootRelative
    alphaBoolBlend {blendOption=Linear}
    boundBone {boneName=dyn_clothFront_inner_05_l}
    componentAppliedLinearAccClamp=<3000 3000 3000>
    componentLinearAccScale=<1 1 1>
    componentLinearVelScale=<1 1 1>
    gravityOverride {z=-980}
    physicsBodyDefinitions [{
      sphereCollisionRadius=0
      boundBone {boneName=dyn_clothFront_inner_05_l}
      boxExtents=<25 40 40>
      constraintSetup {angularLimitsMax=<15 15 2>, angularLimitsMin=<-15 -15 -15>, angularTarget=<1 0 0>}
      localJointOffset=<4 0 0>
     }]
   }
   n65 = Anim Dynamics - Bone: dyn_clothFront_outer_01_r!(ComponentPose=@n64, Alpha=1, ExternalForce="0,0,0") -> (Pose) {
    angularDampingOverride=1
    angularSpringConstant=20000
    bAngularSpring=true
    bGravityOverrideInSimSpace=true
    bOverrideAngularDamping=true
    bUseGravityOverride=true
    gravityScale=0.35
    simulationSpace=RootRelative
    alphaBoolBlend {blendOption=Linear}
    boundBone {boneName=dyn_clothFront_outer_01_r}
    componentAppliedLinearAccClamp=<3000 3000 3000>
    componentLinearAccScale=<1 1 1>
    componentLinearVelScale=<1 1 1>
    gravityOverride {z=-980}
    physicsBodyDefinitions [{
      sphereCollisionRadius=0
      boundBone {boneName=dyn_clothFront_outer_01_r}
      boxExtents=<25 40 40>
      constraintSetup {angularLimitsMax=<15 15 20>, angularLimitsMin=<-15 -15 5>, angularTarget=<1 0 0>}
      localJointOffset=<4 0 0>
     }]
   }
   n66 = Anim Dynamics - Bone: dyn_clothFront_outer_02_r(ComponentPose=@n65, Alpha=1, ExternalForce="0,0,0") {
    angularDampingOverride=1
    angularSpringConstant=20000
    bAngularSpring=true
    bGravityOverrideInSimSpace=true
    bOverrideAngularDamping=true
    bUseGravityOverride=true
    gravityScale=0.35
    simulationSpace=RootRelative
    alphaBoolBlend {blendOption=Linear}
    boundBone {boneName=dyn_clothFront_outer_02_r}
    componentAppliedLinearAccClamp=<3000 3000 3000>
    componentLinearAccScale=<1 1 1>
    componentLinearVelScale=<1 1 1>
    gravityOverride {z=-980}
    physicsBodyDefinitions [{
      sphereCollisionRadius=0
      boundBone {boneName=dyn_clothFront_outer_02_r}
      boxExtents=<25 40 40>
      constraintSetup {angularLimitsMax=<15 15 15>, angularLimitsMin=<-15 -15 -2>, angularTarget=<1 0 0>}
      localJointOffset=<4 0 0>
     }]
   }
   n67 = Anim Dynamics - Bone: dyn_clothFront_outer_03_r(ComponentPose=@n66, Alpha=1, ExternalForce="0,0,0") {
    angularDampingOverride=1
    angularSpringConstant=20000
    bAngularSpring=true
    bGravityOverrideInSimSpace=true
    bOverrideAngularDamping=true
    bUseGravityOverride=true
    gravityScale=0.35
    simulationSpace=RootRelative
    alphaBoolBlend {blendOption=Linear}
    boundBone {boneName=dyn_clothFront_outer_03_r}
    componentAppliedLinearAccClamp=<3000 3000 3000>
    componentLinearAccScale=<1 1 1>
    componentLinearVelScale=<1 1 1>
    gravityOverride {z=-980}
    physicsBodyDefinitions [{
      sphereCollisionRadius=0
      boundBone {boneName=dyn_clothFront_outer_03_r}
      boxExtents=<25 40 40>
      constraintSetup {angularLimitsMax=<15 15 15>, angularLimitsMin=<-15 -15 -2>, angularTarget=<1 0 0>}
      localJointOffset=<4 0 0>
     }]
   }
   n68 = Anim Dynamics - Bone: dyn_clothFront_outer_04_r(ComponentPose=@n67, Alpha=1, ExternalForce="0,0,0") {
    angularDampingOverride=1
    angularSpringConstant=20000
    bAngularSpring=true
    bGravityOverrideInSimSpace=true
    bOverrideAngularDamping=true
    bUseGravityOverride=true
    gravityScale=0.35
    simulationSpace=RootRelative
    alphaBoolBlend {blendOption=Linear}
    boundBone {boneName=dyn_clothFront_outer_04_r}
    componentAppliedLinearAccClamp=<3000 3000 3000>
    componentLinearAccScale=<1 1 1>
    componentLinearVelScale=<1 1 1>
    gravityOverride {z=-980}
    physicsBodyDefinitions [{
      sphereCollisionRadius=0
      boundBone {boneName=dyn_clothFront_outer_04_r}
      boxExtents=<25 40 40>
      constraintSetup {angularLimitsMax=<15 15 15>, angularLimitsMin=<-15 -15 -2>, angularTarget=<1 0 0>}
      localJointOffset=<4 0 0>
     }]
   }
   n69 = Anim Dynamics - Bone: dyn_clothFront_outer_05_r(ComponentPose=@n68, Alpha=1, ExternalForce="0,0,0") {
    angularDampingOverride=1
    angularSpringConstant=20000
    bAngularSpring=true
    bGravityOverrideInSimSpace=true
    bOverrideAngularDamping=true
    bUseGravityOverride=true
    gravityScale=0.35
    simulationSpace=RootRelative
    alphaBoolBlend {blendOption=Linear}
    boundBone {boneName=dyn_clothFront_outer_05_r}
    componentAppliedLinearAccClamp=<3000 3000 3000>
    componentLinearAccScale=<1 1 1>
    componentLinearVelScale=<1 1 1>
    gravityOverride {z=-980}
    physicsBodyDefinitions [{
      sphereCollisionRadius=0
      boundBone {boneName=dyn_clothFront_outer_05_r}
      boxExtents=<25 40 40>
      constraintSetup {angularLimitsMax=<15 15 15>, angularLimitsMin=<-15 -15 -2>, angularTarget=<1 0 0>}
      localJointOffset=<4 0 0>
     }]
   }
   n70 = Anim Dynamics - Bone: dyn_clothFront_outer_06_r(ComponentPose=@n69, Alpha=1, ExternalForce="0,0,0") {
    angularDampingOverride=1
    angularSpringConstant=20000
    bAngularSpring=true
    bGravityOverrideInSimSpace=true
    bOverrideAngularDamping=true
    bUseGravityOverride=true
    gravityScale=0.35
    simulationSpace=RootRelative
    alphaBoolBlend {blendOption=Linear}
    boundBone {boneName=dyn_clothFront_outer_06_r}
    componentAppliedLinearAccClamp=<3000 3000 3000>
    componentLinearAccScale=<1 1 1>
    componentLinearVelScale=<1 1 1>
    gravityOverride {z=-980}
    physicsBodyDefinitions [{
      sphereCollisionRadius=0
      boundBone {boneName=dyn_clothFront_outer_06_r}
      boxExtents=<25 40 40>
      constraintSetup {angularLimitsMax=<15 15 15>, angularLimitsMin=<-15 -15 -2>, angularTarget=<1 0 0>}
      localJointOffset=<4 0 0>
     }]
   }
   n71 = Anim Dynamics - Bone: dyn_clothFront_inner_01_r(ComponentPose=@n70, Alpha=1, ExternalForce="0,0,0") {
    angularDampingOverride=1
    angularSpringConstant=20000
    bAngularSpring=true
    bGravityOverrideInSimSpace=true
    bOverrideAngularDamping=true
    bUseGravityOverride=true
    gravityScale=0.35
    simulationSpace=RootRelative
    alphaBoolBlend {blendOption=Linear}
    boundBone {boneName=dyn_clothFront_inner_01_r}
    componentAppliedLinearAccClamp=<3000 3000 3000>
    componentLinearAccScale=<1 1 1>
    componentLinearVelScale=<1 1 1>
    gravityOverride {z=-980}
    physicsBodyDefinitions [{
      sphereCollisionRadius=0
      boundBone {boneName=dyn_clothFront_inner_01_r}
      boxExtents=<25 40 40>
      constraintSetup {angularLimitsMax=<15 15 20>, angularLimitsMin=<-15 -15 5>, angularTarget=<1 0 0>}
      localJointOffset=<4 0 0>
     }]
   }
   n72 = Anim Dynamics - Bone: dyn_clothFront_inner_02_r(ComponentPose=@n71, Alpha=1, ExternalForce="0,0,0") {
    angularDampingOverride=1
    angularSpringConstant=20000
    bAngularSpring=true
    bGravityOverrideInSimSpace=true
    bOverrideAngularDamping=true
    bUseGravityOverride=true
    gravityScale=0.35
    simulationSpace=RootRelative
    alphaBoolBlend {blendOption=Linear}
    boundBone {boneName=dyn_clothFront_inner_02_r}
    componentAppliedLinearAccClamp=<3000 3000 3000>
    componentLinearAccScale=<1 1 1>
    componentLinearVelScale=<1 1 1>
    gravityOverride {z=-980}
    physicsBodyDefinitions [{
      sphereCollisionRadius=0
      boundBone {boneName=dyn_clothFront_inner_02_r}
      boxExtents=<25 40 40>
      constraintSetup {angularLimitsMax=<15 15 15>, angularLimitsMin=<-15 -15 -2>, angularTarget=<1 0 0>}
      localJointOffset=<4 0 0>
     }]
   }
   n73 = Anim Dynamics - Bone: dyn_clothFront_inner_03_r(ComponentPose=@n72, Alpha=1, ExternalForce="0,0,0") {
    angularDampingOverride=1
    angularSpringConstant=20000
    bAngularSpring=true
    bGravityOverrideInSimSpace=true
    bOverrideAngularDamping=true
    bUseGravityOverride=true
    gravityScale=0.35
    simulationSpace=RootRelative
    alphaBoolBlend {blendOption=Linear}
    boundBone {boneName=dyn_clothFront_inner_03_r}
    componentAppliedLinearAccClamp=<3000 3000 3000>
    componentLinearAccScale=<1 1 1>
    componentLinearVelScale=<1 1 1>
    gravityOverride {z=-980}
    physicsBodyDefinitions [{
      sphereCollisionRadius=0
      boundBone {boneName=dyn_clothFront_inner_03_r}
      boxExtents=<25 40 40>
      constraintSetup {angularLimitsMax=<15 15 15>, angularLimitsMin=<-15 -15 -2>, angularTarget=<1 0 0>}
      localJointOffset=<4 0 0>
     }]
   }
   n74 = Anim Dynamics - Bone: dyn_clothFront_inner_04_r(ComponentPose=@n73, Alpha=1, ExternalForce="0,0,0") {
    angularDampingOverride=1
    angularSpringConstant=20000
    bAngularSpring=true
    bGravityOverrideInSimSpace=true
    bOverrideAngularDamping=true
    bUseGravityOverride=true
    gravityScale=0.35
    simulationSpace=RootRelative
    alphaBoolBlend {blendOption=Linear}
    boundBone {boneName=dyn_clothFront_inner_04_r}
    componentAppliedLinearAccClamp=<3000 3000 3000>
    componentLinearAccScale=<1 1 1>
    componentLinearVelScale=<1 1 1>
    gravityOverride {z=-980}
    physicsBodyDefinitions [{
      sphereCollisionRadius=0
      boundBone {boneName=dyn_clothFront_inner_04_r}
      boxExtents=<25 40 40>
      constraintSetup {angularLimitsMax=<15 15 15>, angularLimitsMin=<-15 -15 -2>, angularTarget=<1 0 0>}
      localJointOffset=<4 0 0>
     }]
   }
   n75 = Anim Dynamics - Bone: dyn_clothFront_inner_05_r(ComponentPose=@n74, Alpha=1, ExternalForce="0,0,0") {
    angularDampingOverride=1
    angularSpringConstant=20000
    bAngularSpring=true
    bGravityOverrideInSimSpace=true
    bOverrideAngularDamping=true
    bUseGravityOverride=true
    gravityScale=0.35
    simulationSpace=RootRelative
    alphaBoolBlend {blendOption=Linear}
    boundBone {boneName=dyn_clothFront_inner_05_r}
    componentAppliedLinearAccClamp=<3000 3000 3000>
    componentLinearAccScale=<1 1 1>
    componentLinearVelScale=<1 1 1>
    gravityOverride {z=-980}
    physicsBodyDefinitions [{
      sphereCollisionRadius=0
      boundBone {boneName=dyn_clothFront_inner_05_r}
      boxExtents=<25 40 40>
      constraintSetup {angularLimitsMax=<15 15 15>, angularLimitsMin=<-15 -15 -2>, angularTarget=<1 0 0>}
      localJointOffset=<4 0 0>
     }]
   }
   n76 = Anim Dynamics - Bone: dyn_clothSkirt_outer_01_r(ComponentPose=@n75, Alpha=1, ExternalForce="0,0,0") {
    angularDampingOverride=1
    angularSpringConstant=20000
    bAngularSpring=true
    bGravityOverrideInSimSpace=true
    bOverrideAngularDamping=true
    bUseGravityOverride=true
    gravityScale=0.35
    simulationSpace=RootRelative
    alphaBoolBlend {blendOption=Linear}
    boundBone {boneName=dyn_clothSkirt_outer_01_r}
    componentAppliedLinearAccClamp=<3000 3000 3000>
    componentLinearAccScale=<1 1 1>
    componentLinearVelScale=<1 1 1>
    gravityOverride {z=-980}
    physicsBodyDefinitions [{
      sphereCollisionRadius=0
      boundBone {boneName=dyn_clothSkirt_outer_01_r}
      boxExtents=<30 40 40>
      constraintSetup {angularLimitsMax=<15 5 15>, angularLimitsMin=<-15 -15 -15>, angularTarget=<1 0 0>}
      localJointOffset=<4 0 0>
     }]
   }
   n77 = Anim Dynamics - Bone: dyn_clothSkirt_outer_02_r(ComponentPose=@n76, Alpha=1, ExternalForce="0,0,0") {
    angularDampingOverride=1
    angularSpringConstant=20000
    bAngularSpring=true
    bGravityOverrideInSimSpace=true
    bOverrideAngularDamping=true
    bUseGravityOverride=true
    gravityScale=0.35
    simulationSpace=RootRelative
    alphaBoolBlend {blendOption=Linear}
    boundBone {boneName=dyn_clothSkirt_outer_02_r}
    componentAppliedLinearAccClamp=<3000 3000 3000>
    componentLinearAccScale=<1 1 1>
    componentLinearVelScale=<1 1 1>
    gravityOverride {z=-980}
    physicsBodyDefinitions [{
      sphereCollisionRadius=0
      boundBone {boneName=dyn_clothSkirt_outer_02_r}
      boxExtents=<30 40 40>
      constraintSetup {angularLimitsMax=<15 5 15>, angularLimitsMin=<-15 -15 -15>, angularTarget=<1 0 0>}
      localJointOffset=<4 0 0>
     }]
   }
   n78 = Anim Dynamics - Bone: dyn_clothSkirt_outer_03_r(ComponentPose=@n77, Alpha=1, ExternalForce="0,0,0") {
    angularDampingOverride=1
    angularSpringConstant=20000
    bAngularSpring=true
    bGravityOverrideInSimSpace=true
    bOverrideAngularDamping=true
    bUseGravityOverride=true
    gravityScale=0.35
    simulationSpace=RootRelative
    alphaBoolBlend {blendOption=Linear}
    boundBone {boneName=dyn_clothSkirt_outer_03_r}
    componentAppliedLinearAccClamp=<3000 3000 3000>
    componentLinearAccScale=<1 1 1>
    componentLinearVelScale=<1 1 1>
    gravityOverride {z=-980}
    physicsBodyDefinitions [{
      sphereCollisionRadius=0
      boundBone {boneName=dyn_clothSkirt_outer_03_r}
      boxExtents=<30 40 40>
      constraintSetup {angularLimitsMax=<15 5 15>, angularLimitsMin=<-15 -15 -15>, angularTarget=<1 0 0>}
      localJointOffset=<4 0 0>
     }]
   }
   n79 = Anim Dynamics - Bone: dyn_clothSkirt_outer_04_r(ComponentPose=@n78, Alpha=1, ExternalForce="0,0,0") {
    angularDampingOverride=1
    angularSpringConstant=20000
    bAngularSpring=true
    bGravityOverrideInSimSpace=true
    bOverrideAngularDamping=true
    bUseGravityOverride=true
    gravityScale=0.35
    simulationSpace=RootRelative
    alphaBoolBlend {blendOption=Linear}
    boundBone {boneName=dyn_clothSkirt_outer_04_r}
    componentAppliedLinearAccClamp=<3000 3000 3000>
    componentLinearAccScale=<1 1 1>
    componentLinearVelScale=<1 1 1>
    gravityOverride {z=-980}
    physicsBodyDefinitions [{
      sphereCollisionRadius=0
      boundBone {boneName=dyn_clothSkirt_outer_04_r}
      boxExtents=<30 40 40>
      constraintSetup {angularLimitsMax=<15 5 15>, angularLimitsMin=<-15 -15 -15>, angularTarget=<1 0 0>}
      localJointOffset=<4 0 0>
     }]
   }
   n80 = Anim Dynamics - Bone: dyn_clothSkirt_inner_01_r!(ComponentPose=@n79, Alpha=1, ExternalForce="0,0,0") -> (Pose) {
    angularDampingOverride=1
    angularSpringConstant=20000
    bAngularSpring=true
    bGravityOverrideInSimSpace=true
    bOverrideAngularDamping=true
    bUseGravityOverride=true
    gravityScale=0.35
    simulationSpace=RootRelative
    alphaBoolBlend {blendOption=Linear}
    boundBone {boneName=dyn_clothSkirt_inner_01_r}
    componentAppliedLinearAccClamp=<3000 3000 3000>
    componentLinearAccScale=<1 1 1>
    componentLinearVelScale=<1 1 1>
    gravityOverride {z=-980}
    physicsBodyDefinitions [{
      sphereCollisionRadius=0
      boundBone {boneName=dyn_clothSkirt_inner_01_r}
      boxExtents=<30 40 40>
      constraintSetup {angularLimitsMax=<15 5 15>, angularLimitsMin=<-15 -15 -15>, angularTarget=<1 0 0>}
      localJointOffset=<4 0 0>
     }]
   }
   n81 = Anim Dynamics - Bone: dyn_clothSkirt_inner_02_r(ComponentPose=@n80, Alpha=1, ExternalForce="0,0,0") {
    angularDampingOverride=1
    angularSpringConstant=20000
    bAngularSpring=true
    bGravityOverrideInSimSpace=true
    bOverrideAngularDamping=true
    bUseGravityOverride=true
    gravityScale=0.35
    simulationSpace=RootRelative
    alphaBoolBlend {blendOption=Linear}
    boundBone {boneName=dyn_clothSkirt_inner_02_r}
    componentAppliedLinearAccClamp=<3000 3000 3000>
    componentLinearAccScale=<1 1 1>
    componentLinearVelScale=<1 1 1>
    gravityOverride {z=-980}
    physicsBodyDefinitions [{
      sphereCollisionRadius=0
      boundBone {boneName=dyn_clothSkirt_inner_02_r}
      boxExtents=<30 40 40>
      constraintSetup {angularLimitsMax=<15 5 15>, angularLimitsMin=<-15 -15 -15>, angularTarget=<1 0 0>}
      localJointOffset=<4 0 0>
     }]
   }
   n82 = Anim Dynamics - Bone: dyn_clothSkirt_inner_03_r(ComponentPose=@n81, Alpha=1, ExternalForce="0,0,0") {
    angularDampingOverride=1
    angularSpringConstant=20000
    bAngularSpring=true
    bGravityOverrideInSimSpace=true
    bOverrideAngularDamping=true
    bUseGravityOverride=true
    gravityScale=0.35
    simulationSpace=RootRelative
    alphaBoolBlend {blendOption=Linear}
    boundBone {boneName=dyn_clothSkirt_inner_03_r}
    componentAppliedLinearAccClamp=<3000 3000 3000>
    componentLinearAccScale=<1 1 1>
    componentLinearVelScale=<1 1 1>
    gravityOverride {z=-980}
    physicsBodyDefinitions [{
      sphereCollisionRadius=0
      boundBone {boneName=dyn_clothSkirt_inner_03_r}
      boxExtents=<30 40 40>
      constraintSetup {angularLimitsMax=<15 5 15>, angularLimitsMin=<-15 -15 -15>, angularTarget=<1 0 0>}
      localJointOffset=<4 0 0>
     }]
   }
   n83 = Anim Dynamics - Bone: dyn_clothSkirt_inner_04_r(ComponentPose=@n82, Alpha=1, ExternalForce="0,0,0") {
    angularDampingOverride=1
    angularSpringConstant=20000
    bAngularSpring=true
    bGravityOverrideInSimSpace=true
    bOverrideAngularDamping=true
    bUseGravityOverride=true
    gravityScale=0.35
    simulationSpace=RootRelative
    alphaBoolBlend {blendOption=Linear}
    boundBone {boneName=dyn_clothSkirt_inner_04_r}
    componentAppliedLinearAccClamp=<3000 3000 3000>
    componentLinearAccScale=<1 1 1>
    componentLinearVelScale=<1 1 1>
    gravityOverride {z=-980}
    physicsBodyDefinitions [{
      sphereCollisionRadius=0
      boundBone {boneName=dyn_clothSkirt_inner_04_r}
      boxExtents=<30 40 40>
      constraintSetup {angularLimitsMax=<15 5 15>, angularLimitsMin=<-15 -15 -15>, angularTarget=<1 0 0>}
      localJointOffset=<4 0 0>
     }]
   }
   n84 = Anim Dynamics - Bone: dyn_clothSkirt_inner_01_l!(ComponentPose=@n83, Alpha=1, ExternalForce="0,0,0") -> (Pose) {
    angularDampingOverride=1
    angularSpringConstant=20000
    bAngularSpring=true
    bGravityOverrideInSimSpace=true
    bOverrideAngularDamping=true
    bUseGravityOverride=true
    gravityScale=0.35
    simulationSpace=RootRelative
    alphaBoolBlend {blendOption=Linear}
    boundBone {boneName=dyn_clothSkirt_inner_01_l}
    componentAppliedLinearAccClamp=<3000 3000 3000>
    componentLinearAccScale=<1 1 1>
    componentLinearVelScale=<1 1 1>
    gravityOverride {z=-980}
    physicsBodyDefinitions [{
      sphereCollisionRadius=0
      boundBone {boneName=dyn_clothSkirt_inner_01_l}
      boxExtents=<30 40 40>
      constraintSetup {angularLimitsMax=<15 5 15>, angularLimitsMin=<-15 -15 -15>, angularTarget=<1 0 0>}
      localJointOffset=<4 0 0>
     }]
   }
   n85 = Anim Dynamics - Bone: dyn_clothSkirt_inner_02_l(ComponentPose=@n84, Alpha=1, ExternalForce="0,0,0") {
    angularDampingOverride=1
    angularSpringConstant=20000
    bAngularSpring=true
    bGravityOverrideInSimSpace=true
    bOverrideAngularDamping=true
    bUseGravityOverride=true
    gravityScale=0.35
    simulationSpace=RootRelative
    alphaBoolBlend {blendOption=Linear}
    boundBone {boneName=dyn_clothSkirt_inner_02_l}
    componentAppliedLinearAccClamp=<3000 3000 3000>
    componentLinearAccScale=<1 1 1>
    componentLinearVelScale=<1 1 1>
    gravityOverride {z=-980}
    physicsBodyDefinitions [{
      sphereCollisionRadius=0
      boundBone {boneName=dyn_clothSkirt_inner_02_l}
      boxExtents=<30 40 40>
      constraintSetup {angularLimitsMax=<15 5 15>, angularLimitsMin=<-15 -15 -15>, angularTarget=<1 0 0>}
      localJointOffset=<4 0 0>
     }]
   }
   n86 = Anim Dynamics - Bone: dyn_clothSkirt_inner_03_l(ComponentPose=@n85, Alpha=1, ExternalForce="0,0,0") {
    angularDampingOverride=1
    angularSpringConstant=20000
    bAngularSpring=true
    bGravityOverrideInSimSpace=true
    bOverrideAngularDamping=true
    bUseGravityOverride=true
    gravityScale=0.35
    simulationSpace=RootRelative
    alphaBoolBlend {blendOption=Linear}
    boundBone {boneName=dyn_clothSkirt_inner_03_l}
    componentAppliedLinearAccClamp=<3000 3000 3000>
    componentLinearAccScale=<1 1 1>
    componentLinearVelScale=<1 1 1>
    gravityOverride {z=-980}
    physicsBodyDefinitions [{
      sphereCollisionRadius=0
      boundBone {boneName=dyn_clothSkirt_inner_03_l}
      boxExtents=<30 40 40>
      constraintSetup {angularLimitsMax=<15 5 15>, angularLimitsMin=<-15 -15 -15>, angularTarget=<1 0 0>}
      localJointOffset=<4 0 0>
     }]
   }
   n87 = Anim Dynamics - Bone: dyn_clothSkirt_inner_04_l(ComponentPose=@n86, Alpha=1, ExternalForce="0,0,0") {
    angularDampingOverride=1
    angularSpringConstant=20000
    bAngularSpring=true
    bGravityOverrideInSimSpace=true
    bOverrideAngularDamping=true
    bUseGravityOverride=true
    gravityScale=0.35
    simulationSpace=RootRelative
    alphaBoolBlend {blendOption=Linear}
    boundBone {boneName=dyn_clothSkirt_inner_04_l}
    componentAppliedLinearAccClamp=<3000 3000 3000>
    componentLinearAccScale=<1 1 1>
    componentLinearVelScale=<1 1 1>
    gravityOverride {z=-980}
    physicsBodyDefinitions [{
      sphereCollisionRadius=0
      boundBone {boneName=dyn_clothSkirt_inner_04_l}
      boxExtents=<30 40 40>
      constraintSetup {angularLimitsMax=<15 5 15>, angularLimitsMin=<-15 -15 -15>, angularTarget=<1 0 0>}
      localJointOffset=<4 0 0>
     }]
   }
   n88 = Anim Dynamics - Bone: dyn_clothSkirt_outer_01_l!(ComponentPose=@n87, Alpha=1, ExternalForce="0,0,0") -> (Pose) {
    angularDampingOverride=1
    angularSpringConstant=20000
    bAngularSpring=true
    bGravityOverrideInSimSpace=true
    bOverrideAngularDamping=true
    bUseGravityOverride=true
    gravityScale=0.35
    simulationSpace=RootRelative
    alphaBoolBlend {blendOption=Linear}
    boundBone {boneName=dyn_clothSkirt_outer_01_l}
    componentAppliedLinearAccClamp=<3000 3000 3000>
    componentLinearAccScale=<1 1 1>
    componentLinearVelScale=<1 1 1>
    gravityOverride {z=-980}
    physicsBodyDefinitions [{
      sphereCollisionRadius=0
      boundBone {boneName=dyn_clothSkirt_outer_01_l}
      boxExtents=<30 40 40>
      constraintSetup {angularLimitsMax=<15 5 15>, angularLimitsMin=<-15 -15 -15>, angularTarget=<1 0 0>}
      localJointOffset=<4 0 0>
     }]
   }
   n89 = Anim Dynamics - Bone: dyn_clothSkirt_outer_02_l(ComponentPose=@n88, Alpha=1, ExternalForce="0,0,0") {
    angularDampingOverride=1
    angularSpringConstant=20000
    bAngularSpring=true
    bGravityOverrideInSimSpace=true
    bOverrideAngularDamping=true
    bUseGravityOverride=true
    gravityScale=0.35
    simulationSpace=RootRelative
    alphaBoolBlend {blendOption=Linear}
    boundBone {boneName=dyn_clothSkirt_outer_02_l}
    componentAppliedLinearAccClamp=<3000 3000 3000>
    componentLinearAccScale=<1 1 1>
    componentLinearVelScale=<1 1 1>
    gravityOverride {z=-980}
    physicsBodyDefinitions [{
      sphereCollisionRadius=0
      boundBone {boneName=dyn_clothSkirt_outer_02_l}
      boxExtents=<30 40 40>
      constraintSetup {angularLimitsMax=<15 5 15>, angularLimitsMin=<-15 -15 -15>, angularTarget=<1 0 0>}
      localJointOffset=<4 0 0>
     }]
   }
   n90 = Anim Dynamics - Bone: dyn_clothSkirt_outer_03_l(ComponentPose=@n89, Alpha=1, ExternalForce="0,0,0") {
    angularDampingOverride=1
    angularSpringConstant=20000
    bAngularSpring=true
    bGravityOverrideInSimSpace=true
    bOverrideAngularDamping=true
    bUseGravityOverride=true
    gravityScale=0.35
    simulationSpace=RootRelative
    alphaBoolBlend {blendOption=Linear}
    boundBone {boneName=dyn_clothSkirt_outer_03_l}
    componentAppliedLinearAccClamp=<3000 3000 3000>
    componentLinearAccScale=<1 1 1>
    componentLinearVelScale=<1 1 1>
    gravityOverride {z=-980}
    physicsBodyDefinitions [{
      sphereCollisionRadius=0
      boundBone {boneName=dyn_clothSkirt_outer_03_l}
      boxExtents=<30 40 40>
      constraintSetup {angularLimitsMax=<15 5 15>, angularLimitsMin=<-15 -15 -15>, angularTarget=<1 0 0>}
      localJointOffset=<4 0 0>
     }]
   }
   n91 = Anim Dynamics - Bone: dyn_clothSkirt_outer_04_l(ComponentPose=@n90, Alpha=1, ExternalForce="0,0,0") {
    angularDampingOverride=1
    angularSpringConstant=20000
    bAngularSpring=true
    bGravityOverrideInSimSpace=true
    bOverrideAngularDamping=true
    bUseGravityOverride=true
    gravityScale=0.35
    simulationSpace=RootRelative
    alphaBoolBlend {blendOption=Linear}
    boundBone {boneName=dyn_clothSkirt_outer_04_l}
    componentAppliedLinearAccClamp=<3000 3000 3000>
    componentLinearAccScale=<1 1 1>
    componentLinearVelScale=<1 1 1>
    gravityOverride {z=-980}
    physicsBodyDefinitions [{
      sphereCollisionRadius=0
      boundBone {boneName=dyn_clothSkirt_outer_04_l}
      boxExtents=<30 40 40>
      constraintSetup {angularLimitsMax=<15 5 15>, angularLimitsMin=<-15 -15 -15>, angularTarget=<1 0 0>}
      localJointOffset=<4 0 0>
     }]
   }
   n92 = Layered blend per bone(
    BasePose=Use cached pose 'preSimCache'()
    BlendPoses_0=Sequence Player 'Downpour_Male_Idle_Flaps'() {sequence=/BRCosmetics/Animation/Game/MainPlayer/Menu/BR/CustomSkins/_S16/Downpour/Downpour_Male_Idle_Flaps}
   ) {layerSetup [{branchFilters [
       {boneName=dyn_clothFront_inner_mstr_r}
       {boneName=dyn_clothFront_outer_mstr_r}
       {boneName=dyn_clothBack_outer_mstr_r}
       {boneName=capeBase}
       {boneName=dyn_clothBack_outer_mstr_l}
       {boneName=dyn_clothFront_outer_mstr_l}
       {boneName=dyn_clothFront_inner_mstr_l}
       {boneName=dyn_clothSkirt_outer_mstr_l}
       {boneName=dyn_clothSkirt_inner_mstr_l}
       {boneName=dyn_clothSkirt_inner_mstr_r}
       {boneName=dyn_clothSkirt_outer_mstr_r}
      ]}]}
   n93 = Control Rig(Source=@n92) {
    alphaBoolBlend {blendOption=Linear}
    controlRigAssetReference {blueprintRigClass=&/M_MED_Downpour_CtrlRig.M_MED_Downpour_CtrlRig_C}
   }
   n94 = Blend Poses by bool(
    BlendPose_0=@n93
    BlendPose_1=Blend Poses by bool(BlendPose_0=Component To Local(ComponentPose=@n17), BlendPose_1=Component To Local(ComponentPose=@n91), bActiveValue=Get bRigidBodyEnabled()) {blendType=Linear}
    bActiveValue=Get isPlayingFrontendIdle() -> (isPlayingFrontendIdle:bool)
   ) {blendType=Linear}
   n95 = "Transform (Modify) Bone - Bone: capeBase"(
    ComponentPose=Local To Component(LocalPose=Use cached pose 'preSimCache'())
    Alpha=Get Is Jumping Float() -> (Is Jumping Float:double)
    Translation="0,0,0"
    Rotation="0,0,10"
    Scale="1,1,1"
   ) {
    rotationMode=BMM_Additive
    rotationSpace=BCS_BoneSpace
    translationMode=BMM_Additive
    translationSpace=BCS_BoneSpace
    alphaBoolBlend {blendOption=Linear}
    boneToModify {boneName=capeBase}
    rotation {roll=10}
   }
   n96 = Get Is Skydiving Float() -> (Is Skydiving Float:double)
   n97 = "Transform (Modify) Bone - Bone: capeBase"(ComponentPose=@n95, Alpha=@n96, Translation="0,0,0", Rotation="0,0,16", Scale="1,1,1") {
    rotationMode=BMM_Additive
    rotationSpace=BCS_BoneSpace
    translationMode=BMM_Additive
    translationSpace=BCS_BoneSpace
    alphaBoolBlend {blendOption=Linear}
    boneToModify {boneName=capeBase}
    rotation {roll=16}
   }
   n98 = "Transform (Modify) Bone - Bone: dyn_clothBack_outer_mstr_l"(ComponentPose=@n97, Alpha=@n96, Translation="0,0,0", Rotation="15,15,0", Scale="1,1,1") {
    rotationMode=BMM_Additive
    rotationSpace=BCS_BoneSpace
    translationMode=BMM_Additive
    translationSpace=BCS_BoneSpace
    alphaBoolBlend {blendOption=Linear}
    boneToModify {boneName=dyn_clothBack_outer_mstr_l}
    rotation {pitch=15, yaw=15}
   }
   n99 = "Transform (Modify) Bone - Bone: dyn_clothBack_outer_mstr_r"!(ComponentPose=@n98, Alpha:float=@n96, Translation="0,0,0", Rotation="15,-15,0", Scale="1,1,1") -> (Pose) {
    rotationMode=BMM_Additive
    rotationSpace=BCS_BoneSpace
    translationMode=BMM_Additive
    translationSpace=BCS_BoneSpace
    alphaBoolBlend {blendOption=Linear}
    boneToModify {boneName=dyn_clothBack_outer_mstr_r}
    rotation {pitch=15, yaw=-15}
   }
   n100 = Get Is Not Skydiving Dive Mode Float() -> (Is Not Skydiving Dive Mode Float:double)
   n101 = Trail controller - Bone: dyn_clothBack_inner_06_l(ComponentPose=@n99, Alpha=@n100) {
    bUsePlanarLimit=true
    chainLength=5
    debugLifeTime=0.1
    maxDeltaTime=0.333
    alphaScaleBiasClamp {bInterpResult=true}
    baseJoint {boneName=capeBase}
    planarLimits [{drivingBone {boneName=dyn_clothBack_inner_01_l}, planeTransform {location=<0 5 0>, rotation=<rot 0 0 0>, scale=<1 1 1>}}]
    rotationLimits [{limitMax=<180 180 180>}, {limitMax=<180 180 180>}, {limitMax=<180 180 180>}, {limitMax=<180 180 180>}, {limitMax=<180 180 180>}]
    rotationOffsets [<0 0 0>, <0 0 0>, <0 0 0>, <0 0 0>, <0 0 0>]
    trailBone {boneName=dyn_clothBack_inner_06_l}
    trailRelaxationSpeed {editorCurveData {keys [<at -0.0134 19.3841 RCIM_Linear>,<at 0.5079 12.8986 RCIM_Linear>,<at 0.9955 9.6044 RCIM_Linear>]}}
   }
   n102 = Trail controller - Bone: dyn_clothBack_mid_06_l(ComponentPose=@n101, Alpha=@n100) {
    bUsePlanarLimit=true
    chainLength=5
    debugLifeTime=0.1
    maxDeltaTime=0.0333
    alphaScaleBiasClamp {bInterpResult=true}
    baseJoint {boneName=capeBase}
    planarLimits [{drivingBone {boneName=dyn_clothBack_mid_01_l}, planeTransform {location=<0 5 0>, rotation=<rot 0 0 0>, scale=<1 1 1>}}]
    rotationLimits [
     {limitMax=<180 180 180>}
     {limitMax=<180 180 180>}
     {limitMax=<180 90 180>, limitMin=<-180 -90 -180>}
     {limitMax=<180 180 180>}
     {limitMax=<180 180 180>}
    ]
    rotationOffsets [<0 0 0>, <0 0 0>, <0 0 0>, <0 0 0>, <0 0 0>]
    trailBone {boneName=dyn_clothBack_mid_06_l}
    trailRelaxationSpeed {editorCurveData {keys [<at 0 15 RCIM_Linear>,<at 0.5289 13.1838 RCIM_Linear>,<at 1 6 RCIM_Linear>]}}
   }
   n103 = Trail controller - Bone: dyn_clothBack_outer_06_l(ComponentPose=@n102, Alpha=@n100) {
    bUsePlanarLimit=true
    chainLength=5
    debugLifeTime=0.1
    maxDeltaTime=0.0333
    alphaScaleBiasClamp {bInterpResult=true}
    baseJoint {boneName=capeBase}
    planarLimits [{drivingBone {boneName=dyn_clothBack_outer_mstr_l}, planeTransform {location=<0 0 0>, rotation=<rot 0 0 0>, scale=<1 1 1>}}]
    rotationLimits [
     {limitMax=<180 180 180>}
     {limitMax=<180 180 180>}
     {limitMax=<180 90 180>, limitMin=<-180 -90 -180>}
     {limitMax=<180 180 180>}
     {limitMax=<180 180 180>}
    ]
    rotationOffsets [<0 0 0>, <0 0 0>, <0 0 0>, <0 0 0>, <0 0 0>]
    trailBone {boneName=dyn_clothBack_outer_06_l}
    trailRelaxationSpeed {editorCurveData {keys [<at 0 15 RCIM_Linear>,<at 0.5289 13.1838 RCIM_Linear>,<at 1 6 RCIM_Linear>]}}
   }
   n104 = Trail controller - Bone: dyn_clothBack_inner_06_r(ComponentPose=@n103, Alpha=@n100) {
    bUsePlanarLimit=true
    chainLength=5
    debugLifeTime=0.1
    maxDeltaTime=0.333
    alphaScaleBiasClamp {bInterpResult=true}
    baseJoint {boneName=capeBase}
    planarLimits [{drivingBone {boneName=dyn_clothBack_inner_01_r}, planeTransform {location=<0 5 0>, rotation=<rot 0 0 0>, scale=<1 1 1>}}]
    rotationLimits [{limitMax=<180 180 180>}, {limitMax=<180 180 180>}, {limitMax=<180 180 180>}, {limitMax=<180 180 180>}, {limitMax=<180 180 180>}]
    rotationOffsets [<0 0 0>, <0 0 0>, <0 0 0>, <0 0 0>, <0 0 0>]
    trailBone {boneName=dyn_clothBack_inner_06_r}
    trailRelaxationSpeed {editorCurveData {keys [<at -0.0134 19.3841 RCIM_Linear>,<at 0.5079 12.8986 RCIM_Linear>,<at 0.9955 9.6044 RCIM_Linear>]}}
   }
   n105 = Trail controller - Bone: dyn_clothBack_mid_06_r(ComponentPose=@n104, Alpha=@n100) {
    bUsePlanarLimit=true
    chainLength=5
    debugLifeTime=0.1
    maxDeltaTime=0.0333
    alphaScaleBiasClamp {bInterpResult=true}
    baseJoint {boneName=capeBase}
    planarLimits [{drivingBone {boneName=dyn_clothBack_mid_01_r}, planeTransform {location=<0 5 0>, rotation=<rot 0 0 0>, scale=<1 1 1>}}]
    rotationLimits [
     {limitMax=<180 180 180>}
     {limitMax=<180 180 180>}
     {limitMax=<180 90 180>, limitMin=<-180 -90 -180>}
     {limitMax=<180 180 180>}
     {limitMax=<180 180 180>}
    ]
    rotationOffsets [<0 0 0>, <0 0 0>, <0 0 0>, <0 0 0>, <0 0 0>]
    trailBone {boneName=dyn_clothBack_mid_06_r}
    trailRelaxationSpeed {editorCurveData {keys [<at 0 15 RCIM_Linear>,<at 0.5289 13.1838 RCIM_Linear>,<at 1 6 RCIM_Linear>]}}
   }
   n106 = Trail controller - Bone: dyn_clothBack_outer_06_r(ComponentPose=@n105, Alpha=@n100) {
    bUsePlanarLimit=true
    chainLength=5
    debugLifeTime=0.1
    maxDeltaTime=0.0333
    alphaScaleBiasClamp {bInterpResult=true}
    baseJoint {boneName=capeBase}
    planarLimits [{drivingBone {boneName=dyn_clothBack_outer_mstr_r}, planeTransform {location=<0 0 0>, rotation=<rot 0 0 0>, scale=<1 1 1>}}]
    rotationLimits [
     {limitMax=<180 180 180>}
     {limitMax=<180 180 180>}
     {limitMax=<180 90 180>, limitMin=<-180 -90 -180>}
     {limitMax=<180 180 180>}
     {limitMax=<180 180 180>}
    ]
    rotationOffsets [<0 0 0>, <0 0 0>, <0 0 0>, <0 0 0>, <0 0 0>]
    trailBone {boneName=dyn_clothBack_outer_06_r}
    trailRelaxationSpeed {editorCurveData {keys [<at 0 15 RCIM_Linear>,<at 0.5289 13.1838 RCIM_Linear>,<at 1 6 RCIM_Linear>]}}
   }
   n107 = Component To Local(ComponentPose=@n106)
   n108 = Get bAnimDynamicsEnabled() -> (bAnimDynamicsEnabled:bool)
   n109 = Anim Dynamics - Bone: dyn_pouch(
    ComponentPose=Local To Component(LocalPose=Blend Poses by bool(BlendPose_0=@n94, BlendPose_1=@n107, bActiveValue=@n108) {blendType=Linear})
    Alpha=1
    ExternalForce="0,0,0"
   ) {
    angularDampingOverride=1
    angularSpringConstant=20000
    bAngularSpring=true
    bGravityOverrideInSimSpace=true
    bOverrideAngularDamping=true
    bUseGravityOverride=true
    gravityScale=0.35
    simulationSpace=RootRelative
    alphaBoolBlend {blendOption=Linear}
    boundBone {boneName=dyn_pouch}
    componentAppliedLinearAccClamp=<3000 3000 3000>
    componentLinearAccScale=<1 1 1>
    componentLinearVelScale=<1 1 1>
    gravityOverride {z=-980}
    physicsBodyDefinitions [{
      sphereCollisionRadius=0
      boundBone {boneName=dyn_pouch}
      boxExtents=<20 30 30>
      constraintSetup {angularLimitsMax=<15 15 15>, angularLimitsMin=<-15 -15 -15>, angularTarget=<1 0 0>}
      localJointOffset=<4 0 0>
     }]
   }
   Output Pose(Result=Component To Local(ComponentPose=@n109)) {layerGroup=None}
   n110 = "Constraint - dyn_clothFront_inner_mstr_r (PPPPPPP)"!(
    ComponentPose=Local To Component(LocalPose=Copy Pose From Mesh() {bUseAttachedParent=true})
    Alpha=1
    ConstraintWeights_0=0.1211
    ConstraintWeights_1:float=0.0212
    ConstraintWeights_2:float=0.1256
    ConstraintWeights_3:float=0.0401
    ConstraintWeights_4:float=0.1432
    ConstraintWeights_5:float=0.1651
    ConstraintWeights_6:float=0.3838
   ) -> (Pose) {
    alphaBoolBlend {blendOption=Linear}
    boneToModify {boneName=dyn_clothFront_inner_mstr_r}
    constraintSetup [
     {targetBone {boneName=upperarm_twist_01_r}}
     {targetBone {boneName=spine_02}}
     {targetBone {boneName=spine_03}}
     {targetBone {boneName=dyn_clothFront_inner_01_r}}
     {targetBone {boneName=spine_05}}
     {targetBone {boneName=clavicle_r}}
     {targetBone {boneName=spine_04}}
    ]
    constraintWeights [1,1,1,1,1,1,1]
   }
   n111 = "Constraint - dyn_clothFront_inner_mstr_l (PPPPPPP)"!(
    ComponentPose=@n110
    Alpha=1
    ConstraintWeights_0=0.1386
    ConstraintWeights_1:float=0.04
    ConstraintWeights_2:float=0.0136
    ConstraintWeights_3:float=0.1185
    ConstraintWeights_4:float=0.425
    ConstraintWeights_5:float=0.1375
    ConstraintWeights_6:float=0.1268
   ) -> (Pose) {
    alphaBoolBlend {blendOption=Linear}
    boneToModify {boneName=dyn_clothFront_inner_mstr_l}
    constraintSetup [
     {targetBone {boneName=clavicle_l}}
     {targetBone {boneName=dyn_clothFront_inner_01_l}}
     {targetBone {boneName=spine_02}}
     {targetBone {boneName=spine_03}}
     {targetBone {boneName=spine_04}}
     {targetBone {boneName=spine_05}}
     {targetBone {boneName=upperarm_twist_01_l}}
    ]
    constraintWeights [1,1,1,1,1,1,1]
   }
   n112 = "Constraint - dyn_clothFront_outer_mstr_r (PPPPPP)"!(
    ComponentPose=@n111
    Alpha=1
    ConstraintWeights_0=0.467
    ConstraintWeights_1:float=0.0458
    ConstraintWeights_2:float=0.061
    ConstraintWeights_3:float=0.0567
    ConstraintWeights_4:float=0.3294
    ConstraintWeights_5:float=0.04
   ) -> (Pose) {
    alphaBoolBlend {blendOption=Linear}
    boneToModify {boneName=dyn_clothFront_outer_mstr_r}
    constraintSetup [
     {targetBone {boneName=upperarm_twist_01_r}}
     {targetBone {boneName=spine_03}}
     {targetBone {boneName=spine_04}}
     {targetBone {boneName=spine_05}}
     {targetBone {boneName=clavicle_r}}
     {targetBone {boneName=dyn_clothFront_outer_01_r}}
    ]
    constraintWeights [1,1,1,1,1,1]
   }
   n113 = "Constraint - dyn_clothFront_outer_mstr_l (PPPPPP)"!(
    ComponentPose=@n112
    Alpha=1
    ConstraintWeights_0=0.04
    ConstraintWeights_1:float=0.323
    ConstraintWeights_2:float=0.0458
    ConstraintWeights_3:float=0.0658
    ConstraintWeights_4:float=0.0482
    ConstraintWeights_5:float=0.4772
   ) -> (Pose) {
    alphaBoolBlend {blendOption=Linear}
    boneToModify {boneName=dyn_clothFront_outer_mstr_l}
    constraintSetup [
     {targetBone {boneName=dyn_clothFront_outer_01_l}}
     {targetBone {boneName=clavicle_l}}
     {targetBone {boneName=spine_03}}
     {targetBone {boneName=spine_04}}
     {targetBone {boneName=spine_05}}
     {targetBone {boneName=upperarm_twist_01_l}}
    ]
    constraintWeights [1,1,1,1,1,1]
   }
   n114 = "Constraint - dyn_pouch_mstr (PPPP)"!(
    ComponentPose=@n113
    Alpha=1
    ConstraintWeights_0=0.3393
    ConstraintWeights_1:float=0.0606
    ConstraintWeights_2:float=0.0264
    ConstraintWeights_3:float=0.5737
   ) -> (Pose) {
    alphaBoolBlend {blendOption=Linear}
    boneToModify {boneName=dyn_pouch_mstr}
    constraintSetup [{targetBone {boneName=spine_01}}, {targetBone {boneName=spine_02}}, {targetBone {boneName=thigh_twist_01_l}}, {targetBone {boneName=pelvis}}]
    constraintWeights [1,1,1,1]
   }
   n115 = "Transform (Modify) Bone - Bone: dyn_pouch_mstr"(ComponentPose=@n114, Alpha=1, Translation="-0.5,0,0.000", Rotation="0,0,0", Scale="1,1,1") {
    translationMode=BMM_Additive
    translationSpace=BCS_BoneSpace
    alphaBoolBlend {blendInTime=10, blendOption=Linear, blendOutTime=10}
    boneToModify {boneName=dyn_pouch_mstr}
    translation {x=-0.5}
   }
   n116 = "Transform (Modify) Bone - Bone: dyn_pouch_mstr"!(ComponentPose=@n115, bAlphaBoolEnabled:bool=Get bIsCrouching(), Translation="0,1,0.25", Rotation="0,25,-10", Scale="1,1,1") -> (Pose) {
    alphaInputType=Bool
    rotationMode=BMM_Additive
    rotationSpace=BCS_BoneSpace
    translationMode=BMM_Additive
    translationSpace=BCS_BoneSpace
    alphaBoolBlend {blendInTime=10, blendOption=Linear, blendOutTime=10}
    boneToModify {boneName=dyn_pouch_mstr}
    rotation {roll=-10, yaw=25}
    translation {y=1, z=0.25}
   }
   Save cached pose 'preSimCache'(Pose=Component To Local(ComponentPose=@n116)) {cacheName=preSimCache}
  ]}
 EventGraph {nodes [
   Event BlueprintUpdateAnimation()
   > n1 = Sequence!(execute) -> (then_0, then_1, then_2:Exec, then_3:Exec, then_4:Exec, then_5:Exec, then_6:Exec, then_7:Exec)
   n2 = CheckForBackbling()
   > Set isWearingBackbling(isWearingBackbling=OR Boolean(A=@n2, B=@n2.bIsBackblingLong))
   n3 = Gravity Override State Function DEPRECATED(
    execute=@n1.then_1
    CharacterPartAnimInstance=Self-Reference()
    GameTimeInSeconds=GetGameTimeInSeconds()
    GravityOverrideParams=Get GravityOverrideParams() -> (GravityOverrideParams:Gravity Override Params Struct Structure)
   )
   > Set GravityOverride(GravityOverride=@n3)
   n4 = Get bIsJumping()
   n5 = Get bIsFalling()
   Set Is Jumping Float(execute=@n1.then_5, Is Jumping Float:double="To Float (Boolean)"(InBool=OR Boolean(A=@n5, B=@n4))) -> (Output_Get:double)
   Set Is Skydiving Float(execute=@n1.then_6, Is Skydiving Float:double="To Float (Boolean)"(InBool=Get bIsSkyDiving())) -> (Output_Get:double)
   n6 = Get bIsSkydiveDiveMode()
   Set Is Not Skydiving Dive Mode Float(execute=@n1.then_7, Is Not Skydiving Dive Mode Float:double="To Float (Boolean)"(InBool=NOT Boolean(A=@n6))) -> (Output_Get:double)
   n7 = GetAnimInstance(self=GetOwnerSkeletalMeshComponent())
   Set isPlayingFrontendIdle(
    execute=@n1.then_2
    isPlayingFrontendIdle:bool="float > float"(A=GetCurveValue(self=@n7, CurveName=FrontEnd_Idle_Bypass), B=0) {is=title K2Node_CallFunction}
   ) -> (Output_Get:bool)
   n8 = BreakTransform(InTransform=GetDeltaTransformFromRefPose(self=GetOwnerSkeletalMeshComponent(), BoneName=upperArm_l, BaseName=clavicle_l))
   Set L Arm Up Alpha(execute=@n1.then_3, L Arm Up Alpha:double=MapRangeClamped(Value=BreakRotator(InRot=@n8.Rotation).Yaw, InRangeA=-20, InRangeB=50)) -> (Output_Get:double)
   Set CrouchWalkAdjAlpha(execute=@n1.then_4, CrouchWalkAdjAlpha:bool=AND Boolean(A=Get bIsCrouching(), B=Get bIsMoving2D())) -> (Output_Get:bool)
   n9 = Event OnWeaponChanged()
   n10 = Get Dynamic Material Instance(execute=[@n9.then, @n18.then], SkeletalMeshComponent=GetPartSkeletalMeshComponent())
   > n11 = Sequence!(execute) -> (then_0, then_1)
   n12 = PickaxeItemDefinitionHasAnyMetaTag(Tags=Make Array("[0]"=(TagName="Cosmetics.MetaTags.Downpour")), bRequiresEquipped=true)
   > n13 = Branch(Condition=@n12)
   > n14 = Branch(Condition=Get bIsSkyDiving())
   SetScalarParameterValue(execute=[@n16.then, @n14.then, @n15.else, @n13.else], self=@n10, ParameterName=Hide Element 01, Value=0)
   SetScalarParameterValue(execute=[@n14.else, @n15.then], self=@n10, ParameterName=Hide Element 01, Value=1)
   n15 = Branch(execute=[@n17.then, @n11.then_1], Condition=@n12)
   n16 = Event OnBeginSkydiving()
   n17 = Event OnEndSkydiving()
   n18 = Event OnCharacterCustomizationLoaded()
  ]}
]
variables [
 CrouchWalkAdjAlpha {type=boolean}
 GravityOverride {type=Vector}
 GravityOverrideParams {type=GravityOverrideParamsStruct, default {skydive_GoingUp {gravityOverride=<0 0 -100>, windOffset=<0 -600 400>}}}
 Is Jumping Float {type=number}
 Is Not Skydiving Dive Mode Float {type=number}
 Is Skydiving Float {type=number}
 IsPlayingFrontendIdle {type=boolean}
 isWearingBackbling {type=boolean}
 L Arm Up Alpha {type=number}
]
